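# MF_SDFCirle is a UI material function in the Toasts foundation folder that draws a centred circle or ring from a signed distance field.
# It remaps UV to a -1 to 1 range and subtracts Size from the distance to centre.
# A static Stroke switch turns that field into an outline using its absolute value minus Stroke Thickness.
# It outputs the raw SDF and an inverted SmoothStep Shape, with Glow Start and Glow End setting edge softness.
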MaterialFunction /Game/UI/Foundation/Toasts/Materials/MF_SDFCirle
graph {nodes [
  n1 = FunctionInput(Preview=Constant()) {bUsePreviewValueAsDefault=true, description=0 - 1, inputName=Glow Start, inputType=FunctionInput_Scalar, sortPriority=3}
  n2 = FunctionInput(Preview=Constant() {r=0.01}) {bUsePreviewValueAsDefault=true, description=0 - 1, inputName=Glow End, inputType=FunctionInput_Scalar, sortPriority=4}
  n3 = FunctionInput(Preview=TextureCoordinate()) {bUsePreviewValueAsDefault=true, description=UV, inputName=UV, inputType=FunctionInput_Vector2, sortPriority=1}
  n4 = FunctionInput(Preview=Constant() {r=1}) {bUsePreviewValueAsDefault=true, description=0 - 1 to fit in the Frame, inputName=Size, inputType=FunctionInput_Scalar, sortPriority=2}
  n5 = Subtract(A=Distance(A=ConstantBiasScale(None=@n3) {bias=-0.5, scale=2}, B=Constant()), B=@n4)
  n6 = FunctionInput(Preview=Constant() {r=0.1}) {
   bUsePreviewValueAsDefault=true
   description="0 - 1 (1 being huuuge)"
   inputName=Stroke Thickness
   inputType=FunctionInput_Scalar
   sortPriority=6
  }
  n7 = StaticSwitch(
   True=Subtract(A=Abs(None=@n5), B=@n6)
   False=@n5
   Value=FunctionInput(Preview=StaticBool()) {bUsePreviewValueAsDefault=true, inputName=Stroke, inputType=FunctionInput_StaticBool, sortPriority=5}
  )
  FunctionOutput(None=OneMinus(None=SmoothStep(Min=@n1, Max=@n2, Value=@n7))) {description=this is the SDF + glow, outputName=Shape, sortPriority=1}
  FunctionOutput(None=@n7) {description=this is just the SDF before the smoothstep, outputName=SDF, sortPriority=2}
 ]}
