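# Translucent UI material for a toast icon, built around a channel-packed SDF texture of an alert triangle.
# The texture's RGBA feeds MF_UI_ChannelPackedDistanceField with constant glow bounds, producing per-channel fill masks.
# The blue fill drives a lerp from an orange-red constant and is also combined with the alpha fill through Max.
# A OneMinus node separately inverts the fill.
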Material /Game/UI/Foundation/Toasts/Materials/M_UI_ToastIcon_RGBA_SDF
graph {
 properties {
  bUsedWithStaticMesh=false
  blendMode=BLEND_Translucent
  materialDomain=MD_UI
  pixelDepthOffsetMode=PDOM_Legacy
  usedShadingModels=Default Lit
 }
 nodes [
  n1 = MaterialFunctionCall!(
   RGBA Channels=TextureSample() {texture=/Game/UI/Foundation/Toasts/Materials/Textures/T_UI_AlertTriangleIcon_SDF}.RGBA
   Glow Min=Constant() {r=0.4}
   Glow Max=Constant() {r=0.6}
   Outline Thickness
  ) {materialFunction=/FortUILibrary/MaterialFunctions/Basics/MF_UI_ChannelPackedDistanceField}
  n2 = ComponentMask(None=@n1.Fill) {b=true}
  LinearInterpolate(A=Constant3Vector() {constant=<rgba 0.8308 0.1651 0 1>}, Alpha=@n2)
  Max(A=@n2, B=ComponentMask(None=@n1.Fill) {a=true})
  OneMinus(None=@n1.Fill)
 ]
}
