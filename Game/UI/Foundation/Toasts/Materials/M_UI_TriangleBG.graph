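# Translucent UI material for a toast background, with a panning triangle pattern, angled side edges and a bottom progress bar.
# Screen-aligned, aspect-corrected UVs tile the triangle texture, tinted blue, green or red by lerp parameters and revealed from white by a burst gradient.
# Dot-product gradients cut the slanted edges and trim mask, with widget-size data from UV channel 3 compensating for scale.
# The progress bar overlays blue and gold gradients and can be hidden.
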Material /Game/UI/Foundation/Toasts/Materials/M_UI_TriangleBG
graph {
 properties {
  bUsedWithStaticMesh=false
  blendMode=BLEND_Translucent
  materialDomain=MD_UI
  pixelDepthOffsetMode=PDOM_Legacy
  usedShadingModels=Default Lit
 }
 nodes [
  n1 = TextureCoordinate() {coordinateIndex=3}
  AppendVector(A=ComponentMask(None=@n1) {r=true}, B=ComponentMask(None=@n1) {g=true})
  Constant() {r=1}
  n2 = Multiply(A=OneMinus(None=ScalarParameter() {defaultValue=0.9, group=Trim, parameterName=Sharpness, sliderMax=1}), B=NamedRerouteUsage())
  n3 = TextureCoordinate()
  n4 = ComponentMask(None=@n3) {r=true}
  n5 = Divide(A=Multiply(A=NamedRerouteUsage(), B=ScalarParameter() {defaultValue=75, group=Trim, parameterName=Angle})) {constB=-360}
  n6 = NamedRerouteUsage()
  n7 = DotProduct(A=AppendVector(A=OneMinus(None=@n4), B=ComponentMask(None=@n3) {g=true}), B=AppendVector(A=Constant() {r=1}, B=Multiply(A=@n5, B=@n6)))
  n8 = NamedRerouteUsage()
  n9 = Multiply(A=@n8, B=ScalarParameter() {group=Trim, parameterName=RightOffset, sliderMax=1})
  n10 = NamedRerouteUsage()
  n11 = Saturate(None=Subtract(A=@n7, B=Multiply(A=@n9, B=@n10)))
  Floor(None=SmoothStep(Max=@n2, Value=@n11))
  n12 = NamedRerouteUsage()
  n13 = TextureCoordinate()
  n14 = ComponentMask(None=@n13) {r=true}
  n15 = Multiply(A=NamedRerouteUsage(), B=ScalarParameter() {defaultValue=75, group=Trim, parameterName=Angle})
  n16 = Divide(A=LinearInterpolate(A=@n15, B=Multiply(A=@n15) {constB=0.75}, Alpha=ScalarParameter() {parameterName=LerpAngleChange}) {constAlpha=0}) {constB=-360}
  n17 = NamedRerouteUsage()
  n18 = DotProduct(A=AppendVector(A=OneMinus(None=@n14), B=ComponentMask(None=@n13) {g=true}), B=AppendVector(A=Constant() {r=1}, B=Multiply(A=@n16, B=@n17)))
  n19 = ScalarParameter() {parameterName=LerpAngleChange}
  n20 = Multiply(A=ScalarParameter() {defaultValue=0.915, group=Trim, parameterName=LeftOffset, sliderMax=1}, B=@n8)
  n21 = Saturate(None=Subtract(A=LinearInterpolate(A=@n7, B=@n18, Alpha=@n19) {constAlpha=0}, B=Multiply(A=@n20, B=@n10)))
  n22 = Min(A=@n21, B=@n11)
  LinearInterpolate(A=@n12, B=@n12, Alpha=NamedRerouteDeclaration(None=SmoothStep(Max=@n2, Value=@n22)) {name=LeftSide, nodeColor=<rgba 0.9688 1 0 1>})
  n23 = TextureCoordinate()
  n24 = Subtract(A=ComponentMask(None=@n23) {g=true}, B=OneMinus(None=ScalarParameter() {defaultValue=0.025, parameterName=ProgressBarHeight}))
  n25 = MaterialFunctionCall(Input=@n24, Input Low=Constant(), Input High=ScalarParameter() {defaultValue=0.025, parameterName=ProgressBarHeight}) {materialFunction=&RemapValue_0-1}
  n26 = Multiply(A=Ceil(None=@n24))
  n27 = ComponentMask(None=@n23) {r=true}
  n28 = Saturate(None=Subtract(A=OneMinus(None=@n27), B=ScalarParameter() {parameterName=Progress}))
  n29 = LinearInterpolate(
   A=LinearInterpolate(A=Constant3Vector() {constant=<rgba 0.276 0.6329 1 1>}, B=Constant3Vector() {constant=<rgba 0.855 0.9301 1 1>}, Alpha=@n25.Result)
   B=LinearInterpolate(A=Constant3Vector() {constant=<rgba 1 0.5906 0.008 1>}, B=Constant3Vector() {constant=<rgba 1 0.7913 0.0086 1>}, Alpha=@n25.Result)
   Alpha=Ceil(None=Multiply(A=@n26, B=@n28))
  )
  n30 = OneMinus(None=@n26)
  LinearInterpolate(A=Add(A=Multiply(A=@n29, B=@n26), B=Multiply(A=@n30, B=@n12)), B=@n12, Alpha=ScalarParameter() {parameterName=HideProgressBar})
  MaterialFunctionCall() {materialFunction=&MF_UI_UVScaling}
  NamedRerouteDeclaration(None=@n21) {name=NonTrimGradient, nodeColor=<rgba 0 0.4375 1 1>}
  n31 = ScalarParameter() {parameterName=LerpToGreen}
  n32 = ScalarParameter() {parameterName=LerpToRed}
  n33 = LinearInterpolate(
   A=LinearInterpolate(A=Constant3Vector() {constant=<rgba 0.6038 0.8879 1 1>}, B=Constant3Vector() {constant=<rgba 0.8073 1 0.8266 1>}, Alpha=@n31)
   B=Constant3Vector() {constant=<rgba 1 0.8651 0.8073 1>}
   Alpha=@n32
  )
  n34 = LinearInterpolate(
   A=LinearInterpolate(A=Constant3Vector() {constant=<rgba 0.0331 0.402 1 1>}, B=Constant3Vector() {constant=<rgba 0 0.6771 0.0874 1>}, Alpha=@n31)
   B=Constant3Vector() {constant=<rgba 0.755 0.0421 0.0973 1>}
   Alpha=@n32
  )
  n35 = MaterialFunctionCall(Float2=MaterialFunctionCall() {materialFunction=&ScreenResolution}.Visible Resolution) {materialFunction=&BreakOutFloat2Components}
  n36 = AppendVector(A=Divide(A=@n35.R, B=Max(A=@n35.G)), B=Constant() {r=1})
  n37 = MaterialFunctionCall() {materialFunction=&ScreenAlignedUVs}
  n38 = Panner(
   Coordinate=Rotator(
    Coordinate=Multiply(A=@n36, B=@n37."X 100%, Y 100%")
    Time=ScalarParameter() {defaultValue=-0.0571, group=Triangles, parameterName=Tri Angle}
   ) {speed=6.2832}
   Speed=AppendVector(A=ScalarParameter() {defaultValue=0.0025, group=Triangles, parameterName=PanSpeed}, B=Constant())
  )
  n39 = ScalarParameter() {defaultValue=5, group=Triangles, parameterName=Tri Tile Amount}
  n40 = Multiply(
   A=TextureSample(UVs=Multiply(A=@n38, B=@n39)) {texture=/Game/Global/Textures/Default/Masks_and_Gradients/T-TriangleGrid}.R
   B=ScalarParameter() {defaultValue=0.2, group=Triangles, parameterName=TriangleDensity}
  )
  n41 = MaterialFunctionCall() {materialFunction=&GetUserInterfaceUV}
  n42 = LinearInterpolate(
   A=ScalarParameter() {group=white gradient, parameterName=white_Burst_start}
   B=ScalarParameter() {defaultValue=1, group=white gradient, parameterName=white_Burst_end}
   Alpha=ScalarParameter() {group=white gradient, parameterName=Burst}
  )
  n43 = Multiply(
   A=ComponentMask(None=Subtract(A=@n41.Normalized UV, B=@n42)) {r=true}
   B=ScalarParameter() {defaultValue=1.5, group=white gradient, parameterName=White_sharpness}
  )
  NamedRerouteDeclaration(None=LinearInterpolate(A=Constant() {r=1}, B=LinearInterpolate(A=@n33, B=@n34, Alpha=@n40), Alpha=Saturate(None=@n43))) {name=TriangleBackground, nodeColor=<rgba 0.0781 0 1 1>}
  n44 = TextureCoordinate() {coordinateIndex=3}
  n45 = ComponentMask(None=@n44) {r=true}
  n46 = NamedRerouteDeclaration(None=Divide(A=ComponentMask(None=@n44) {g=true}, B=@n45) {constA=250, constB=250}) {name=ScaleX, nodeColor=<rgba 0 0 0 1>}
  NamedRerouteDeclaration(None=Multiply(
    A=Divide(B=@n46) {constA=0.1572, constB=0.1572, desc=Designed at height}
    B=Divide(B=@n45) {constA=719, constB=6.36, desc=Designed at height}
   )) {name=XAdj, nodeColor=<rgba 0.4297 0 1 1>}
  NamedRerouteUsage()
  NamedRerouteUsage()
  VectorParameter() {group=Left Bar, parameterName=Trim_Color, defaultValue=<rgba 0.9301 1 0.0103 1>}
 ]
}
