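# Material function for toast UI that outputs a radial sweep Mask driven by a Progress input, with an EdgeSoftness control.
# Centres the UVs and converts them to a normalised 0–1 angle with fast Arctangent2 and Frac.
# A SmoothStep softens the sweep edge from a Progress-derived threshold across the EdgeSoftness width.
# Floor of EdgeSoftness blends to a hard Floor(Progress + angle) mask, so the default of 1 gives a hard edge.
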MaterialFunction /Game/UI/Foundation/Toasts/Materials/MF_RadialMask
graph {nodes [
  n1 = FunctionInput(Preview=Constant() {r=0.65}) {inputName=Progress, inputType=FunctionInput_Scalar, sortPriority=1}
  n2 = Clamp(None=LinearInterpolate(Alpha=OneMinus(None=@n1)) {constA=-0.01}) {minDefault=-0.1}
  n3 = FunctionInput() {bUsePreviewValueAsDefault=true, inputName=EdgeSoftness, inputType=FunctionInput_Scalar, previewValue=<1 0 0 1>}
  n4 = Add(A=@n2, B=@n3)
  n5 = TextureCoordinate()
  n6 = OneMinus(None=Subtract(A=FunctionInput(Preview=@n5) {bUsePreviewValueAsDefault=true, inputName=UVs, inputType=FunctionInput_Vector2}) {constB=-0.5})
  n7 = Arctangent2Fast(Y=ComponentMask(None=@n6) {r=true}, X=ComponentMask(None=@n6) {g=true})
  n8 = Frac(None=Divide(A=ComponentMask(None=@n7) {r=true}, B=Constant() {r=6.2832}))
  n9 = Add(A=@n1, B=@n8)
  FunctionOutput(None=LinearInterpolate(
    A=MaterialFunctionCall(Alpha=@n8, Min=@n2, Max=@n4) {materialFunction=&SmoothStep}.Result
    B=Floor(None=@n9)
    Alpha=Floor(None=@n3)
   )) {outputName=Mask}
 ]}
