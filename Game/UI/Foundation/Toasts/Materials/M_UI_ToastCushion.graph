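# UI-domain translucent material for a toast cushion, drawing white with a blue gradient strip along the bottom for a progress bar.
# ProgressBarHeight offsets the V coordinate, and a Ceil of the result masks the strip without branching.
# The same offset, remapped to 0-1, blends the strip vertically from blue to pale blue.
# HideProgressBar lerps the composite to plain white; no input here drives horizontal fill.
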Material /Game/UI/Foundation/Toasts/Materials/M_UI_ToastCushion
graph {
 properties {
  bUsedWithStaticMesh=false
  blendMode=BLEND_Translucent
  materialDomain=MD_UI
  pixelDepthOffsetMode=PDOM_Legacy
  usedShadingModels=Default Lit
 }
 nodes [
  Constant() {r=1}
  n1 = Subtract(A=ComponentMask(None=TextureCoordinate()) {g=true}, B=OneMinus(None=ScalarParameter() {defaultValue=0.025, parameterName=ProgressBarHeight}))
  n2 = LinearInterpolate(
   A=Constant3Vector() {constant=<rgba 0.276 0.6329 1 1>}
   B=Constant3Vector() {constant=<rgba 0.855 0.9301 1 1>}
   Alpha=MaterialFunctionCall(Input=@n1, Input Low=Constant(), Input High=ScalarParameter() {defaultValue=0.025, parameterName=ProgressBarHeight}) {materialFunction=&RemapValue_0-1}.Result
  )
  n3 = Ceil(None=@n1)
  n4 = OneMinus(None=@n3)
  n5 = Constant() {r=1}
  LinearInterpolate(A=Add(A=Multiply(A=@n2, B=@n3), B=Multiply(A=@n4, B=@n5)), B=@n5, Alpha=ScalarParameter() {parameterName=HideProgressBar})
 ]
}
