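# Toast widget on FortPlayerPowerRatingToast that announces a player power change with banner, rating, difference and description.
# ShowToast plays a slide-in Intro, then a looping glow pulse runs while a display-time timer, paused on mouse hover, counts down.
# The slide-up, fade-out Outro stops the pulse and calls HideToast when finished.
# Construct switches between power icons and Moonbeam text backings using the bAllStWMoonbeamHUD runtime option.
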WidgetBlueprint /Game/UI/Foundation/Toasts/PowerToastWidget
animations [
 Intro {
  binding_count=3
  display_label=""
  length_seconds=0.5
  object=&self:Intro
  sequence {
   bindings [
    SizeBox_2 {
     id=#1
     possessed_class=SizeBox
     tracks [{
       type=MovieScene2DTransformTrack
       sections [{
         type=MovieScene2DTransformSection
         end_frame=5
         start_frame=0
         channels [
          Translation.X {type=MovieSceneScriptingFloatChannel, keys [[0,-640],[5,0]]}
          Translation.Y {type=MovieSceneScriptingFloatChannel, const [0,1]}
          Angle {type=MovieSceneScriptingFloatChannel, const [0,1]}
          Scale.X {type=MovieSceneScriptingFloatChannel, const [1,1]}
          Scale.Y {type=MovieSceneScriptingFloatChannel, const [1,1]}
          Shear.X {type=MovieSceneScriptingFloatChannel, const [0,1]}
          Shear.Y {type=MovieSceneScriptingFloatChannel, const [0,1]}
         ]
        }]
      }]
    }
    Shadow {
     id=#2
     possessed_class=CommonBorder
     tracks [{
       type=MovieSceneColorTrack
       display_name=Brush Color
       sections [{
         type=MovieSceneColorSection
         end_frame=5
         start_frame=0
         channels [
          Color.R {type=MovieSceneScriptingFloatChannel, const [1,1]}
          Color.G {type=MovieSceneScriptingFloatChannel, const [1,1]}
          Color.B {type=MovieSceneScriptingFloatChannel, const [1,1]}
          Color.A {type=MovieSceneScriptingFloatChannel, keys [[0,0],[5,0.4]]}
         ]
        }]
      }]
    }
    FullBackground {
     id=#3
     possessed_class=CommonBorder
     tracks [
      {
       type=MovieSceneColorTrack
       display_name=Content Color and Opacity
       sections [{
         type=MovieSceneColorSection
         end_frame=5
         start_frame=0
         channels [
          Color.R {type=MovieSceneScriptingFloatChannel, const [1,1]}
          Color.G {type=MovieSceneScriptingFloatChannel, const [1,1]}
          Color.B {type=MovieSceneScriptingFloatChannel, const [1,1]}
          Color.A {type=MovieSceneScriptingFloatChannel, keys [[0,0],[5,1]]}
         ]
        }]
      }
      {
       type=MovieSceneColorTrack
       display_name=Brush Color
       sections [{
         type=MovieSceneColorSection
         end_frame=5
         start_frame=0
         channels [
          Color.R {type=MovieSceneScriptingFloatChannel, const [1,1]}
          Color.G {type=MovieSceneScriptingFloatChannel, const [1,1]}
          Color.B {type=MovieSceneScriptingFloatChannel, const [1,1]}
          Color.A {type=MovieSceneScriptingFloatChannel, keys [[0,0],[5,1]]}
         ]
        }]
      }
     ]
    }
   ]
   display_rate {numerator=10}
  }
 }
 Outro {
  binding_count=4
  display_label=""
  length_seconds=0.5
  object=&self:Outro
  sequence {bindings [
    Shadow {
     id=#4
     possessed_class=CommonBorder
     tracks [{
       type=MovieSceneColorTrack
       display_name=Brush Color
       sections [{
         type=MovieSceneColorSection
         end_frame=10
         start_frame=0
         channels [
          Color.R {type=MovieSceneScriptingFloatChannel, const [1,1]}
          Color.G {type=MovieSceneScriptingFloatChannel, const [1,1]}
          Color.B {type=MovieSceneScriptingFloatChannel, const [1,1]}
          Color.A {type=MovieSceneScriptingFloatChannel, keys [[0,0.4],[10,0]]}
         ]
        }]
      }]
    }
    FullBackground {
     id=#5
     possessed_class=CommonBorder
     tracks [
      {
       type=MovieSceneColorTrack
       display_name=Content Color and Opacity
       sections [{
         type=MovieSceneColorSection
         end_frame=10
         start_frame=0
         channels [
          Color.R {type=MovieSceneScriptingFloatChannel, const [1,1]}
          Color.G {type=MovieSceneScriptingFloatChannel, const [1,1]}
          Color.B {type=MovieSceneScriptingFloatChannel, const [1,1]}
          Color.A {type=MovieSceneScriptingFloatChannel, keys [[0,1],[10,0]]}
         ]
        }]
      }
      {
       type=MovieSceneColorTrack
       display_name=Brush Color
       sections [{
         type=MovieSceneColorSection
         end_frame=10
         start_frame=0
         channels [
          Color.R {type=MovieSceneScriptingFloatChannel, const [1,1]}
          Color.G {type=MovieSceneScriptingFloatChannel, const [1,1]}
          Color.B {type=MovieSceneScriptingFloatChannel, const [1,1]}
          Color.A {type=MovieSceneScriptingFloatChannel, keys [[0,1],[10,0]]}
         ]
        }]
      }
     ]
    }
    SizeBox_2 {
     id=#6
     possessed_class=SizeBox
     tracks [{
       type=MovieScene2DTransformTrack
       sections [{
         type=MovieScene2DTransformSection
         end_frame=10
         start_frame=0
         channels [
          Translation.X {type=MovieSceneScriptingFloatChannel, const [0,2]}
          Translation.Y {type=MovieSceneScriptingFloatChannel, keys [[0,0],[10,-640]]}
          Angle {type=MovieSceneScriptingFloatChannel}
          Scale.X {type=MovieSceneScriptingFloatChannel}
          Scale.Y {type=MovieSceneScriptingFloatChannel}
          Shear.X {type=MovieSceneScriptingFloatChannel}
          Shear.Y {type=MovieSceneScriptingFloatChannel}
         ]
        }]
      }]
    }
    Glow {
     id=#7
     possessed_class=CommonBorder
     tracks [{
       type=MovieSceneColorTrack
       display_name=Brush Color
       sections [{
         type=MovieSceneColorSection
         end_frame=10
         start_frame=0
         channels [
          Color.R {type=MovieSceneScriptingFloatChannel, const [1,1]}
          Color.G {type=MovieSceneScriptingFloatChannel, const [1,1]}
          Color.B {type=MovieSceneScriptingFloatChannel, const [1,1]}
          Color.A {type=MovieSceneScriptingFloatChannel, keys [[0,0.5],[10,1]]}
         ]
        }]
      }]
    }
   ]}
 }
 LookAtMe {
  binding_count=1
  display_label=""
  length_seconds=0.6
  object=&self:LookAtMe
  sequence {bindings [Glow {
     id=#8
     possessed_class=CommonBorder
     tracks [{
       type=MovieSceneColorTrack
       display_name=Brush Color
       sections [{
         type=MovieSceneColorSection
         end_frame=12
         start_frame=0
         channels [
          Color.R {type=MovieSceneScriptingFloatChannel, const [1,1]}
          Color.G {type=MovieSceneScriptingFloatChannel, const [1,1]}
          Color.B {type=MovieSceneScriptingFloatChannel, const [1,1]}
          Color.A {type=MovieSceneScriptingFloatChannel, keys [[0,0.5],[6,1],[12,0.5]]}
         ]
        }]
      }]
    }]}
 }
]
graphs [EventGraph {nodes [
   HandleAnimationDelay() {is=title K2Node_CustomEvent}
   > PlayAnimation(InAnimation=Get Outro() -> (Outro:Widget Animation Object Reference))
   Event ShowToast() {is=title K2Node_Event}
   > SetVisibility(InVisibility=SelfHitTestInvisible)
   > SetVisibility(
    self=Get Moonbeam_Power_Back_Difference() -> (Moonbeam_Power_Back_Difference:Image Object Reference)
    InVisibility=GetVisibility!(self:Widget Object Reference=Get Text_PowerDifference() -> (Text_PowerDifference:Common Text Object Reference)) -> (ReturnValue:ESlateVisibility Enum=Visible)
   )
   n1 = Get Intro()
   > n2 = Branch(Condition=IsAnimationPlaying(InAnimation=@n1))
   PlayAnimation(execute=@n2.else, InAnimation=@n1)
   Event Construct()
   > n3 = Branch(Condition=Get bAllStWMoonbeamHUD!(self:Fort Runtime Options Object Reference=GetRuntimeOptions()) -> (bAllStWMoonbeamHUD:bool))
   > SetVisibility(self=[Get Image_Power() -> (Image_Power:Image Object Reference), Get Image_2() -> (Image_2:Image Object Reference)], InVisibility=Collapsed)
   SetVisibility(
    execute=@n3.else
    self=[Get Moonbeam_Power_Back_1() -> (Moonbeam_Power_Back_1:Image Object Reference), Get Moonbeam_Power_Back_Difference()]
    InVisibility=Collapsed
   )
   Event OnMouseEnter()
   > Pause Timer by Handle!(execute, Handle:Timer Handle Structure=Get AnimationDelayTimer() -> (AnimationDelayTimer:Timer Handle Structure)) -> (then)
   Event OnMouseLeave!() -> (OutputDelegate, then, MouseEvent:"Pointer Event Structure (by ref)") {is=title K2Node_Event}
   > Unpause Timer by Handle!(execute, Handle:Timer Handle Structure=Get AnimationDelayTimer()) -> (then)
   "Animation Finished (Intro)"!() -> (OutputDelegate:Delegate, then:Exec) {is=title K2Node_WidgetAnimationEvent}
   > PlayAnimation(InAnimation=Get LookAtMe() -> (LookAtMe:Widget Animation Object Reference), NumLoopsToPlay=0)
   > n4 = Set Timer by Event(
    Delegate=Create Event()
    Time=GetDisplayTime!(self:Fort Toast Widget Object Reference) -> (bOutDisplayTimeOverridden:bool, ReturnValue:float).ReturnValue
    bLooping=false
   )
   > Set AnimationDelayTimer(AnimationDelayTimer:Timer Handle Structure=@n4) -> (Output_Get:Timer Handle Structure)
   "Animation Started (Outro)"!() -> (OutputDelegate:Delegate, then:Exec) {is=title K2Node_WidgetAnimationEvent}
   > StopAnimation(InAnimation=Get LookAtMe())
   "Animation Finished (Outro)"!() -> (OutputDelegate:Delegate, then:Exec) {is=title K2Node_WidgetAnimationEvent}
   > HideToast(self:Fort Player Power Rating Toast Object Reference)
   PlaySound2D()
  ]}]
variables [AnimationDelayTimer {type=TimerHandle}]
widgets {
 info {inheritedWidgetCount=5, parentClass=/Script/FortniteUI.FortPlayerPowerRatingToast, rootWidgetClass=&SizeBox, widgetCount=27}
 widgets [
  SizeBox_2 {
   widgetClassPath=&SizeBox
   widget=SizeBox_2
   properties {
    bOverride_MinDesiredHeight=true
    bOverride_WidthOverride=true
    heightOverride=144
    maxDesiredHeight=144
    minDesiredHeight=152
    widthOverride=656
   }
  }
  VerticalBox_1 {widgetClassPath=&VerticalBox, parent=SizeBox_2, widget=VerticalBox_1, slot {class=SizeBoxSlot}}
  FullBackground {
   widgetClassPath=&CommonBorder
   parent=VerticalBox_1
   widget=FullBackground
   properties {style=/FortUILibrary/Styles/Border/Border-PowerToast.Border-PowerToast_C, padding=<ltrb 0 0 0 0>}
   slot {class=VerticalBoxSlot, properties {size {sizeRule=Fill, value=1}}}
  }
  Glow {
   widgetClassPath=&CommonBorder
   parent=FullBackground
   widget=Glow
   properties {style=/FortUILibrary/Styles/Border/Border-PowerToastGlow.Border-PowerToastGlow_C, brushColor=<rgba 1 1 1 0.5>, padding=<ltrb 16 0 0 0>}
   slot {class=BorderSlot, properties {padding=<ltrb 0 0 0 0>}}
  }
  HorizontalBox_1 {
   widgetClassPath=&HorizontalBox
   parent=Glow
   widget=HorizontalBox_1
   slot {class=BorderSlot, properties {padding=<ltrb 16 0 0 0>}}
  }
  SizeBox_5 {
   widgetClassPath=&SizeBox
   parent=HorizontalBox_1
   widget=SizeBox_5
   properties {bOverride_WidthOverride=true, widthOverride=96}
   slot {class=HorizontalBoxSlot}
  }
  VerticalBox_2 {
   widgetClassPath=&VerticalBox
   parent=SizeBox_5
   widget=VerticalBox_2
   slot {class=SizeBoxSlot, properties {horizontalAlignment=HAlign_Center, verticalAlignment=VAlign_Center}}
  }
  HorizontalBox_9 {
   widgetClassPath=&HorizontalBox
   parent=VerticalBox_2
   widget=HorizontalBox_9
   slot {class=VerticalBoxSlot, properties {padding=<ltrb 0 0 0 4>}}
  }
  Image_2 {
   widgetClassPath=&Image
   bIsVariable=true
   parent=HorizontalBox_9
   widget=Image_2
   properties {
    visibility=SelfHitTestInvisible
    brush {
     bIsDynamicallyLoaded=false
     drawAs=Image
     imageType=NoImage
     mirroring=NoMirror
     resourceName=None
     resourceObject=/Game/UI/Foundation/Textures/Icons/Shell/T-Icon-Power-64
     tiling=NoTile
     imageSize=<24 24>
     margin=<ltrb 0 0 0 0>
     outlineSettings {
      bUseBrushTransparency=false
      roundingType=HalfHeightRadius
      width=0
      color {colorUseRule=UseColor_Specified, specifiedColor=<rgba 0 0 0 0>}
      cornerRadii=<0 0 0 0>
     }
     tintColor {colorUseRule=UseColor_Specified, specifiedColor=<rgba 1 1 1 1>}
     uVRegion {bIsValid=false, max=<0 0>, min=<0 0>}
    }
    renderTransform {angle=0, scale=<1 1>, shear=<0 0>, translation=<2 0>}
   }
   slot {class=HorizontalBoxSlot, properties {horizontalAlignment=HAlign_Left, verticalAlignment=VAlign_Bottom}}
  }
  Overlay_1 {widgetClassPath=&Overlay, parent=HorizontalBox_9, widget=Overlay_1, slot {class=HorizontalBoxSlot}}
  Moonbeam_Power_Back_1 {
   widgetClassPath=&Image
   bIsVariable=true
   parent=Overlay_1
   widget=Moonbeam_Power_Back_1
   properties {
    brush {
     bIsDynamicallyLoaded=false
     drawAs=Image
     imageType=NoImage
     mirroring=NoMirror
     resourceName=None
     resourceObject=/Game/UI/Foundation/Textures/Icons/Items/T-Items-Power-BG
     tiling=NoTile
     imageSize=<40 24>
     margin=<ltrb 0 0 0 0>
     outlineSettings {
      bUseBrushTransparency=false
      roundingType=HalfHeightRadius
      width=0
      color {colorUseRule=UseColor_Specified, specifiedColor=<rgba 0 0 0 0>}
      cornerRadii=<0 0 0 0>
     }
     tintColor {colorUseRule=UseColor_Specified, specifiedColor=<rgba 1 1 1 1>}
     uVRegion {bIsValid=false, max=<0 0>, min=<0 0>}
    }
    colorAndOpacity=<rgba 0.131 0.1714 0.3278 0.65>
   }
   slot {class=OverlaySlot, properties {horizontalAlignment=HAlign_Fill, verticalAlignment=VAlign_Fill}}
  }
  Text_PowerRating {
   widgetClassPath=&CommonTextBlock
   bInherited=true
   parent=Overlay_1
   widget=Text_PowerRating
   properties {
    autoWrapText=true
    style=/FortUILibrary/Styles/Typography/Foundation/TextStyle-Base-XS-B.TextStyle-Base-XS-B_C
    text=3
    font {
     bForceMonospaced=false
     bMaterialIsStencil=false
     fontMaterial=null
     fontObject=&NotoSans
     letterSpacing=0
     monospacedWidth=1
     size=16
     skewAmount=0
     typefaceFontName=NotoSans-Bold
     outlineSettings {
      bApplyOutlineToDropShadows=false
      bMiteredCorners=false
      bSeparateFillAlpha=false
      outlineMaterial=null
      outlineSize=0
      outlineColor=<rgba 0 0 0 1>
     }
    }
    shadowOffset=<0 0>
   }
   slot {class=OverlaySlot, properties {horizontalAlignment=HAlign_Center, padding=<ltrb 8 0 8 0>}}
  }
  Banner_MemberBanner {
   widgetClassPath=/Script/FortniteUI.FortPlayerBanner
   bInherited=true
   bIsVariable=true
   parent=VerticalBox_2
   widget=Banner_MemberBanner
   slot {class=VerticalBoxSlot, properties {horizontalAlignment=HAlign_Center, verticalAlignment=VAlign_Top}}
  }
  LineSeparator {
   widgetClassPath=&Image
   bIsVariable=true
   parent=HorizontalBox_1
   widget=LineSeparator
   properties {
    brush {
     bIsDynamicallyLoaded=false
     drawAs=Image
     imageType=NoImage
     mirroring=NoMirror
     resourceName=None
     resourceObject=null
     tiling=NoTile
     imageSize=<3 112>
     margin=<ltrb 0 0 0 0>
     outlineSettings {
      bUseBrushTransparency=false
      roundingType=HalfHeightRadius
      width=0
      color {colorUseRule=UseColor_Specified, specifiedColor=<rgba 0 0 0 0>}
      cornerRadii=<0 0 0 0>
     }
     tintColor {colorUseRule=UseColor_Specified, specifiedColor=<rgba 1 1 1 1>}
     uVRegion {bIsValid=false, max=<0 0>, min=<0 0>}
    }
    colorAndOpacity=<rgba 1 1 1 0.4>
   }
   slot {class=HorizontalBoxSlot, properties {horizontalAlignment=HAlign_Center, verticalAlignment=VAlign_Center}}
  }
  VerticalBox_0 {
   widgetClassPath=&VerticalBox
   parent=HorizontalBox_1
   widget=VerticalBox_0
   slot {class=HorizontalBoxSlot, properties {verticalAlignment=VAlign_Center, padding=<ltrb 24 0 32 0>, size {sizeRule=Fill, value=1}}}
  }
  Text_Title {
   widgetClassPath=&CommonTextBlock
   bInherited=true
   parent=VerticalBox_0
   widget=Text_Title
   properties {
    autoWrapText=true
    flowDirectionPreference=Culture
    style=&TextStyle-Base-XS-B_C
    text="Notification Title!"
    font {
     bForceMonospaced=false
     bMaterialIsStencil=false
     fontMaterial=null
     fontObject=&NotoSans
     letterSpacing=0
     monospacedWidth=1
     size=16
     skewAmount=0
     typefaceFontName=NotoSans-Bold
     outlineSettings {
      bApplyOutlineToDropShadows=false
      bMiteredCorners=false
      bSeparateFillAlpha=false
      outlineMaterial=null
      outlineSize=0
      outlineColor=<rgba 0 0 0 1>
     }
    }
    shadowOffset=<0 0>
   }
   slot {class=VerticalBoxSlot}
  }
  Text_Description {
   widgetClassPath=&CommonTextBlock
   bInherited=true
   parent=VerticalBox_0
   widget=Text_Description
   properties {
    autoWrapText=true
    flowDirectionPreference=Culture
    style=/FortUILibrary/Styles/Typography/Foundation/TextStyle-Base-XS-EnchantedBlue.TextStyle-Base-XS-EnchantedBlue_C
    text=Your Power Level and F.O.R.T. stats have now increased.
    colorAndOpacity {colorUseRule=UseColor_Specified, specifiedColor=<rgba 0.0953 1 1 1>}
    font {
     bForceMonospaced=false
     bMaterialIsStencil=false
     fontMaterial=null
     fontObject=&NotoSans
     letterSpacing=0
     monospacedWidth=1
     size=16
     skewAmount=0
     typefaceFontName=NotoSans-Regular
     outlineSettings {
      bApplyOutlineToDropShadows=false
      bMiteredCorners=false
      bSeparateFillAlpha=false
      outlineMaterial=null
      outlineSize=0
      outlineColor=<rgba 0 0 0 1>
     }
    }
    shadowOffset=<0 0>
   }
   slot {class=VerticalBoxSlot}
  }
  SizeBox_6 {
   widgetClassPath=&SizeBox
   parent=HorizontalBox_1
   widget=SizeBox_6
   properties {bOverride_WidthOverride=true, widthOverride=96}
   slot {class=HorizontalBoxSlot}
  }
  Overlay_0 {widgetClassPath=&Overlay, parent=SizeBox_6, widget=Overlay_0, slot {class=SizeBoxSlot}}
  PowerIconGlow {
   widgetClassPath=&Image
   bIsVariable=true
   parent=Overlay_0
   widget=PowerIconGlow
   properties {
    brush {
     bIsDynamicallyLoaded=false
     drawAs=Image
     imageType=NoImage
     mirroring=NoMirror
     resourceName=None
     resourceObject=/FortUILibrary/Materials/Foundation/M-HomeBasePower-OuterGlow
     tiling=NoTile
     imageSize=<96 120>
     margin=<ltrb 0 0 0 0>
     outlineSettings {
      bUseBrushTransparency=false
      roundingType=HalfHeightRadius
      width=0
      color {colorUseRule=UseColor_Specified, specifiedColor=<rgba 0 0 0 0>}
      cornerRadii=<0 0 0 0>
     }
     tintColor {colorUseRule=UseColor_Specified, specifiedColor=<rgba 1 1 1 1>}
     uVRegion {bIsValid=false, max=<0 0>, min=<0 0>}
    }
    renderTransform {angle=0, scale=<1 1>, shear=<0 0>, translation=<-6 0>}
   }
   slot {class=OverlaySlot, properties {horizontalAlignment=HAlign_Center, verticalAlignment=VAlign_Center}}
  }
  VerticalBox_4 {
   widgetClassPath=&VerticalBox
   parent=Overlay_0
   widget=VerticalBox_4
   slot {class=OverlaySlot, properties {horizontalAlignment=HAlign_Fill, verticalAlignment=VAlign_Center}}
  }
  Overlay_2 {
   widgetClassPath=&Overlay
   parent=VerticalBox_4
   widget=Overlay_2
   slot {class=VerticalBoxSlot, properties {horizontalAlignment=HAlign_Center, verticalAlignment=VAlign_Center, padding=<ltrb 0 0 12 0>}}
  }
  Moonbeam_Power_Back_Difference {
   widgetClassPath=&Image
   bIsVariable=true
   parent=Overlay_2
   widget=Moonbeam_Power_Back_Difference
   properties {
    brush {
     bIsDynamicallyLoaded=false
     drawAs=Image
     imageType=NoImage
     mirroring=NoMirror
     resourceName=None
     resourceObject=&T-Items-Power-BG
     tiling=NoTile
     imageSize=<40 24>
     margin=<ltrb 0 0 0 0>
     outlineSettings {
      bUseBrushTransparency=false
      roundingType=HalfHeightRadius
      width=0
      color {colorUseRule=UseColor_Specified, specifiedColor=<rgba 0 0 0 0>}
      cornerRadii=<0 0 0 0>
     }
     tintColor {colorUseRule=UseColor_Specified, specifiedColor=<rgba 1 1 1 1>}
     uVRegion {bIsValid=false, max=<0 0>, min=<0 0>}
    }
    colorAndOpacity=<rgba 0.131 0.1714 0.3278 0.6>
   }
   slot {class=OverlaySlot, properties {horizontalAlignment=HAlign_Fill, verticalAlignment=VAlign_Fill}}
  }
  Text_PowerDifference {
   widgetClassPath=&CommonTextBlock
   bInherited=true
   bIsVariable=true
   parent=Overlay_2
   widget=Text_PowerDifference
   properties {
    autoWrapText=true
    style=/FortUILibrary/Styles/Typography/Foundation/TextStyle-Base-XS-B-EnchantedBlue.TextStyle-Base-XS-B-EnchantedBlue_C
    text=+999
    colorAndOpacity {colorUseRule=UseColor_Specified, specifiedColor=<rgba 0.0953 1 1 1>}
    font {
     bForceMonospaced=false
     bMaterialIsStencil=false
     fontMaterial=null
     fontObject=&NotoSans
     letterSpacing=0
     monospacedWidth=1
     size=16
     skewAmount=0
     typefaceFontName=NotoSans-Bold
     outlineSettings {
      bApplyOutlineToDropShadows=false
      bMiteredCorners=false
      bSeparateFillAlpha=false
      outlineMaterial=null
      outlineSize=0
      outlineColor=<rgba 0 0 0 1>
     }
    }
    shadowOffset=<0 0>
   }
   slot {class=OverlaySlot, properties {padding=<ltrb 8 0 8 0>}}
  }
  Image_Power {
   widgetClassPath=&Image
   bIsVariable=true
   parent=VerticalBox_4
   widget=Image_Power
   properties {
    visibility=SelfHitTestInvisible
    brush {
     bIsDynamicallyLoaded=false
     drawAs=Image
     imageType=NoImage
     mirroring=NoMirror
     resourceName=None
     resourceObject=/Game/UI/Foundation/Textures/Icons/Shell/T-Icon-Power-128
     tiling=NoTile
     imageSize=<56 56>
     margin=<ltrb 0 0 0 0>
     outlineSettings {
      bUseBrushTransparency=false
      roundingType=HalfHeightRadius
      width=0
      color {colorUseRule=UseColor_Specified, specifiedColor=<rgba 0 0 0 0>}
      cornerRadii=<0 0 0 0>
     }
     tintColor {colorUseRule=UseColor_Specified, specifiedColor=<rgba 1 1 1 1>}
     uVRegion {bIsValid=false, max=<0 0>, min=<0 0>}
    }
    renderTransform {angle=0, scale=<1 1>, shear=<0 0>, translation=<-6 0>}
   }
   slot {class=VerticalBoxSlot, properties {horizontalAlignment=HAlign_Center, verticalAlignment=VAlign_Center}}
  }
  Shadow {
   widgetClassPath=&CommonBorder
   parent=VerticalBox_1
   widget=Shadow
   properties {style=/FortUILibrary/Styles/Border/Border_Solid_DkBlue.Border_Solid_DkBlue_C, brushColor=<rgba 1 1 1 0.4>, padding=<ltrb 0 0 0 0>}
   slot {class=VerticalBoxSlot}
  }
  Spacer_0 {
   widgetClassPath=&Spacer
   parent=Shadow
   widget=Spacer_0
   properties {size=<0 8>}
   slot {class=BorderSlot, properties {padding=<ltrb 0 0 0 0>}}
  }
 ]
}
