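# Social notification toast showing title, description, image and an action button labeled by friend-notification type.
# A switcher picks hold or cursor button by cursor mode, touch and zone state.
# Intro plays with a sound; a display timer, lengthened for some party types, follows and pauses on hover or hold.
# Taking action or cancelling clears the timer, hides buttons and plays Outro_New in reverse after a short delay.
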WidgetBlueprint /Game/UI/Foundation/Toasts/ToastSocialWidget
animations [
 Intro {
  binding_count=5
  display_label=""
  length_seconds=0.75
  object=&self:Intro
  sequence {
   bindings [
    B_NotificationImage {
     id=#1
     possessed_class=CommonBorder
     tracks [{
       type=MovieScene2DTransformTrack
       sections [{
         type=MovieScene2DTransformSection
         end_frame=75
         start_frame=0
         channels [
          Translation.X {type=MovieSceneScriptingFloatChannel}
          Translation.Y {type=MovieSceneScriptingFloatChannel}
          Angle {type=MovieSceneScriptingFloatChannel, keys [[0,0],[25,0],[75,-360]]}
          Scale.X {type=MovieSceneScriptingFloatChannel, keys [[0,0],[14.9983,0.8471],[17.9983,1.3111],[21.9983,1.2062],[23.9983,1.0246],[25.9983,1]]}
          Scale.Y {type=MovieSceneScriptingFloatChannel, keys [[0,0],[10.9983,1.7484],[17.9983,0.795],[21.9983,1.2062],[23.9983,1.0246],[25.9983,1]]}
          Shear.X {type=MovieSceneScriptingFloatChannel}
          Shear.Y {type=MovieSceneScriptingFloatChannel}
         ]
        }]
      }]
    }
    B_NotificationImageExtended {
     id=#2
     possessed_class=CommonBorder
     tracks [{
       type=MovieScene2DTransformTrack
       sections [{
         type=MovieScene2DTransformSection
         end_frame=75
         start_frame=0
         channels [
          Translation.X {type=MovieSceneScriptingFloatChannel}
          Translation.Y {type=MovieSceneScriptingFloatChannel}
          Angle {type=MovieSceneScriptingFloatChannel}
          Scale.X {type=MovieSceneScriptingFloatChannel}
          Scale.Y {type=MovieSceneScriptingFloatChannel, thinned_from=6, keys [[0,0],[25,0],[42.9983,1.1427],[52.9983,1],[75,1]]}
          Shear.X {type=MovieSceneScriptingFloatChannel}
          Shear.Y {type=MovieSceneScriptingFloatChannel}
         ]
        }]
      }]
    }
    HorizontalBox_1 {
     id=#3
     possessed_class=HorizontalBox
     tracks [{
       type=MovieScene2DTransformTrack
       sections [{
         type=MovieScene2DTransformSection
         end_frame=75
         start_frame=0
         channels [
          Translation.X {type=MovieSceneScriptingFloatChannel}
          Translation.Y {type=MovieSceneScriptingFloatChannel}
          Angle {type=MovieSceneScriptingFloatChannel}
          Scale.X {type=MovieSceneScriptingFloatChannel, const [1,5]}
          Scale.Y {type=MovieSceneScriptingFloatChannel, keys [[0,1],[34.9983,1],[44.9983,1.1489],[62,0.8419],[75,1]]}
          Shear.X {type=MovieSceneScriptingFloatChannel}
          Shear.Y {type=MovieSceneScriptingFloatChannel}
         ]
        }]
      }]
    }
    B_Dropshadow {
     id=#4
     possessed_class=Border
     tracks [{
       type=MovieScene2DTransformTrack
       sections [{
         type=MovieScene2DTransformSection
         end_frame=75
         start_frame=0
         channels [
          Translation.X {type=MovieSceneScriptingFloatChannel}
          Translation.Y {type=MovieSceneScriptingFloatChannel}
          Angle {type=MovieSceneScriptingFloatChannel}
          Scale.X {type=MovieSceneScriptingFloatChannel, keys [[0,0],[52.9983,0],[62.9983,1],[64.9983,1.05],[69.9983,0.96],[75,1]]}
          Scale.Y {type=MovieSceneScriptingFloatChannel}
          Shear.X {type=MovieSceneScriptingFloatChannel}
          Shear.Y {type=MovieSceneScriptingFloatChannel}
         ]
        }]
      }]
    }
    CommonBorder_MainWidget {
     id=#5
     possessed_class=CommonBorder
     tracks [
      {
       type=MovieScene2DTransformTrack
       sections [{
         type=MovieScene2DTransformSection
         end_frame=75
         start_frame=0
         channels [
          Translation.X {type=MovieSceneScriptingFloatChannel}
          Translation.Y {type=MovieSceneScriptingFloatChannel}
          Angle {type=MovieSceneScriptingFloatChannel}
          Scale.X {type=MovieSceneScriptingFloatChannel, keys [[0,0],[52.9983,0],[62.9983,1],[69.9983,0.9461],[75,1]]}
          Scale.Y {type=MovieSceneScriptingFloatChannel, const [1,2]}
          Shear.X {type=MovieSceneScriptingFloatChannel}
          Shear.Y {type=MovieSceneScriptingFloatChannel}
         ]
        }]
      }
      {
       type=MovieSceneColorTrack
       display_name=Content Color and Opacity
       sections [{
         type=MovieSceneColorSection
         end_frame=75
         start_frame=0
         channels [
          Color.R {type=MovieSceneScriptingFloatChannel}
          Color.G {type=MovieSceneScriptingFloatChannel}
          Color.B {type=MovieSceneScriptingFloatChannel}
          Color.A {type=MovieSceneScriptingFloatChannel, keys [[0,0],[57.9983,0],[68.9983,1],[75,1]]}
         ]
        }]
      }
     ]
    }
   ]
   display_rate {numerator=100}
  }
 }
 Outro_New {
  binding_count=5
  display_label=""
  length_seconds=0.63
  object=&self:Outro_New
  sequence {
   bindings [
    B_NotificationImage {
     id=#1
     possessed_class=CommonBorder
     tracks [{
       type=MovieScene2DTransformTrack
       sections [{
         type=MovieScene2DTransformSection
         end_frame=75
         start_frame=0
         channels [
          Translation.X {type=MovieSceneScriptingFloatChannel}
          Translation.Y {type=MovieSceneScriptingFloatChannel}
          Angle {type=MovieSceneScriptingFloatChannel, const [-360,2]}
          Scale.X {type=MovieSceneScriptingFloatChannel, keys [[0,0],[14.9983,0.8471],[17.9983,1.3111],[21.9983,1.4447],[23.9983,1.0246],[25.9983,1]]}
          Scale.Y {type=MovieSceneScriptingFloatChannel, keys [[0,0],[10.9983,1.7484],[17.9983,0.795],[21.9983,0.7927],[23.9983,1.0246],[25.9983,1]]}
          Shear.X {type=MovieSceneScriptingFloatChannel}
          Shear.Y {type=MovieSceneScriptingFloatChannel}
         ]
        }]
      }]
    }
    B_NotificationImageExtended {
     id=#2
     possessed_class=CommonBorder
     tracks [{
       type=MovieScene2DTransformTrack
       sections [{
         type=MovieScene2DTransformSection
         end_frame=75
         start_frame=0
         channels [
          Translation.X {type=MovieSceneScriptingFloatChannel}
          Translation.Y {type=MovieSceneScriptingFloatChannel}
          Angle {type=MovieSceneScriptingFloatChannel}
          Scale.X {type=MovieSceneScriptingFloatChannel}
          Scale.Y {type=MovieSceneScriptingFloatChannel, thinned_from=6, keys [[0,0],[16.9983,0],[31,1.1427],[40.9983,1],[62.9983,1]]}
          Shear.X {type=MovieSceneScriptingFloatChannel}
          Shear.Y {type=MovieSceneScriptingFloatChannel}
         ]
        }]
      }]
    }
    HorizontalBox_1 {
     id=#3
     possessed_class=HorizontalBox
     tracks [{
       type=MovieScene2DTransformTrack
       sections [{
         type=MovieScene2DTransformSection
         end_frame=75
         start_frame=0
         channels [
          Translation.X {type=MovieSceneScriptingFloatChannel}
          Translation.Y {type=MovieSceneScriptingFloatChannel}
          Angle {type=MovieSceneScriptingFloatChannel}
          Scale.X {type=MovieSceneScriptingFloatChannel, const [1,5]}
          Scale.Y {type=MovieSceneScriptingFloatChannel, keys [[0,1],[26.9983,1],[36.9983,1.1489],[62.9983,1]]}
          Shear.X {type=MovieSceneScriptingFloatChannel}
          Shear.Y {type=MovieSceneScriptingFloatChannel}
         ]
        }]
      }]
    }
    B_Dropshadow {
     id=#4
     possessed_class=Border
     tracks [{
       type=MovieScene2DTransformTrack
       sections [{
         type=MovieScene2DTransformSection
         end_frame=75
         start_frame=0
         channels [
          Translation.X {type=MovieSceneScriptingFloatChannel}
          Translation.Y {type=MovieSceneScriptingFloatChannel}
          Angle {type=MovieSceneScriptingFloatChannel}
          Scale.X {type=MovieSceneScriptingFloatChannel, keys [[0,0],[40.9983,0],[50.9983,1],[52.9983,1.05],[58,0.96],[62.9983,1]]}
          Scale.Y {type=MovieSceneScriptingFloatChannel}
          Shear.X {type=MovieSceneScriptingFloatChannel}
          Shear.Y {type=MovieSceneScriptingFloatChannel}
         ]
        }]
      }]
    }
    CommonBorder_MainWidget {
     id=#5
     possessed_class=CommonBorder
     tracks [
      {
       type=MovieScene2DTransformTrack
       sections [{
         type=MovieScene2DTransformSection
         end_frame=75
         start_frame=0
         channels [
          Translation.X {type=MovieSceneScriptingFloatChannel}
          Translation.Y {type=MovieSceneScriptingFloatChannel}
          Angle {type=MovieSceneScriptingFloatChannel}
          Scale.X {type=MovieSceneScriptingFloatChannel, keys [[0,0],[40.9983,0],[50.9983,1],[62.9983,1]]}
          Scale.Y {type=MovieSceneScriptingFloatChannel, const [1,2]}
          Shear.X {type=MovieSceneScriptingFloatChannel}
          Shear.Y {type=MovieSceneScriptingFloatChannel}
         ]
        }]
      }
      {
       type=MovieSceneColorTrack
       display_name=Content Color and Opacity
       sections [{
         type=MovieSceneColorSection
         end_frame=75
         start_frame=0
         channels [
          Color.R {type=MovieSceneScriptingFloatChannel}
          Color.G {type=MovieSceneScriptingFloatChannel}
          Color.B {type=MovieSceneScriptingFloatChannel}
          Color.A {type=MovieSceneScriptingFloatChannel, keys [[0,0],[46,0],[57,1],[62.9983,1]]}
         ]
        }]
      }
     ]
    }
   ]
   display_rate {numerator=100}
  }
 }
]
graphs [
 CleanupBeforeClosing {nodes [
   CleanupBeforeClosing() {is=title K2Node_FunctionEntry}
   > Clear and Invalidate Timer by Handle(Handle=Get AnimationDelayTimer() -> (AnimationDelayTimer:Timer Handle Structure))
   > SetVisibility(self=Get Switcher_InputButtons() -> (Switcher_InputButtons:Common Widget Switcher Legacy Object Reference), InVisibility=Collapsed)
  ]}
 EventGraph {nodes [
   n1 = "On Button Clicked (Button_Hold)"!() -> (OutputDelegate, then, Button:Common Button Legacy Object Reference) {is=title K2Node_ComponentBoundEvent}
   n2 = TakeAction(execute=[@n1.then, @n3.then, @n8.then])
   CleanupBeforeClosing(execute=[@n2.then, @n7.then]) {is=title K2Node_CallFunction}
   > Delay(Duration=0.2)
   > PlayOutro()
   n3 = "On Button Clicked (Button_Cursor)"!() -> (OutputDelegate, then, Button:Common Button Legacy Object Reference) {is=title K2Node_ComponentBoundEvent}
   HandleOutroFinished() {is=title K2Node_CustomEvent}
   > NotifyCloseAnimationFinished(self:Fort Toast Widget Object Reference)
   HandleIntroFinished() {is=title K2Node_CustomEvent}
   > n4 = Set Timer by Event(
    Delegate=Create Event()
    Time=GetDisplayTime!(self:Fort Toast Widget Object Reference) -> (bOutDisplayTimeOverridden:bool, ReturnValue:float).ReturnValue
    bLooping=false
   )
   > Set AnimationDelayTimer(AnimationDelayTimer:Timer Handle Structure=@n4) -> (Output_Get:Timer Handle Structure)
   PlayOutro() {is=title K2Node_CustomEvent}
   > PlayAnimation(InAnimation=Get Outro_New() -> (Outro_New:Widget Animation Object Reference), PlayMode=Reverse, PlaybackSpeed=1.5)
   Event OnPlayOpenAnimation() {is=title K2Node_Event}
   > StartIntro()
   Event OnUserStartedHoldActionNotification() {is=title K2Node_Event}
   > Pause Timer by Handle!(execute, Handle:Timer Handle Structure=Get AnimationDelayTimer()) -> (then)
   Event OnMouseEnter()
   > Pause Timer by Handle(Handle=Get AnimationDelayTimer())
   Event OnMouseLeave!() -> (OutputDelegate, then, MouseEvent:"Pointer Event Structure (by ref)") {is=title K2Node_Event}
   > Unpause Timer by Handle!(execute, Handle:Timer Handle Structure=Get AnimationDelayTimer()) -> (then)
   n5 = Event OnCursorModeChanged!() -> (OutputDelegate, then, bCursorModeEnabled:bool) {is=title K2Node_Event}
   n6 = IsUsingTouch(PlayerController=GetOwningPlayer())
   > SetActiveWidget(
    self:Widget Switcher Object Reference=Get Switcher_InputButtons()
    Widget=Select(
     Option 0:Icon Text Button Object Reference=Get Button_Hold() -> (Button_Hold:Icon Text Button Object Reference)
     Option 1:Icon Text Button Object Reference=Get Button_Cursor() -> (Button_Cursor:Icon Text Button Object Reference)
     Index=OR Boolean(A=@n5.bCursorModeEnabled, B=@n6)
    ) -> (ReturnValue:Icon Text Button Object Reference)
   )
   > Unpause Timer by Handle(Handle=Get AnimationDelayTimer())
   Event Construct()
   > SetupAnimBindings!(execute, self, Unbind:bool=false) -> (then)
   n7 = Event OnCancelNotification() {is=title K2Node_Event}
   n8 = Event OnTakeActionNotification() {is=title K2Node_Event}
   n9 = Event OnNotificationSet!() -> (OutputDelegate, then, Notification:Fort UINotification Object Reference) {is=title K2Node_Event}
   > UpdateToast!(execute, self, Notification:Fort UINotification Object Reference=@n9.Notification) -> (then)
   Event OnInputModeChanged!() -> (OutputDelegate, then, NewInputType:ECommonInputType Enum) {is=title K2Node_Event}
   > SetActiveWidgetIndex(self=Get Switcher_InputButtons(), Index=GetActiveWidgetIndex!() -> (ReturnValue:int=0))
   > Unpause Timer by Handle(Handle=Get AnimationDelayTimer())
   Event Destruct()
   > SetupAnimBindings(Unbind=true)
   n10 = Event OnNotificationUpdated!() -> (OutputDelegate, then, Notification:Fort UINotification Object Reference) {is=title K2Node_Event}
   > UpdateToast(Notification=@n10.Notification)
  ]}
 GetActiveWidgetIndex {nodes [
   GetActiveWidgetIndex() {is=title K2Node_FunctionEntry}
   n1 = IsInZone!(self:Fort Global UIContext Object Reference=Get FortGlobalUIContext() -> (ReturnValue:Fort Global UIContext Object Reference) {is=title K2Node_GetSubsystem}) -> (ReturnValue:bool)
   n2 = NOT Boolean(A=IsUsingTouch(PlayerController=GetOwningPlayer()))
   > Return Node!(execute, ReturnValue:int=SelectInt(A=0, B=1, bPickA=AND Boolean(A=@n1, B=@n2)))
  ]}
 GetButtonCursorText {nodes [
   GetButtonCursorText() {is=title K2Node_FunctionEntry}
   > n1 = Cast To FortUIFriendNotification!(execute, Object=Get CachedNotification() -> (CachedNotification:Fort UINotification Object Reference)) -> (then, CastFailed, AsFort UIFriend Notification:Fort UIFriend Notification Object Reference) {is=title K2Node_DynamicCast}
   > n2 = Switch on EFortUIFriendNotificationType!(
    execute
    Selection:EFortUIFriendNotificationType Enum=Get FriendActionType!(self:Fort UIFriend Notification Object Reference=@n1) -> (FriendActionType:EFortUIFriendNotificationType Enum=Default)
   ) -> (
    Default:Exec
    FriendRequest:Exec
    PartyInviteSent:Exec
    PartyInvite:Exec
    InviteFriendToParty:Exec
    PartyRequest:Exec
    AutoImportFriendSuggestion:Exec
    PartyMemberCreated:Exec
    PartyMemberLeft:Exec
    PartyJoin:Exec
    EpicPartyInvite:Exec
    EpicPartyInviteSent:Exec
    EpicPartyMemberJoined:Exec
    EpicPartyMemberLeft:Exec
    EpicPartyJoined:Exec
    InviteFriendToEpicParty:Exec
    EpicPartyJoin:Exec
    GroupInvite:Exec
    EpicPartyJoinabilityChanged:Exec
    EpicPartyJoinRequestSent:Exec
    EpicPartyJoinRequestReceived:Exec
    EpicPartyJoinRequestAccepted:Exec
    EpicPartyPromotedToLeader:Exec
    PartyInvite1v1:Exec
    PartyInvite1v1Sent:Exec
    PartyMemberSittingOut:Exec
   ) {is=title K2Node_SwitchEnum}
   Return Node!(execute=@n2.AutoImportFriendSuggestion, Text:Text="LOCTEXT(\"Enable\")")
   Return Node(execute=[@n1.CastFailed, @n2.Default, @n2.PartyMemberCreated], Text="LOCTEXT(\"Open\")")
   Return Node(execute=[@n2.FriendRequest, @n2.PartyInvite, @n2.PartyInvite1v1, @n2.PartyRequest], Text="LOCTEXT(\"Accept\")")
   Return Node(execute=@n2.InviteFriendToParty, Text="LOCTEXT(\"Invite\")")
  ]}
 GetButtonHoldText {nodes [
   GetButtonHoldText() {is=title K2Node_FunctionEntry}
   > n1 = Cast To FortUIFriendNotification(Object=Get CachedNotification())
   > n2 = Switch on EFortUIFriendNotificationType(Selection=Get FriendActionType(self=@n1) -> (FriendActionType=Default))
   Return Node(execute=@n2.AutoImportFriendSuggestion, Text="LOCTEXT(\"Hold to Enable\")")
   Return Node(execute=[@n1.CastFailed, @n2.Default, @n2.PartyMemberCreated], Text="LOCTEXT(\"Hold to Open\")")
   Return Node(execute=[@n2.FriendRequest, @n2.PartyInvite, @n2.PartyInvite1v1, @n2.PartyRequest], Text="LOCTEXT(\"Hold to Accept\")")
   Return Node(execute=@n2.InviteFriendToParty, Text="LOCTEXT(\"Hold to Invite\")")
  ]}
 GetDisplayTime {nodes [
   GetDisplayTime() {is=title K2Node_FunctionEntry}
   n1 = Parent: GetDisplayTime!() -> (bOutDisplayTimeOverridden:bool, ReturnValue:float) {is=title K2Node_CallParentFunction}
   > n2 = Branch(Condition=@n1)
   > Return Node!(execute, bOutDisplayTimeOverridden:bool=true, ReturnValue:float=@n1.ReturnValue)
   n3 = Cast To FortUIFriendNotification(execute=@n2.else, Object=Get CachedNotification())
   > n4 = Switch on EFortUIFriendNotificationType(Selection=Get FriendActionType(self=@n3) -> (FriendActionType=Default))
   Return Node(
    execute=[
     @n3.CastFailed
     @n4.AutoImportFriendSuggestion
     @n4.Default
     @n4.FriendRequest
     @n4.PartyInvite
     @n4.PartyInvite1v1
     @n4.PartyInviteSent
     @n4.PartyMemberCreated
    ]
    bOutDisplayTimeOverridden=false
    ReturnValue=@n1.ReturnValue
   )
   Return Node(execute=@n4.InviteFriendToParty, bOutDisplayTimeOverridden=true, ReturnValue=10)
   Return Node(execute=@n4.PartyRequest, bOutDisplayTimeOverridden=true, ReturnValue=15)
  ]}
 GetInputSwitcherVisibility {nodes [
   GetInputSwitcherVisibility() {is=title K2Node_FunctionEntry}
   n1 = Get CachedNotification()
   > Return Node!(
    execute
    ReturnValue:ESlateVisibility Enum=Select(
     Option 0:ESlateVisibility Enum=Collapsed
     Option 1:ESlateVisibility Enum=Visible
     Index=Get bHasAction!(self:Fort UINotification Object Reference=@n1) -> (bHasAction:bool)
    ) -> (ReturnValue:ESlateVisibility Enum)
   )
  ]}
 SetImage {nodes [
   SetImage() {is=title K2Node_FunctionEntry}
   n1 = GetDisplayAsset!(self:Fort UINotification Object Reference=Get CachedNotification()) -> (ReturnValue:"Soft Object Reference (by ref)")
   > n2 = Branch(Condition=IsValidSoftObjectReference(SoftObjectReference=@n1))
   > SetBrushFromLazyDisplayAsset!(
    execute
    self:Common Lazy Image Object Reference=Get NotificationImage() -> (NotificationImage:Fort Lazy Image Object Reference)
    LazyObject:"Soft Object Reference (by ref)"=@n1
    bMatchTextureSize:bool=false
   ) -> (then)
   > n3 = Cast To TwitchNotification!(execute, Object=Get CachedNotification()) -> (then, CastFailed, AsTwitch Notification:Twitch Notification Object Reference) {is=title K2Node_DynamicCast}
   > SetStyle!(
    execute
    self:Common Border Object Reference=[
     Get B_NotificationImage() -> (B_NotificationImage:Common Border Object Reference)
     Get B_NotificationImageExtended() -> (B_NotificationImageExtended:Common Border Object Reference)
    ]
    InStyle:Common Border Style Class Reference=/FortUILibrary/Styles/Border/Border-SolidBG-Purple.Border-SolidBG-Purple_C
   ) -> (then)
   SetStyle(
    execute=[@n4.then, @n3.CastFailed]
    self=[Get B_NotificationImage(), Get B_NotificationImageExtended()]
    InStyle=/FortUILibrary/Styles/Border/Border-SolidBG-RewardBack.Border-SolidBG-RewardBack_C
   )
   n4 = SetBrushFromTexture!(
    execute=@n2.else
    self:Image Object Reference=Get NotificationImage()
    Texture:Texture 2D Object Reference=/Game/UI/Icons/Icon-InfoBox-M
    bMatchSize:bool=false
   ) -> (then)
  ]}
 SetVisibilityForOptionalElements {nodes [
   SetVisibilityForOptionalElements() {is=title K2Node_FunctionEntry}
   > SetVisibility(
    self=Get Switcher_InputButtons()
    InVisibility=GetInputSwitcherVisibility!() -> (ReturnValue:ESlateVisibility Enum=Visible) {is=title K2Node_CallFunction}
   )
  ]}
 SetupAnimBindings {nodes [
   n1 = SetupAnimBindings!() -> (then, Unbind:bool=~) {is=title K2Node_FunctionEntry}
   > n2 = Branch(Condition=@n1)
   > UnbindFromAnimationFinished!(
    execute
    self:Widget Animation Object Reference=Get Intro()
    Widget:User Widget Object Reference=Self-Reference()
    Delegate:Delegate=Create Event()
   ) -> (then)
   > UnbindFromAnimationFinished(self=Get Outro_New(), Widget=Self-Reference(), Delegate=Create Event())
   BindToAnimationFinished!(
    execute=@n2.else
    self:Widget Animation Object Reference=Get Intro()
    Widget:User Widget Object Reference=Self-Reference()
    Delegate=Create Event()
   ) -> (then)
   > BindToAnimationFinished(self=Get Outro_New(), Widget=Self-Reference(), Delegate=Create Event())
  ]}
 StartIntro {nodes [
   StartIntro() {is=title K2Node_FunctionEntry}
   > PlayAnimation(InAnimation=Get Intro())
   > Set bActionTaken(bActionTaken:bool=false) -> (Output_Get:bool)
   > PlaySound2D(Sound=/Game/Sounds/Fort_UI_Sounds/FrontEnd/Fort_UI_Party_Invited_Cue)
  ]}
 TakeAction {nodes [
   TakeAction() {is=title K2Node_FunctionEntry}
   > Branch(Condition=Get bHasAction(self=Get CachedNotification()))
   > n1 = Branch(Condition=Get bActionTaken() -> (bActionTaken:bool))
   Set bActionTaken(execute=@n1.else, bActionTaken=true)
   > TakeAction(self:Fort UINotification Object Reference=Get CachedNotification()) {is=title K2Node_CallFunction}
  ]}
 UpdateToast {nodes [
   UpdateToast!() -> (then, Notification:Fort UINotification Object Reference) {is=title K2Node_FunctionEntry}
   > SetText(
    self=Get TitleRichText() -> (TitleRichText:Common Rich Text Block Object Reference)
    InText=Get Title!(self:Fort UINotification Object Reference=Get Notification() -> (Notification:Fort UINotification Object Reference)) -> (Title:Text)
   )
   > SetText(
    self=Get DescriptionRichText() -> (DescriptionRichText:Common Rich Text Block Object Reference)
    InText=Get Description!(self:Fort UINotification Object Reference=Get Notification()) -> (Description:Text)
   )
   > SetActiveWidgetIndex(self=Get Switcher_InputButtons(), Index=GetActiveWidgetIndex() -> (ReturnValue=0) {is=title K2Node_CallFunction})
   > SetVisibilityForOptionalElements() {is=title K2Node_CallFunction}
   > SetImage() {is=title K2Node_CallFunction}
   > Set Text!(
    execute
    self:Icon Text Button Object Reference=Get Button_Cursor()
    ButtonText:Text=GetButtonCursorText!() -> (Text:Text) {is=title K2Node_CallFunction}
   ) -> (then) {is=title K2Node_CallFunction}
   > Set Text(self=Get Button_Hold(), ButtonText=GetButtonHoldText!() -> (Text:Text) {is=title K2Node_CallFunction})
  ]}
]
variables [
 AnimationDelayTimer {type=TimerHandle}
 bActionTaken {type=boolean}
]
widgets {
 info {
  inheritedWidgetCount=4
  namedSlotCount=1
  parentClass=/Script/FortniteUI.FortSocialNotificationWidget
  rootWidgetClass=&Overlay
  widgetCount=23
 }
 widgets [
  Overlay_1 {widgetClassPath=&Overlay, widget=Overlay_1}
  B_Dropshadow {
   widgetClassPath=&Border
   parent=Overlay_1
   widget=B_Dropshadow
   properties {
    visibility=SelfHitTestInvisible
    background {
     bIsDynamicallyLoaded=false
     drawAs=Image
     imageType=NoImage
     mirroring=NoMirror
     resourceName=None
     resourceObject=/FortUILibrary/Styles/Border/Materials/M_Border_DropShadow
     tiling=NoTile
     imageSize=<32 32>
     margin=<ltrb 0 0 0 0>
     outlineSettings {
      bUseBrushTransparency=false
      roundingType=HalfHeightRadius
      width=0
      color {colorUseRule=UseColor_Specified, specifiedColor=<rgba 0 0 0 0>}
      cornerRadii=<0 0 0 0>
     }
     tintColor {colorUseRule=UseColor_Specified, specifiedColor=<rgba 1 1 1 1>}
     uVRegion {bIsValid=false, max=<0 0>, min=<0 0>}
    }
    renderTransform {angle=0, scale=<1 1>, shear=<0 0>, translation=<5 8>}
    renderTransformPivot=<0 0.5>
   }
   slot {class=OverlaySlot, properties {horizontalAlignment=HAlign_Fill, verticalAlignment=VAlign_Fill}}
  }
  HorizontalBox_1 {
   widgetClassPath=&HorizontalBox
   parent=Overlay_1
   widget=HorizontalBox_1
   slot {class=OverlaySlot, properties {horizontalAlignment=HAlign_Fill, verticalAlignment=VAlign_Fill}}
  }
  Overlay_0 {widgetClassPath=&Overlay, parent=HorizontalBox_1, widget=Overlay_0, slot {class=HorizontalBoxSlot}}
  B_NotificationImageExtended {
   widgetClassPath=&CommonBorder
   bIsVariable=true
   parent=Overlay_0
   widget=B_NotificationImageExtended
   properties {
    horizontalAlignment=HAlign_Center
    verticalAlignment=VAlign_Center
    background {
     bIsDynamicallyLoaded=false
     drawAs=Box
     imageType=NoImage
     mirroring=NoMirror
     resourceName=None
     resourceObject=null
     tiling=NoTile
     imageSize=<240 76>
     margin=<ltrb 0.01 0.05 0.98 0.95>
     outlineSettings {
      bUseBrushTransparency=false
      roundingType=HalfHeightRadius
      width=0
      color {colorUseRule=UseColor_Specified, specifiedColor=<rgba 0 0 0 0>}
      cornerRadii=<0 0 0 0>
     }
     tintColor {colorUseRule=UseColor_Specified, specifiedColor=<rgba 0 0.4072 0.8879 1>}
     uVRegion {bIsValid=false, max=<0 0>, min=<0 0>}
    }
    renderTransformPivot=<0 0.5>
   }
   slot {class=OverlaySlot, properties {horizontalAlignment=HAlign_Fill, verticalAlignment=VAlign_Fill}}
  }
  B_NotificationImage {
   widgetClassPath=&CommonBorder
   bIsVariable=true
   parent=Overlay_0
   widget=B_NotificationImage
   properties {
    horizontalAlignment=HAlign_Center
    verticalAlignment=VAlign_Center
    background {
     bIsDynamicallyLoaded=false
     drawAs=Box
     imageType=NoImage
     mirroring=NoMirror
     resourceName=None
     resourceObject=null
     tiling=NoTile
     imageSize=<240 76>
     margin=<ltrb 0.01 0.05 0.98 0.95>
     outlineSettings {
      bUseBrushTransparency=false
      roundingType=HalfHeightRadius
      width=0
      color {colorUseRule=UseColor_Specified, specifiedColor=<rgba 0 0 0 0>}
      cornerRadii=<0 0 0 0>
     }
     tintColor {colorUseRule=UseColor_Specified, specifiedColor=<rgba 1 1 1 1>}
     uVRegion {bIsValid=false, max=<0 0>, min=<0 0>}
    }
    brushColor=<rgba 1 1 1 0>
    padding=<ltrb 0 0 0 0>
    renderTransform {angle=-360, scale=<1 1>, shear=<0 0>, translation=<0 0>}
   }
   slot {class=OverlaySlot, properties {horizontalAlignment=HAlign_Center, verticalAlignment=VAlign_Center, padding=<ltrb 10 0 10 0>}}
  }
  SizeBox_0 {
   widgetClassPath=&SizeBox
   parent=B_NotificationImage
   widget=SizeBox_0
   properties {
    bOverride_MinDesiredHeight=true
    bOverride_MinDesiredWidth=true
    heightOverride=40
    minDesiredHeight=40
    minDesiredWidth=40
    widthOverride=40
   }
   slot {class=BorderSlot, properties {horizontalAlignment=HAlign_Center, verticalAlignment=VAlign_Center, padding=<ltrb 0 0 0 0>}}
  }
  Overlay_2 {widgetClassPath=&Overlay, parent=SizeBox_0, widget=Overlay_2, slot {class=SizeBoxSlot}}
  NotificationImage {
   widgetClassPath=&FortLazyImage
   bIsVariable=true
   parent=Overlay_2
   widget=NotificationImage
   properties {brush {
     bIsDynamicallyLoaded=false
     drawAs=Image
     imageType=NoImage
     mirroring=NoMirror
     resourceName=None
     resourceObject=&Icon-InfoBox-M
     tiling=NoTile
     imageSize=<40 40>
     margin=<ltrb 0 0 0 0>
     outlineSettings {
      bUseBrushTransparency=false
      roundingType=HalfHeightRadius
      width=0
      color {colorUseRule=UseColor_Specified, specifiedColor=<rgba 0 0 0 0>}
      cornerRadii=<0 0 0 0>
     }
     tintColor {colorUseRule=UseColor_Specified, specifiedColor=<rgba 1 1 1 1>}
     uVRegion {bIsValid=false, max=<0 0>, min=<0 0>}
    }}
   slot {class=OverlaySlot, properties {padding=<ltrb 10 0 10 0>}}
  }
  SizeBox_4 {
   widgetClassPath=&SizeBox
   parent=HorizontalBox_1
   widget=SizeBox_4
   properties {
    bOverride_MaxDesiredWidth=true
    bOverride_MinDesiredWidth=true
    maxDesiredHeight=150
    maxDesiredWidth=750
    minDesiredHeight=150
    minDesiredWidth=400
    widthOverride=600
   }
   slot {class=HorizontalBoxSlot}
  }
  CommonBorder_MainWidget {
   widgetClassPath=&CommonBorder
   bIsVariable=true
   parent=SizeBox_4
   widget=CommonBorder_MainWidget
   properties {
    background {
     bIsDynamicallyLoaded=false
     drawAs=Box
     imageType=NoImage
     mirroring=NoMirror
     resourceName=None
     resourceObject=null
     tiling=NoTile
     imageSize=<52 36>
     margin=<ltrb 0.35 0 0.35 0>
     outlineSettings {
      bUseBrushTransparency=false
      roundingType=HalfHeightRadius
      width=0
      color {colorUseRule=UseColor_Specified, specifiedColor=<rgba 0 0 0 0>}
      cornerRadii=<0 0 0 0>
     }
     tintColor {colorUseRule=UseColor_Specified, specifiedColor=<rgba 1 1 1 0.95>}
     uVRegion {bIsValid=false, max=<0 0>, min=<0 0>}
    }
    brushColor=<rgba 1 1 1 0.9>
    padding=<ltrb 0 0 0 0>
    renderTransformPivot=<0 0.5>
   }
   slot {class=SizeBoxSlot}
  }
  HorizontalBox_0 {
   widgetClassPath=&HorizontalBox
   parent=CommonBorder_MainWidget
   widget=HorizontalBox_0
   slot {class=BorderSlot, properties {padding=<ltrb 0 0 0 0>}}
  }
  VerticalBox_0 {
   widgetClassPath=&VerticalBox
   parent=HorizontalBox_0
   widget=VerticalBox_0
   slot {class=HorizontalBoxSlot, properties {horizontalAlignment=HAlign_Left, padding=<ltrb 20 10 10 10>, size {sizeRule=Fill, value=1}}}
  }
  TitleRichText {
   widgetClassPath=&CommonRichTextBlock
   bIsVariable=true
   parent=VerticalBox_0
   widget=TitleRichText
   properties {
    bAutoCollapseWithEmptyText=true
    bOverrideDefaultStyle=true
    text=Friend Request Accepted
    textStyleSet=&/ToastTextStyleTable_LegacyTitle
    textTransformPolicy=ToUpper
    defaultTextStyleOverride {
     overflowPolicy=Clip
     transformPolicy=None
     colorAndOpacity {colorUseRule=UseColor_Specified, specifiedColor=<rgba 0.0085 0.1352 0.2448 1>}
     font {
      bForceMonospaced=false
      bMaterialIsStencil=false
      fontMaterial=null
      fontObject=&BurbankSmall
      letterSpacing=0
      monospacedWidth=1
      size=20
      skewAmount=0
      typefaceFontName=Black
      outlineSettings {
       bApplyOutlineToDropShadows=false
       bMiteredCorners=false
       bSeparateFillAlpha=false
       outlineMaterial=null
       outlineSize=0
       outlineColor=<rgba 0 0 0 1>
      }
     }
     highlightColor {colorUseRule=UseColor_Specified, specifiedColor=<rgba 0 0 0 1>}
     highlightShape {
      bIsDynamicallyLoaded=false
      drawAs=Image
      imageType=NoImage
      mirroring=NoMirror
      resourceName=None
      resourceObject=null
      tiling=NoTile
      imageSize=<32 32>
      margin=<ltrb 0 0 0 0>
      outlineSettings {
       bUseBrushTransparency=false
       roundingType=HalfHeightRadius
       width=0
       color {colorUseRule=UseColor_Specified, specifiedColor=<rgba 0 0 0 0>}
       cornerRadii=<0 0 0 0>
      }
      tintColor {colorUseRule=UseColor_Specified, specifiedColor=<rgba 1 1 1 1>}
      uVRegion {bIsValid=false, max=<0 0>, min=<0 0>}
     }
     selectedBackgroundColor {colorUseRule=UseColor_Foreground, specifiedColor=<rgba 1 0 1 1>}
     shadowColorAndOpacity=<rgba 0 0 0 1>
     shadowOffset=<0 0>
     strikeBrush {
      bIsDynamicallyLoaded=false
      drawAs=Image
      imageType=NoImage
      mirroring=NoMirror
      resourceName=None
      resourceObject=null
      tiling=NoTile
      imageSize=<32 32>
      margin=<ltrb 0 0 0 0>
      outlineSettings {
       bUseBrushTransparency=false
       roundingType=HalfHeightRadius
       width=0
       color {colorUseRule=UseColor_Specified, specifiedColor=<rgba 0 0 0 0>}
       cornerRadii=<0 0 0 0>
      }
      tintColor {colorUseRule=UseColor_Specified, specifiedColor=<rgba 1 1 1 1>}
      uVRegion {bIsValid=false, max=<0 0>, min=<0 0>}
     }
     underlineBrush {
      bIsDynamicallyLoaded=false
      drawAs=Image
      imageType=NoImage
      mirroring=NoMirror
      resourceName=None
      resourceObject=null
      tiling=NoTile
      imageSize=<32 32>
      margin=<ltrb 0 0 0 0>
      outlineSettings {
       bUseBrushTransparency=false
       roundingType=HalfHeightRadius
       width=0
       color {colorUseRule=UseColor_Specified, specifiedColor=<rgba 0 0 0 0>}
       cornerRadii=<0 0 0 0>
      }
      tintColor {colorUseRule=UseColor_Specified, specifiedColor=<rgba 1 1 1 1>}
      uVRegion {bIsValid=false, max=<0 0>, min=<0 0>}
     }
    }
   }
   slot {class=VerticalBoxSlot}
  }
  DescriptionRichText {
   widgetClassPath=&CommonRichTextBlock
   bIsVariable=true
   parent=VerticalBox_0
   widget=DescriptionRichText
   properties {
    bAutoCollapseWithEmptyText=true
    bOverrideDefaultStyle=true
    scrollStyle=&TextScrollStyle-Base_C
    text=Invite Sent to EnronTheGreat
    textStyleSet=&/ToastTextStyleTable_LegacyDescription
    textTransformPolicy=ToUpper
    defaultTextStyleOverride {
     overflowPolicy=Clip
     transformPolicy=None
     colorAndOpacity {colorUseRule=UseColor_Specified, specifiedColor=<rgba 0 0.4072 0.8879 1>}
     font {
      bForceMonospaced=false
      bMaterialIsStencil=false
      fontMaterial=null
      fontObject=&BurbankSmall
      letterSpacing=0
      monospacedWidth=1
      size=16
      skewAmount=0
      typefaceFontName=Bold
      outlineSettings {
       bApplyOutlineToDropShadows=false
       bMiteredCorners=false
       bSeparateFillAlpha=false
       outlineMaterial=null
       outlineSize=0
       outlineColor=<rgba 0 0 0 1>
      }
     }
     highlightColor {colorUseRule=UseColor_Specified, specifiedColor=<rgba 0 0 0 1>}
     highlightShape {
      bIsDynamicallyLoaded=false
      drawAs=Image
      imageType=NoImage
      mirroring=NoMirror
      resourceName=None
      resourceObject=null
      tiling=NoTile
      imageSize=<32 32>
      margin=<ltrb 0 0 0 0>
      outlineSettings {
       bUseBrushTransparency=false
       roundingType=HalfHeightRadius
       width=0
       color {colorUseRule=UseColor_Specified, specifiedColor=<rgba 0 0 0 0>}
       cornerRadii=<0 0 0 0>
      }
      tintColor {colorUseRule=UseColor_Specified, specifiedColor=<rgba 1 1 1 1>}
      uVRegion {bIsValid=false, max=<0 0>, min=<0 0>}
     }
     selectedBackgroundColor {colorUseRule=UseColor_Foreground, specifiedColor=<rgba 1 0 1 1>}
     shadowColorAndOpacity=<rgba 0 0 0 1>
     shadowOffset=<0 0>
     strikeBrush {
      bIsDynamicallyLoaded=false
      drawAs=Image
      imageType=NoImage
      mirroring=NoMirror
      resourceName=None
      resourceObject=null
      tiling=NoTile
      imageSize=<32 32>
      margin=<ltrb 0 0 0 0>
      outlineSettings {
       bUseBrushTransparency=false
       roundingType=HalfHeightRadius
       width=0
       color {colorUseRule=UseColor_Specified, specifiedColor=<rgba 0 0 0 0>}
       cornerRadii=<0 0 0 0>
      }
      tintColor {colorUseRule=UseColor_Specified, specifiedColor=<rgba 1 1 1 1>}
      uVRegion {bIsValid=false, max=<0 0>, min=<0 0>}
     }
     underlineBrush {
      bIsDynamicallyLoaded=false
      drawAs=Image
      imageType=NoImage
      mirroring=NoMirror
      resourceName=None
      resourceObject=null
      tiling=NoTile
      imageSize=<32 32>
      margin=<ltrb 0 0 0 0>
      outlineSettings {
       bUseBrushTransparency=false
       roundingType=HalfHeightRadius
       width=0
       color {colorUseRule=UseColor_Specified, specifiedColor=<rgba 0 0 0 0>}
       cornerRadii=<0 0 0 0>
      }
      tintColor {colorUseRule=UseColor_Specified, specifiedColor=<rgba 1 1 1 1>}
      uVRegion {bIsValid=false, max=<0 0>, min=<0 0>}
     }
    }
   }
   slot {class=VerticalBoxSlot}
  }
  Switcher_InputButtons {
   widgetClassPath=&CommonWidgetSwitcherLegacy
   bIsVariable=true
   parent=HorizontalBox_0
   widget=Switcher_InputButtons
   slot {class=HorizontalBoxSlot, properties {horizontalAlignment=HAlign_Right, padding=<ltrb 8 8 8 8>}}
  }
  Button_Hold {
   widgetClassPath=&IconTextButton
   bIsVariable=true
   parent=Switcher_InputButtons
   widget=Button_Hold
   properties {
    bIgnoreInputActionWidgetText=true
    bIsPersistentBinding=true
    buttonText=""
    controllerInputStyle=/FortUILibrary/Foundation/Buttons/ButtonStyles/Skew/ButtonStyle-Skew_LBlue.ButtonStyle-Skew_LBlue_C
    onButtonBaseClicked=(&self:WidgetTree.Button_Hold.HandleOnButtonClicked)
    onButtonBaseDoubleClicked=(&self:WidgetTree.Button_Hold.HandleOnButtonDoubleClicked)
    onButtonBaseHovered=(&self:WidgetTree.Button_Hold.HandleOnButtonHovered)
    onButtonBaseUnhovered=(&self:WidgetTree.Button_Hold.HandleOnButtonUnhovered)
    onSelectedChangedBase=(&self:WidgetTree.Button_Hold.HandleOnSelectedChanged)
    style=&ButtonStyle-Skew_LBlue_C
   }
   slot {class=WidgetSwitcherSlot}
  }
  HorizontalBox {widgetClassPath=&HorizontalBox, namedSlotHost=Button_Hold, widget=HorizontalBox}
  Action_HoldKeybindWidget {
   widgetClassPath=/FortUILibrary/Input/KeybindWidget
   bInherited=true
   bIsVariable=true
   parent=HorizontalBox
   widget=Action_HoldKeybindWidget
   properties {forcedHoldKeybindStatus=ForcedHold}
   slot {class=HorizontalBoxSlot}
  }
  Button_Cursor {
   widgetClassPath=&IconTextButton
   bIsVariable=true
   parent=Switcher_InputButtons
   widget=Button_Cursor
   properties {
    bIgnoreInputActionWidgetText=true
    bIsPersistentBinding=true
    buttonText=""
    contentAlignment=HAlign_Center
    controllerInputStyle=&ButtonStyle-Skew_LBlue_C
    inputPriority=3
    onButtonBaseClicked=(&self:WidgetTree.Button_Cursor.HandleOnButtonClicked)
    onButtonBaseDoubleClicked=(&self:WidgetTree.Button_Cursor.HandleOnButtonDoubleClicked)
    onButtonBaseHovered=(&self:WidgetTree.Button_Cursor.HandleOnButtonHovered)
    onButtonBaseUnhovered=(&self:WidgetTree.Button_Cursor.HandleOnButtonUnhovered)
    onSelectedChangedBase=(&self:WidgetTree.Button_Cursor.HandleOnSelectedChanged)
    style=&ButtonStyle-Skew_LBlue_C
    visibility=Visible
    triggeringInputAction {dataTable=&InputActionDataTable, rowName=ToastOpen}
   }
   slot {class=WidgetSwitcherSlot, properties {padding=<ltrb 8 8 8 8>}}
  }
  Avatar_MemberIcon {widgetClassPath=/Script/FortniteUI.FortSocialAvatarIcon, bInherited=true, widget=null}
  PlayerAvatar {widgetClassPath=/Script/FortniteUI.FortPlayerAvatar, bInherited=true, widget=null}
  SidebarSocialTagList {widgetClassPath=/Script/FortniteUI.SidebarSocialTagListBase, bInherited=true, widget=null}
 ]
}
