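# Widget Blueprint for a row of social tags in the sidebar, derived from the native SidebarSocialTagListBase class.
# A HorizontalBox holds three inherited SidebarSocialTagWidget slots as equal-fill, top-aligned columns, the first collapsed by default.
# All three start with an empty tag name in the Other category.
# The event graph holds only unwired events, so the Blueprint supplies layout and defaults and adds no logic.
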WidgetBlueprint /Game/UI/Sidebar/Tags/SidebarSocialTagList
graphs [EventGraph {nodes [
   Event PreConstruct()
   Event Construct()
   Event Tick()
  ]}]
variables []
widgets {
 info {inheritedWidgetCount=3, parentClass=/Script/FortniteUI.SidebarSocialTagListBase, rootWidgetClass=&HorizontalBox, widgetCount=4}
 widgets [
  HorizontalBox_1 {widgetClassPath=&HorizontalBox, widget=HorizontalBox_1}
  SidebarSocialTagWidget_3 {
   widgetClassPath=&/SidebarSocialTagWidget
   bInherited=true
   bIsVariable=true
   parent=HorizontalBox_1
   widget=SidebarSocialTagWidget_3
   properties {bShowCategoryCustomization=false, tagCategory=Other, tagName="", visibility=Collapsed}
   slot {class=HorizontalBoxSlot, properties {verticalAlignment=VAlign_Top, padding=<ltrb 0 0 10 0>, size {sizeRule=Fill, value=1}}}
  }
  SidebarSocialTagWidget_1 {
   widgetClassPath=&SidebarSocialTagWidget
   bInherited=true
   bIsVariable=true
   parent=HorizontalBox_1
   widget=SidebarSocialTagWidget_1
   properties {tagCategory=Other, tagName="", visibility=Visible}
   slot {class=HorizontalBoxSlot, properties {verticalAlignment=VAlign_Top, padding=<ltrb 0 0 10 0>, size {sizeRule=Fill, value=1}}}
  }
  SidebarSocialTagWidget_2 {
   widgetClassPath=&SidebarSocialTagWidget
   bInherited=true
   bIsVariable=true
   parent=HorizontalBox_1
   widget=SidebarSocialTagWidget_2
   properties {bShowCategoryCustomization=false, tagCategory=Other, tagName="", visibility=Visible}
   slot {class=HorizontalBoxSlot, properties {verticalAlignment=VAlign_Top, padding=<ltrb 0 0 10 0>, size {sizeRule=Fill, value=1}}}
  }
 ]
}
