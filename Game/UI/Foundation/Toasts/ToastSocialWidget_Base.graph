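# Base social toast widget that slides in horizontally to show either a hold-to-act prompt or a positive, negative or neutral result.
# Intro animations drive background and trim material parameters for burst, wipe and green, red or blue tint, alongside success or fail icons.
# "Simple" variants keep the slide but hold most other values static.
# Anim_HoldProgress fills a progress circle, fades the prompt and turns the toast into a green success state.
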WidgetBlueprint /Game/UI/Foundation/Toasts/ToastSocialWidget_Base
animations [
 Anim_IntroAction {
  binding_count=4
  length_seconds=1.25
  object=&self:Anim_IntroAction
  sequence {bindings [
    Root {
     id=#1
     possessed_class=Overlay
     tracks [{
       type=MovieScene2DTransformTrack
       sections [{
         type=MovieScene2DTransformSection
         end_frame=20
         start_frame=0
         channels [Translation.X {type=MovieSceneScriptingFloatChannel, keys [[0.1667,-1156.3329],[7.4167,0]]}]
        }]
      }]
    }
    VerticalBox_HoldAction {
     id=#2
     possessed_class=VerticalBox
     tracks [{
       type=MovieSceneFloatTrack
       display_name=Render Opacity
       sections [{
         type=MovieSceneFloatSection
         end_frame=25
         start_frame=0
         channels [None {type=MovieSceneScriptingFloatChannel, keys [[0.1667,0],[5.0833,0],[10.0837,1]]}]
        }]
      }]
    }
    Image_BackgroundMat {
     id=#3
     possessed_class=Image
     tracks [{
       type=MovieSceneWidgetMaterialTrack
       sections [{
         type=MovieSceneParameterSection
         end_frame=22
         start_frame=0
         channels [
          RightOffset {type=MovieSceneScriptingFloatChannel, keys [[0.1667,0.75],[3.8333,0.75],[10.0837,0]]}
          Burst {type=MovieSceneScriptingFloatChannel, keys [[0,1],[3.75,1],[11.25,0]]}
         ]
        }]
      }]
    }
    Image_Trim {
     id=#4
     possessed_class=Image
     tracks [
      {
       type=MovieSceneWidgetMaterialTrack
       sections [{
         type=MovieSceneParameterSection
         end_frame=25
         start_frame=0
         channels [
          RightOffset {type=MovieSceneScriptingFloatChannel, keys [[0.1667,0.75],[3.8333,0.75],[10.0837,0]]}
          Burst {type=MovieSceneScriptingFloatChannel, keys [[0,1],[3.75,1],[11.25,0]]}
         ]
        }]
      }
      {
       type=MovieScene2DTransformTrack
       sections [{
         type=MovieScene2DTransformSection
         end_frame=0
         start_frame=0
         channels [
          Translation.X {type=MovieSceneScriptingFloatChannel, const [0,1]}
          Translation.Y {type=MovieSceneScriptingFloatChannel, const [0,1]}
          Angle {type=MovieSceneScriptingFloatChannel}
          Scale.X {type=MovieSceneScriptingFloatChannel}
          Scale.Y {type=MovieSceneScriptingFloatChannel}
          Shear.X {type=MovieSceneScriptingFloatChannel}
          Shear.Y {type=MovieSceneScriptingFloatChannel}
         ]
        }]
      }
     ]
    }
   ]}
 }
 Anim_IntroNegativeInfo {
  binding_count=9
  length_seconds=1.75
  object=&self:Anim_IntroNegativeInfo
  sequence {
   bindings [
    Root {
     id=#1
     possessed_class=Overlay
     tracks [{
       type=MovieScene2DTransformTrack
       sections [{
         type=MovieScene2DTransformSection
         end_frame=420
         start_frame=0
         channels [Translation.X {type=MovieSceneScriptingFloatChannel, keys [[0,-1156.3329],[180,0]]}]
        }]
      }]
    }
    Image_Fail {
     id=#5
     possessed_class=Image
     tracks [{
       type=MovieSceneByteTrack
       display_name=Visibility
       sections [{type=MovieSceneByteSection, end_frame=24, start_frame=0, channels [None {type=MovieSceneScriptingByteChannel, const [4,1]}]}]
      }]
    }
    Overlay_ResultIcon {
     id=#6
     possessed_class=Overlay
     tracks [
      {
       type=MovieSceneFloatTrack
       display_name=Render Opacity
       sections [{
         type=MovieSceneFloatSection
         end_frame=420
         start_frame=0
         channels [None {type=MovieSceneScriptingFloatChannel, keys [[0,0],[240,0],[420.004,1]]}]
        }]
      }
      {
       type=MovieScene2DTransformTrack
       sections [{
         type=MovieScene2DTransformSection
         end_frame=360
         start_frame=0
         channels [
          Scale.X {type=MovieSceneScriptingFloatChannel, keys [[0.004,2],[240,2],[324,1]]}
          Scale.Y {type=MovieSceneScriptingFloatChannel, keys [[0,2],[240,2],[324,1]]}
         ]
        }]
      }
     ]
    }
    VerticalBox_HoldAction {
     id=#2
     possessed_class=VerticalBox
     tracks [{
       type=MovieSceneFloatTrack
       display_name=Render Opacity
       sections [{type=MovieSceneFloatSection, end_frame=248, start_frame=0, channels [None {type=MovieSceneScriptingFloatChannel, const [0,1]}]}]
      }]
    }
    Image_Trim {
     id=#7
     possessed_class=Image
     tracks [
      {
       type=MovieSceneWidgetMaterialTrack
       sections [{
         type=MovieSceneParameterSection
         end_frame=372
         start_frame=0
         channels [
          RightOffset {type=MovieSceneScriptingFloatChannel, keys [[0,0.75],[90,0.75],[240.004,0]]}
          LeftOffset {type=MovieSceneScriptingFloatChannel, keys [[0,0.985],[240.004,0.985],[300.004,0.15]]}
          LerpAngleChange {type=MovieSceneScriptingFloatChannel, keys [[0,0],[240.004,0],[300.004,1]]}
          LerpToGreen {type=MovieSceneScriptingFloatChannel, const [0,1]}
          LerpToRed {type=MovieSceneScriptingFloatChannel, const [1,1]}
          Burst {type=MovieSceneScriptingFloatChannel, keys [[0,1],[180,1],[270,0]]}
          HideProgressBar {type=MovieSceneScriptingFloatChannel, const [1,1]}
          Trim_ColorR {type=MovieSceneScriptingFloatChannel, keys [[0,0.956],[120,0.956],[180,1],[300,0.956]]}
          Trim_ColorG {type=MovieSceneScriptingFloatChannel, keys [[0,0.0513],[120,0.0513],[180,1],[300,0.0513]]}
          Trim_ColorB {type=MovieSceneScriptingFloatChannel, keys [[0,0.1195],[120,0.1195],[180,1],[300,0.1195]]}
          Trim_ColorA {type=MovieSceneScriptingFloatChannel, const [1,4]}
         ]
        }]
      }
      {
       type=MovieScene2DTransformTrack
       sections [{
         type=MovieScene2DTransformSection
         end_frame=0
         start_frame=0
         channels [
          Translation.X {type=MovieSceneScriptingFloatChannel, const [0,1]}
          Translation.Y {type=MovieSceneScriptingFloatChannel, const [0,1]}
          Angle {type=MovieSceneScriptingFloatChannel}
          Scale.X {type=MovieSceneScriptingFloatChannel}
          Scale.Y {type=MovieSceneScriptingFloatChannel}
          Shear.X {type=MovieSceneScriptingFloatChannel}
          Shear.Y {type=MovieSceneScriptingFloatChannel}
         ]
        }]
      }
     ]
    }
    Image_Success {
     id=#8
     possessed_class=Image
     tracks [{
       type=MovieSceneByteTrack
       display_name=Visibility
       sections [{type=MovieSceneByteSection, end_frame=4, start_frame=0, channels [None {type=MovieSceneScriptingByteChannel, const [1,1]}]}]
      }]
    }
    Image_BackgroundMat {
     id=#3
     possessed_class=Image
     tracks [{
       type=MovieSceneWidgetMaterialTrack
       sections [{
         type=MovieSceneParameterSection
         end_frame=360
         start_frame=0
         channels [
          RightOffset {type=MovieSceneScriptingFloatChannel, keys [[0,0.75],[90,0.75],[240.004,0]]}
          LerpToGreen {type=MovieSceneScriptingFloatChannel, const [0,1]}
          LerpToRed {type=MovieSceneScriptingFloatChannel, const [1,1]}
          LeftOffset {type=MovieSceneScriptingFloatChannel, keys [[0,0.915],[240,0.915],[300,0.175]]}
          HideProgressBar {type=MovieSceneScriptingFloatChannel, const [1,1]}
          LerpAngleChange {type=MovieSceneScriptingFloatChannel, keys [[0,0],[240,0],[300,1]]}
          Burst {type=MovieSceneScriptingFloatChannel, keys [[0,1],[180,1],[270,0]]}
          Trim_ColorR {type=MovieSceneScriptingFloatChannel, keys [[0,0.956],[120,0.956],[180,1],[300,0.956]]}
          Trim_ColorG {type=MovieSceneScriptingFloatChannel, keys [[0,0.0513],[120,0.0513],[180,1],[300,0.0513]]}
          Trim_ColorB {type=MovieSceneScriptingFloatChannel, keys [[0,0.1195],[120,0.1195],[180,1],[300,0.1195]]}
          Trim_ColorA {type=MovieSceneScriptingFloatChannel, const [1,4]}
         ]
        }]
      }]
    }
    Image_ToastCushion {
     id=#9
     possessed_class=Image
     tracks [{
       type=MovieSceneWidgetMaterialTrack
       sections [{
         type=MovieSceneParameterSection
         end_frame=420
         start_frame=0
         channels [HideProgressBar {type=MovieSceneScriptingFloatChannel, const [1,1]}]
        }]
      }]
    }
   ]
   display_rate {numerator=240}
   tracks [{
     type=MovieSceneEventTrack
     sections [{
       type=MovieSceneEventTriggerSection
       channels [None {type=MovieSceneScriptingEventChannel, const ["<Struct 'MovieSceneEvent' (0x4687a1a7da0) {}>", 1]}]
      }]
    }]
  }
 }
 Anim_IntroNegativeInfoSimple {
  binding_count=9
  length_seconds=1.75
  object=&self:Anim_IntroNegativeInfoSimple
  sequence {
   bindings [
    Root {
     id=#1
     possessed_class=Overlay
     tracks [{
       type=MovieScene2DTransformTrack
       sections [{
         type=MovieScene2DTransformSection
         end_frame=90
         start_frame=0
         channels [Translation.X {type=MovieSceneScriptingFloatChannel, keys [[0,-1156.3329],[90,0]]}]
        }]
      }]
    }
    Image_ToastCushion {
     id=#10
     possessed_class=Image
     tracks [{
       type=MovieSceneWidgetMaterialTrack
       sections [{
         type=MovieSceneParameterSection
         end_frame=210
         start_frame=0
         channels [HideProgressBar {type=MovieSceneScriptingFloatChannel, const [1,1]}]
        }]
      }]
    }
    Image_Fail {
     id=#5
     possessed_class=Image
     tracks [{
       type=MovieSceneByteTrack
       display_name=Visibility
       sections [{type=MovieSceneByteSection, end_frame=12, start_frame=0, channels [None {type=MovieSceneScriptingByteChannel, const [4,1]}]}]
      }]
    }
    Overlay_ResultIcon {
     id=#6
     possessed_class=Overlay
     tracks [
      {
       type=MovieSceneFloatTrack
       display_name=Render Opacity
       sections [{type=MovieSceneFloatSection, end_frame=90, start_frame=0, channels [None {type=MovieSceneScriptingFloatChannel, const [1,1]}]}]
      }
      {
       type=MovieScene2DTransformTrack
       sections [{
         type=MovieScene2DTransformSection
         end_frame=90
         start_frame=0
         channels [
          Scale.X {type=MovieSceneScriptingFloatChannel, const [1,1]}
          Scale.Y {type=MovieSceneScriptingFloatChannel, const [1,1]}
         ]
        }]
      }
     ]
    }
    VerticalBox_HoldAction {
     id=#2
     possessed_class=VerticalBox
     tracks [{
       type=MovieSceneFloatTrack
       display_name=Render Opacity
       sections [{type=MovieSceneFloatSection, end_frame=90, start_frame=0, channels [None {type=MovieSceneScriptingFloatChannel, const [0,1]}]}]
      }]
    }
    Image_Trim {
     id=#7
     possessed_class=Image
     tracks [{
       type=MovieSceneWidgetMaterialTrack
       sections [{
         type=MovieSceneParameterSection
         end_frame=90
         start_frame=0
         channels [
          RightOffset {type=MovieSceneScriptingFloatChannel, const [0,1]}
          LeftOffset {type=MovieSceneScriptingFloatChannel, const [0.15,1]}
          LerpAngleChange {type=MovieSceneScriptingFloatChannel, const [1,1]}
          LerpToGreen {type=MovieSceneScriptingFloatChannel, const [0,1]}
          LerpToRed {type=MovieSceneScriptingFloatChannel, const [1,1]}
          Burst {type=MovieSceneScriptingFloatChannel, const [0,1]}
          HideProgressBar {type=MovieSceneScriptingFloatChannel, const [1,1]}
          Trim_ColorR {type=MovieSceneScriptingFloatChannel, const [0.956,1]}
          Trim_ColorG {type=MovieSceneScriptingFloatChannel, const [0.0513,1]}
          Trim_ColorB {type=MovieSceneScriptingFloatChannel, const [0.1195,1]}
          Trim_ColorA {type=MovieSceneScriptingFloatChannel, const [1,1]}
         ]
        }]
      }]
    }
    Image_Success {
     id=#8
     possessed_class=Image
     tracks [{
       type=MovieSceneByteTrack
       display_name=Visibility
       sections [{type=MovieSceneByteSection, end_frame=2, start_frame=0, channels [None {type=MovieSceneScriptingByteChannel, const [1,1]}]}]
      }]
    }
    Image_BackgroundMat {
     id=#3
     possessed_class=Image
     tracks [{
       type=MovieSceneWidgetMaterialTrack
       sections [{
         type=MovieSceneParameterSection
         end_frame=90
         start_frame=0
         channels [
          RightOffset {type=MovieSceneScriptingFloatChannel, const [0,1]}
          LerpToGreen {type=MovieSceneScriptingFloatChannel, const [0,1]}
          LerpToRed {type=MovieSceneScriptingFloatChannel, const [1,1]}
          LeftOffset {type=MovieSceneScriptingFloatChannel, const [0.175,1]}
          HideProgressBar {type=MovieSceneScriptingFloatChannel, const [1,1]}
          LerpAngleChange {type=MovieSceneScriptingFloatChannel, const [1,1]}
          Burst {type=MovieSceneScriptingFloatChannel, const [0,1]}
          Trim_ColorR {type=MovieSceneScriptingFloatChannel, const [0.956,1]}
          Trim_ColorG {type=MovieSceneScriptingFloatChannel, const [0.0513,1]}
          Trim_ColorB {type=MovieSceneScriptingFloatChannel, const [0.1195,1]}
          Trim_ColorA {type=MovieSceneScriptingFloatChannel, const [1,1]}
         ]
        }]
      }]
    }
   ]
   display_rate {numerator=120}
   tracks [{
     type=MovieSceneEventTrack
     sections [{
       type=MovieSceneEventTriggerSection
       channels [None {type=MovieSceneScriptingEventChannel, const ["<Struct 'MovieSceneEvent' (0x4687a1a9420) {}>", 1]}]
      }]
    }]
  }
 }
 Anim_IntroPositiveInfo {
  binding_count=9
  length_seconds=1.75
  object=&self:Anim_IntroPositiveInfo
  sequence {
   bindings [
    Root {
     id=#1
     possessed_class=Overlay
     tracks [{
       type=MovieScene2DTransformTrack
       sections [{
         type=MovieScene2DTransformSection
         end_frame=210
         start_frame=0
         channels [Translation.X {type=MovieSceneScriptingFloatChannel, keys [[0,-1156.3329],[90,0]]}]
        }]
      }]
    }
    Image_Fail {
     id=#5
     possessed_class=Image
     tracks [{
       type=MovieSceneByteTrack
       display_name=Visibility
       sections [{type=MovieSceneByteSection, end_frame=12, start_frame=0, channels [None {type=MovieSceneScriptingByteChannel, const [1,1]}]}]
      }]
    }
    Image_ToastCushion {
     id=#11
     possessed_class=Image
     tracks [{
       type=MovieSceneWidgetMaterialTrack
       sections [{
         type=MovieSceneParameterSection
         end_frame=210
         start_frame=0
         channels [HideProgressBar {type=MovieSceneScriptingFloatChannel, const [1,1]}]
        }]
      }]
    }
    Overlay_ResultIcon {
     id=#6
     possessed_class=Overlay
     tracks [
      {
       type=MovieSceneFloatTrack
       display_name=Render Opacity
       sections [{
         type=MovieSceneFloatSection
         end_frame=210
         start_frame=0
         channels [None {type=MovieSceneScriptingFloatChannel, keys [[0,0],[120,0],[210.002,1]]}]
        }]
      }
      {
       type=MovieScene2DTransformTrack
       sections [{
         type=MovieScene2DTransformSection
         end_frame=180
         start_frame=0
         channels [
          Scale.X {type=MovieSceneScriptingFloatChannel, keys [[0.002,2],[120,2],[162,1]]}
          Scale.Y {type=MovieSceneScriptingFloatChannel, keys [[0,2],[120,2],[162,1]]}
         ]
        }]
      }
     ]
    }
    VerticalBox_HoldAction {
     id=#2
     possessed_class=VerticalBox
     tracks [{
       type=MovieSceneFloatTrack
       display_name=Render Opacity
       sections [{type=MovieSceneFloatSection, end_frame=124, start_frame=0, channels [None {type=MovieSceneScriptingFloatChannel, const [0,1]}]}]
      }]
    }
    Image_Success {
     id=#8
     possessed_class=Image
     tracks [{
       type=MovieSceneByteTrack
       display_name=Visibility
       sections [{type=MovieSceneByteSection, end_frame=2, start_frame=0, channels [None {type=MovieSceneScriptingByteChannel, const [4,1]}]}]
      }]
    }
    Image_BackgroundMat {
     id=#3
     possessed_class=Image
     tracks [{
       type=MovieSceneWidgetMaterialTrack
       sections [{
         type=MovieSceneParameterSection
         end_frame=180
         start_frame=0
         channels [
          RightOffset {type=MovieSceneScriptingFloatChannel, keys [[0,0.75],[45,0.75],[120.002,0]]}
          LerpToGreen {type=MovieSceneScriptingFloatChannel, const [1,1]}
          LerpToRed {type=MovieSceneScriptingFloatChannel, const [0,1]}
          LeftOffset {type=MovieSceneScriptingFloatChannel, keys [[0,0.915],[120,0.915],[150,0.175]]}
          HideProgressBar {type=MovieSceneScriptingFloatChannel, const [1,1]}
          LerpAngleChange {type=MovieSceneScriptingFloatChannel, keys [[0,0],[120,0],[150,1]]}
          Burst {type=MovieSceneScriptingFloatChannel, keys [[0,1],[90,1],[135,0]]}
          Trim_ColorR {type=MovieSceneScriptingFloatChannel, keys [[0,0],[60,0],[90,1],[150,0]]}
          Trim_ColorG {type=MovieSceneScriptingFloatChannel, keys [[0,0.6867],[60,0.6867],[90,1],[150,0.6867]]}
          Trim_ColorB {type=MovieSceneScriptingFloatChannel, keys [[0,0.0887],[60,0.0887],[90,1],[150,0.0887]]}
          Trim_ColorA {type=MovieSceneScriptingFloatChannel, const [1,4]}
         ]
        }]
      }]
    }
    Image_Trim {
     id=#12
     possessed_class=Image
     tracks [
      {
       type=MovieSceneWidgetMaterialTrack
       sections [{
         type=MovieSceneParameterSection
         end_frame=186
         start_frame=0
         channels [
          RightOffset {type=MovieSceneScriptingFloatChannel, keys [[0,0.75],[45,0.75],[120.002,0]]}
          LeftOffset {type=MovieSceneScriptingFloatChannel, keys [[0,0.985],[120.002,0.985],[150.002,0.15]]}
          LerpAngleChange {type=MovieSceneScriptingFloatChannel, keys [[0,0],[120.002,0],[150.002,1]]}
          LerpToGreen {type=MovieSceneScriptingFloatChannel, const [1,1]}
          LerpToRed {type=MovieSceneScriptingFloatChannel, const [0,1]}
          Burst {type=MovieSceneScriptingFloatChannel, keys [[0,1],[90,1],[135,0]]}
          HideProgressBar {type=MovieSceneScriptingFloatChannel, const [1,1]}
          Trim_ColorR {type=MovieSceneScriptingFloatChannel, keys [[0,0],[60,0],[90,1],[150,0]]}
          Trim_ColorG {type=MovieSceneScriptingFloatChannel, keys [[0,0.6867],[60,0.6867],[90,1],[150,0.6867]]}
          Trim_ColorB {type=MovieSceneScriptingFloatChannel, keys [[0,0.0887],[60,0.0887],[90,1],[150,0.0887]]}
          Trim_ColorA {type=MovieSceneScriptingFloatChannel, const [1,4]}
         ]
        }]
      }
      {
       type=MovieScene2DTransformTrack
       sections [{
         type=MovieScene2DTransformSection
         end_frame=0
         start_frame=0
         channels [
          Translation.X {type=MovieSceneScriptingFloatChannel, const [0,1]}
          Translation.Y {type=MovieSceneScriptingFloatChannel, const [0,1]}
          Angle {type=MovieSceneScriptingFloatChannel}
          Scale.X {type=MovieSceneScriptingFloatChannel}
          Scale.Y {type=MovieSceneScriptingFloatChannel}
          Shear.X {type=MovieSceneScriptingFloatChannel}
          Shear.Y {type=MovieSceneScriptingFloatChannel}
         ]
        }]
      }
     ]
    }
   ]
   display_rate {numerator=120}
   tracks [{
     type=MovieSceneEventTrack
     sections [{
       type=MovieSceneEventTriggerSection
       channels [None {type=MovieSceneScriptingEventChannel, const ["<Struct 'MovieSceneEvent' (0x4687a1aa3c0) {}>", 1]}]
      }]
    }]
  }
 }
 Anim_IntroPositiveInfoSimple {
  binding_count=9
  length_seconds=1.75
  object=&self:Anim_IntroPositiveInfoSimple
  sequence {
   bindings [
    Root {
     id=#1
     possessed_class=Overlay
     tracks [{
       type=MovieScene2DTransformTrack
       sections [{
         type=MovieScene2DTransformSection
         end_frame=90
         start_frame=0
         channels [Translation.X {type=MovieSceneScriptingFloatChannel, keys [[0,-1156.3329],[90,0]]}]
        }]
      }]
    }
    Image_Fail {
     id=#5
     possessed_class=Image
     tracks [{
       type=MovieSceneByteTrack
       display_name=Visibility
       sections [{type=MovieSceneByteSection, end_frame=12, start_frame=0, channels [None {type=MovieSceneScriptingByteChannel, const [1,1]}]}]
      }]
    }
    Overlay_ResultIcon {
     id=#6
     possessed_class=Overlay
     tracks [
      {
       type=MovieSceneFloatTrack
       display_name=Render Opacity
       sections [{type=MovieSceneFloatSection, end_frame=90, start_frame=0, channels [None {type=MovieSceneScriptingFloatChannel, const [1,1]}]}]
      }
      {
       type=MovieScene2DTransformTrack
       sections [{
         type=MovieScene2DTransformSection
         end_frame=90
         start_frame=0
         channels [
          Scale.X {type=MovieSceneScriptingFloatChannel, const [1,1]}
          Scale.Y {type=MovieSceneScriptingFloatChannel, const [1,1]}
         ]
        }]
      }
     ]
    }
    VerticalBox_HoldAction {
     id=#2
     possessed_class=VerticalBox
     tracks [{
       type=MovieSceneFloatTrack
       display_name=Render Opacity
       sections [{type=MovieSceneFloatSection, end_frame=90, start_frame=0, channels [None {type=MovieSceneScriptingFloatChannel, const [0,1]}]}]
      }]
    }
    Image_Success {
     id=#8
     possessed_class=Image
     tracks [{
       type=MovieSceneByteTrack
       display_name=Visibility
       sections [{type=MovieSceneByteSection, end_frame=2, start_frame=0, channels [None {type=MovieSceneScriptingByteChannel, const [4,1]}]}]
      }]
    }
    Image_BackgroundMat {
     id=#3
     possessed_class=Image
     tracks [{
       type=MovieSceneWidgetMaterialTrack
       sections [{
         type=MovieSceneParameterSection
         end_frame=90
         start_frame=0
         channels [
          RightOffset {type=MovieSceneScriptingFloatChannel, const [0,1]}
          LerpToGreen {type=MovieSceneScriptingFloatChannel, const [1,1]}
          LerpToRed {type=MovieSceneScriptingFloatChannel, const [0,1]}
          LeftOffset {type=MovieSceneScriptingFloatChannel, const [0.175,1]}
          HideProgressBar {type=MovieSceneScriptingFloatChannel, const [1,1]}
          LerpAngleChange {type=MovieSceneScriptingFloatChannel, const [1,1]}
          Burst {type=MovieSceneScriptingFloatChannel, const [0,1]}
          Trim_ColorR {type=MovieSceneScriptingFloatChannel, const [0,1]}
          Trim_ColorG {type=MovieSceneScriptingFloatChannel, const [0.6867,1]}
          Trim_ColorB {type=MovieSceneScriptingFloatChannel, const [0.0887,1]}
          Trim_ColorA {type=MovieSceneScriptingFloatChannel, const [1,1]}
         ]
        }]
      }]
    }
    Image_Trim {
     id=#12
     possessed_class=Image
     tracks [{
       type=MovieSceneWidgetMaterialTrack
       sections [{
         type=MovieSceneParameterSection
         end_frame=90
         start_frame=0
         channels [
          RightOffset {type=MovieSceneScriptingFloatChannel, const [0,1]}
          LeftOffset {type=MovieSceneScriptingFloatChannel, const [0.15,1]}
          LerpAngleChange {type=MovieSceneScriptingFloatChannel, const [1,1]}
          LerpToGreen {type=MovieSceneScriptingFloatChannel, const [1,1]}
          LerpToRed {type=MovieSceneScriptingFloatChannel, const [0,1]}
          Burst {type=MovieSceneScriptingFloatChannel, const [0,1]}
          HideProgressBar {type=MovieSceneScriptingFloatChannel, const [1,1]}
          Trim_ColorR {type=MovieSceneScriptingFloatChannel, const [0,1]}
          Trim_ColorG {type=MovieSceneScriptingFloatChannel, const [0.6867,1]}
          Trim_ColorB {type=MovieSceneScriptingFloatChannel, const [0.0887,1]}
          Trim_ColorA {type=MovieSceneScriptingFloatChannel, const [1,1]}
         ]
        }]
      }]
    }
    Image_ToastCushion {
     id=#13
     possessed_class=Image
     tracks [{
       type=MovieSceneWidgetMaterialTrack
       sections [{
         type=MovieSceneParameterSection
         end_frame=210
         start_frame=0
         channels [HideProgressBar {type=MovieSceneScriptingFloatChannel, const [1,1]}]
        }]
      }]
    }
   ]
   display_rate {numerator=120}
   tracks [{
     type=MovieSceneEventTrack
     sections [{
       type=MovieSceneEventTriggerSection
       channels [None {type=MovieSceneScriptingEventChannel, const ["<Struct 'MovieSceneEvent' (0x4687a1aafa0) {}>", 1]}]
      }]
    }]
  }
 }
 Anim_IntroNeutralInfo {
  binding_count=9
  length_seconds=1.75
  object=&self:Anim_IntroNeutralInfo
  sequence {
   bindings [
    Root {
     id=#1
     possessed_class=Overlay
     tracks [{
       type=MovieScene2DTransformTrack
       sections [{
         type=MovieScene2DTransformSection
         end_frame=210
         start_frame=0
         channels [Translation.X {type=MovieSceneScriptingFloatChannel, keys [[0,-1156.3329],[90,0]]}]
        }]
      }]
    }
    Image_Fail {
     id=#5
     possessed_class=Image
     tracks [{
       type=MovieSceneByteTrack
       display_name=Visibility
       sections [{type=MovieSceneByteSection, end_frame=12, start_frame=0, channels [None {type=MovieSceneScriptingByteChannel, const [1,1]}]}]
      }]
    }
    Image_ToastCushion {
     id=#14
     possessed_class=Image
     tracks [{
       type=MovieSceneWidgetMaterialTrack
       sections [{
         type=MovieSceneParameterSection
         end_frame=210
         start_frame=0
         channels [HideProgressBar {type=MovieSceneScriptingFloatChannel, const [1,1]}]
        }]
      }]
    }
    Overlay_ResultIcon {
     id=#6
     possessed_class=Overlay
     tracks [
      {
       type=MovieSceneFloatTrack
       display_name=Render Opacity
       sections [{
         type=MovieSceneFloatSection
         end_frame=210
         start_frame=0
         channels [None {type=MovieSceneScriptingFloatChannel, keys [[0,0],[120,0],[210.002,1]]}]
        }]
      }
      {
       type=MovieScene2DTransformTrack
       sections [{
         type=MovieScene2DTransformSection
         end_frame=180
         start_frame=0
         channels [
          Scale.X {type=MovieSceneScriptingFloatChannel, keys [[0.002,2],[120,2],[162,1]]}
          Scale.Y {type=MovieSceneScriptingFloatChannel, keys [[0,2],[120,2],[162,1]]}
         ]
        }]
      }
     ]
    }
    VerticalBox_HoldAction {
     id=#2
     possessed_class=VerticalBox
     tracks [{
       type=MovieSceneFloatTrack
       display_name=Render Opacity
       sections [{type=MovieSceneFloatSection, end_frame=124, start_frame=0, channels [None {type=MovieSceneScriptingFloatChannel, const [0,1]}]}]
      }]
    }
    Image_Success {
     id=#8
     possessed_class=Image
     tracks [{
       type=MovieSceneByteTrack
       display_name=Visibility
       sections [{type=MovieSceneByteSection, end_frame=2, start_frame=0, channels [None {type=MovieSceneScriptingByteChannel, const [1,1]}]}]
      }]
    }
    Image_BackgroundMat {
     id=#3
     possessed_class=Image
     tracks [{
       type=MovieSceneWidgetMaterialTrack
       sections [{
         type=MovieSceneParameterSection
         end_frame=180
         start_frame=0
         channels [
          RightOffset {type=MovieSceneScriptingFloatChannel, keys [[0,0.75],[45,0.75],[120.002,0]]}
          LerpToGreen {type=MovieSceneScriptingFloatChannel, const [0,1]}
          LerpToRed {type=MovieSceneScriptingFloatChannel, const [0,1]}
          LeftOffset {type=MovieSceneScriptingFloatChannel, keys [[0,0.915],[120,0.915],[150,0.175]]}
          HideProgressBar {type=MovieSceneScriptingFloatChannel, const [1,1]}
          LerpAngleChange {type=MovieSceneScriptingFloatChannel, keys [[0,0],[120,0],[150,1]]}
          Burst {type=MovieSceneScriptingFloatChannel, keys [[0,1],[90,1],[135,0]]}
          Trim_ColorR {type=MovieSceneScriptingFloatChannel, keys [[0,0.0021],[60,0.0021],[90,1],[150,0.0021]]}
          Trim_ColorG {type=MovieSceneScriptingFloatChannel, keys [[0,0.0612],[60,0.0612],[90,1],[150,0.0612]]}
          Trim_ColorB {type=MovieSceneScriptingFloatChannel, keys [[0,0.5029],[60,0.5029],[90,1],[150,0.5029]]}
          Trim_ColorA {type=MovieSceneScriptingFloatChannel, const [1,4]}
         ]
        }]
      }]
    }
    Image_Trim {
     id=#12
     possessed_class=Image
     tracks [{
       type=MovieSceneWidgetMaterialTrack
       sections [{
         type=MovieSceneParameterSection
         end_frame=186
         start_frame=0
         channels [
          RightOffset {type=MovieSceneScriptingFloatChannel, keys [[0,0.75],[45,0.75],[120.002,0]]}
          LeftOffset {type=MovieSceneScriptingFloatChannel, keys [[0,0.985],[120.002,0.985],[150.002,0.15]]}
          LerpAngleChange {type=MovieSceneScriptingFloatChannel, keys [[0,0],[120.002,0],[150.002,1]]}
          LerpToGreen {type=MovieSceneScriptingFloatChannel, const [0,1]}
          LerpToRed {type=MovieSceneScriptingFloatChannel, const [0,1]}
          Burst {type=MovieSceneScriptingFloatChannel, keys [[0,1],[90,1],[135,0]]}
          HideProgressBar {type=MovieSceneScriptingFloatChannel, const [1,1]}
          Trim_ColorR {type=MovieSceneScriptingFloatChannel, keys [[0,0.0021],[60,0.0021],[90,1],[150,0.0021]]}
          Trim_ColorG {type=MovieSceneScriptingFloatChannel, keys [[0,0.0612],[60,0.0612],[90,1],[150,0.0612]]}
          Trim_ColorB {type=MovieSceneScriptingFloatChannel, keys [[0,0.5029],[60,0.5029],[90,1],[150,0.5029]]}
          Trim_ColorA {type=MovieSceneScriptingFloatChannel, const [1,4]}
         ]
        }]
      }]
    }
   ]
   display_rate {numerator=120}
   tracks [{
     type=MovieSceneEventTrack
     sections [{
       type=MovieSceneEventTriggerSection
       channels [None {type=MovieSceneScriptingEventChannel, const ["<Struct 'MovieSceneEvent' (0x4687a1ad700) {}>", 1]}]
      }]
    }]
  }
 }
 Anim_IntroNeutralInfoSimple {
  binding_count=9
  length_seconds=1.75
  object=&self:Anim_IntroNeutralInfoSimple
  sequence {
   bindings [
    Root {
     id=#1
     possessed_class=Overlay
     tracks [{
       type=MovieScene2DTransformTrack
       sections [{
         type=MovieScene2DTransformSection
         end_frame=90
         start_frame=0
         channels [Translation.X {type=MovieSceneScriptingFloatChannel, keys [[0,-1156.3329],[90,0]]}]
        }]
      }]
    }
    Image_ToastCushion {
     id=#15
     possessed_class=Image
     tracks [{
       type=MovieSceneWidgetMaterialTrack
       sections [{
         type=MovieSceneParameterSection
         end_frame=210
         start_frame=0
         channels [HideProgressBar {type=MovieSceneScriptingFloatChannel, const [1,1]}]
        }]
      }]
    }
    Image_Fail {
     id=#5
     possessed_class=Image
     tracks [{
       type=MovieSceneByteTrack
       display_name=Visibility
       sections [{type=MovieSceneByteSection, end_frame=12, start_frame=0, channels [None {type=MovieSceneScriptingByteChannel, const [1,1]}]}]
      }]
    }
    Overlay_ResultIcon {
     id=#6
     possessed_class=Overlay
     tracks [
      {
       type=MovieSceneFloatTrack
       display_name=Render Opacity
       sections [{type=MovieSceneFloatSection, end_frame=90, start_frame=0, channels [None {type=MovieSceneScriptingFloatChannel, const [1,1]}]}]
      }
      {
       type=MovieScene2DTransformTrack
       sections [{
         type=MovieScene2DTransformSection
         end_frame=90
         start_frame=0
         channels [
          Scale.X {type=MovieSceneScriptingFloatChannel, const [1,1]}
          Scale.Y {type=MovieSceneScriptingFloatChannel, const [1,1]}
         ]
        }]
      }
     ]
    }
    VerticalBox_HoldAction {
     id=#2
     possessed_class=VerticalBox
     tracks [{
       type=MovieSceneFloatTrack
       display_name=Render Opacity
       sections [{type=MovieSceneFloatSection, end_frame=90, start_frame=0, channels [None {type=MovieSceneScriptingFloatChannel, const [0,1]}]}]
      }]
    }
    Image_Success {
     id=#8
     possessed_class=Image
     tracks [{
       type=MovieSceneByteTrack
       display_name=Visibility
       sections [{type=MovieSceneByteSection, end_frame=2, start_frame=0, channels [None {type=MovieSceneScriptingByteChannel, const [1,1]}]}]
      }]
    }
    Image_BackgroundMat {
     id=#3
     possessed_class=Image
     tracks [{
       type=MovieSceneWidgetMaterialTrack
       sections [{
         type=MovieSceneParameterSection
         end_frame=210
         start_frame=0
         channels [
          RightOffset {type=MovieSceneScriptingFloatChannel, const [0,2]}
          LerpToGreen {type=MovieSceneScriptingFloatChannel, const [0,1]}
          LerpToRed {type=MovieSceneScriptingFloatChannel, const [0,1]}
          LeftOffset {type=MovieSceneScriptingFloatChannel, const [0.175,1]}
          HideProgressBar {type=MovieSceneScriptingFloatChannel, const [1,2]}
          LerpAngleChange {type=MovieSceneScriptingFloatChannel, const [1,1]}
          Burst {type=MovieSceneScriptingFloatChannel, const [0,1]}
          Trim_ColorR {type=MovieSceneScriptingFloatChannel, const [0.0021,1]}
          Trim_ColorG {type=MovieSceneScriptingFloatChannel, const [0.0612,1]}
          Trim_ColorB {type=MovieSceneScriptingFloatChannel, const [0.5029,1]}
          Trim_ColorA {type=MovieSceneScriptingFloatChannel, const [1,2]}
         ]
        }]
      }]
    }
    Image_Trim {
     id=#12
     possessed_class=Image
     tracks [{
       type=MovieSceneWidgetMaterialTrack
       sections [{
         type=MovieSceneParameterSection
         end_frame=210
         start_frame=0
         channels [
          RightOffset {type=MovieSceneScriptingFloatChannel, const [0.65,1]}
          LeftOffset {type=MovieSceneScriptingFloatChannel, const [0.8,1]}
          LerpAngleChange {type=MovieSceneScriptingFloatChannel, const [0,1]}
          LerpToGreen {type=MovieSceneScriptingFloatChannel, const [0,1]}
          LerpToRed {type=MovieSceneScriptingFloatChannel, const [0,1]}
          Burst {type=MovieSceneScriptingFloatChannel, const [0,1]}
          HideProgressBar {type=MovieSceneScriptingFloatChannel, const [1,1]}
          Trim_ColorR {type=MovieSceneScriptingFloatChannel, const [0.0021,1]}
          Trim_ColorG {type=MovieSceneScriptingFloatChannel, const [0.0612,1]}
          Trim_ColorB {type=MovieSceneScriptingFloatChannel, const [0.5029,1]}
          Trim_ColorA {type=MovieSceneScriptingFloatChannel, const [1,1]}
         ]
        }]
      }]
    }
   ]
   display_rate {numerator=120}
   tracks [{
     type=MovieSceneEventTrack
     sections [{
       type=MovieSceneEventTriggerSection
       channels [None {type=MovieSceneScriptingEventChannel, const ["<Struct 'MovieSceneEvent' (0x4687a1ae7e0) {}>", 1]}]
      }]
    }]
  }
 }
 Anim_HoldProgress {
  binding_count=8
  length_seconds=1.75
  object=&self:Anim_HoldProgress
  sequence {
   bindings [
    Image_Fail {
     id=#16
     possessed_class=Image
     tracks [{
       type=MovieSceneByteTrack
       display_name=Visibility
       sections [{type=MovieSceneByteSection, end_frame=3, start_frame=0, channels [None {type=MovieSceneScriptingByteChannel, const [1,1]}]}]
      }]
    }
    Image_ProgressCircle {
     id=#17
     possessed_class=Image
     tracks [{
       type=MovieSceneWidgetMaterialTrack
       sections [{
         type=MovieSceneParameterSection
         end_frame=21
         start_frame=0
         channels [Progress {type=MovieSceneScriptingFloatChannel, keys [[0,0],[5,0],[21,1]]}]
        }]
      }]
    }
    InputKeybind {
     id=#18
     possessed_class=Overlay
     tracks [
      {
       type=MovieScene2DTransformTrack
       sections [{
         type=MovieScene2DTransformSection
         end_frame=5
         start_frame=0
         channels [
          Scale.X {type=MovieSceneScriptingFloatChannel, keys [[0,1],[5,0.9]]}
          Scale.Y {type=MovieSceneScriptingFloatChannel, keys [[0,1],[5,0.9]]}
         ]
        }]
      }
      {
       type=MovieSceneFloatTrack
       display_name=Render Opacity
       sections [{
         type=MovieSceneFloatSection
         end_frame=26
         start_frame=0
         channels [None {type=MovieSceneScriptingFloatChannel, keys [[0,1],[21.0003,1],[26,0]]}]
        }]
      }
     ]
    }
    Image_BackgroundMat {
     id=#19
     possessed_class=Image
     tracks [{
       type=MovieSceneWidgetMaterialTrack
       sections [{
         type=MovieSceneParameterSection
         end_frame=35
         start_frame=0
         channels [
          LerpToGreen {type=MovieSceneScriptingFloatChannel, keys [[0,0],[26,0],[35,1]]}
          LeftOffset {type=MovieSceneScriptingFloatChannel, keys [[0,0.915],[26,0.915],[31,0.175]]}
          LerpToRed {type=MovieSceneScriptingFloatChannel, const [0,1]}
          LerpAngleChange {type=MovieSceneScriptingFloatChannel, keys [[0,0],[26,0],[31,1]]}
          Trim_ColorR {type=MovieSceneScriptingFloatChannel, keys [[0,0.0021],[16,0.0021],[21,0.25],[26,1],[31,0]]}
          Trim_ColorG {type=MovieSceneScriptingFloatChannel, keys [[0,0.0612],[16,0.0612],[21,1],[26,1],[31,0.6867]]}
          Trim_ColorB {type=MovieSceneScriptingFloatChannel, keys [[0,0.5029],[16,0.5029],[21,0.25],[26,1],[31,0.0887]]}
          Trim_ColorA {type=MovieSceneScriptingFloatChannel, const [1,5]}
         ]
        }]
      }]
    }
    Image_Trim {
     id=#20
     possessed_class=Image
     tracks [{
       type=MovieSceneWidgetMaterialTrack
       sections [{
         type=MovieSceneParameterSection
         end_frame=35
         start_frame=0
         channels [
          LeftOffset {type=MovieSceneScriptingFloatChannel, keys [[0,0.985],[26,0.985],[31,0.15]]}
          LerpAngleChange {type=MovieSceneScriptingFloatChannel, keys [[0,0],[26,0],[31,1]]}
          LerpToGreen {type=MovieSceneScriptingFloatChannel, keys [[0,0],[26,0],[35,1]]}
          LerpToRed {type=MovieSceneScriptingFloatChannel, const [0,1]}
          Trim_ColorR {type=MovieSceneScriptingFloatChannel, keys [[0,0.0021],[16,0.0021],[21,0.25],[26,1],[31,0]]}
          Trim_ColorG {type=MovieSceneScriptingFloatChannel, keys [[0,0.0612],[16,0.0612],[21,1],[26,1],[31,0.6867]]}
          Trim_ColorB {type=MovieSceneScriptingFloatChannel, keys [[0,0.5029],[16,0.5029],[21,0.25],[26,1],[31,0.0887]]}
          Trim_ColorA {type=MovieSceneScriptingFloatChannel, const [1,5]}
         ]
        }]
      }]
    }
    Text_Action {
     id=#21
     possessed_class=CommonTextBlock
     tracks [{
       type=MovieSceneFloatTrack
       display_name=Render Opacity
       sections [{type=MovieSceneFloatSection, end_frame=20, start_frame=0, channels [None {type=MovieSceneScriptingFloatChannel, keys [[0,1],[5,1],[20,0]]}]}]
      }]
    }
    Image_Success {
     id=#22
     possessed_class=Image
     tracks [{
       type=MovieSceneByteTrack
       display_name=Visibility
       sections [{type=MovieSceneByteSection, end_frame=3, start_frame=0, channels [None {type=MovieSceneScriptingByteChannel, const [4,1]}]}]
      }]
    }
    Image_Success {
     id=#23
     display_name=Overlay_ResultIcon
     possessed_class=Image
     tracks [
      {
       type=MovieScene2DTransformTrack
       sections [{
         type=MovieScene2DTransformSection
         end_frame=33
         start_frame=0
         channels [
          Scale.X {type=MovieSceneScriptingFloatChannel, keys [[0,2],[26,2],[33,1]]}
          Scale.Y {type=MovieSceneScriptingFloatChannel, keys [[0,2],[26,2],[33,1]]}
         ]
        }]
      }
      {
       type=MovieSceneFloatTrack
       display_name=Render Opacity
       sections [{
         type=MovieSceneFloatSection
         end_frame=35
         start_frame=0
         channels [None {type=MovieSceneScriptingFloatChannel, keys [[0,0],[26,0],[35.0003,1]]}]
        }]
      }
     ]
    }
   ]
   tracks [{
     type=MovieSceneEventTrack
     sections [{
       type=MovieSceneEventTriggerSection
       channels [None {type=MovieSceneScriptingEventChannel, const ["<Struct 'MovieSceneEvent' (0x4687a1af460) {}>", 1]}]
      }]
    }]
  }
 }
 Anim_Outro {
  binding_count=2
  length_seconds=0.6
  object=&self:Anim_Outro
  sequence {
   bindings [
    Root {
     id=#24
     possessed_class=Overlay
     tracks [{
       type=MovieScene2DTransformTrack
       sections [{
         type=MovieScene2DTransformSection
         end_frame=36
         start_frame=0
         channels [
          Translation.X {type=MovieSceneScriptingFloatChannel, keys [[0,0],[11,48],[18,-1200],[18,-1214.6853]]}
          Translation.Y {type=MovieSceneScriptingFloatChannel}
          Shear.X {type=MovieSceneScriptingFloatChannel, keys [[0,0],[11,0],[15,-15]]}
          Shear.Y {type=MovieSceneScriptingFloatChannel}
         ]
        }]
      }]
    }
    Image_BackgroundMat {
     id=#25
     possessed_class=Image
     tracks [{
       type=MovieSceneWidgetMaterialTrack
       sections [{type=MovieSceneParameterSection, end_frame=36, start_frame=0, channels [Progress {type=MovieSceneScriptingFloatChannel, const [1,1]}]}]
      }]
    }
   ]
   display_rate {numerator=60}
  }
 }
 Anim_TimeToRespond {
  binding_count=1
  length_seconds=1
  object=&self:Anim_TimeToRespond
  sequence {bindings [Image_BackgroundMat {
     id=#26
     possessed_class=Image
     tracks [{
       type=MovieSceneWidgetMaterialTrack
       sections [{
         type=MovieSceneParameterSection
         end_frame=20
         start_frame=0
         channels [Progress {type=MovieSceneScriptingFloatChannel, keys [[0,0],[20,1]]}]
        }]
      }]
    }]}
 }
 Anim_ClickAccept {
  binding_count=6
  length_seconds=0.45
  object=&self:Anim_ClickAccept
  sequence {
   bindings [
    Image_Fail {
     id=#16
     possessed_class=Image
     tracks [{
       type=MovieSceneByteTrack
       display_name=Visibility
       sections [{type=MovieSceneByteSection, end_frame=3, start_frame=0, channels [None {type=MovieSceneScriptingByteChannel, const [1,1]}]}]
      }]
    }
    Image_BackgroundMat {
     id=#19
     possessed_class=Image
     tracks [{
       type=MovieSceneWidgetMaterialTrack
       sections [{
         type=MovieSceneParameterSection
         end_frame=9
         start_frame=0
         channels [
          LerpToGreen {type=MovieSceneScriptingFloatChannel, keys [[0,0],[9,1]]}
          LeftOffset {type=MovieSceneScriptingFloatChannel, keys [[0,0.915],[5,0.175]]}
          LerpToRed {type=MovieSceneScriptingFloatChannel, const [0,1]}
          LerpAngleChange {type=MovieSceneScriptingFloatChannel, keys [[0,0],[5,1]]}
          Trim_ColorR {type=MovieSceneScriptingFloatChannel, keys [[0,1],[5,0]]}
          Trim_ColorG {type=MovieSceneScriptingFloatChannel, keys [[0,1],[5,0.6867]]}
          Trim_ColorB {type=MovieSceneScriptingFloatChannel, keys [[0,1],[5,0.0887]]}
          Trim_ColorA {type=MovieSceneScriptingFloatChannel, const [1,2]}
         ]
        }]
      }]
    }
    Image_Trim {
     id=#20
     possessed_class=Image
     tracks [{
       type=MovieSceneWidgetMaterialTrack
       sections [{
         type=MovieSceneParameterSection
         end_frame=9
         start_frame=0
         channels [
          LeftOffset {type=MovieSceneScriptingFloatChannel, keys [[0,0.985],[5,0.15]]}
          LerpAngleChange {type=MovieSceneScriptingFloatChannel, keys [[0,0],[5,1]]}
          LerpToGreen {type=MovieSceneScriptingFloatChannel, keys [[0,0],[9,1]]}
          LerpToRed {type=MovieSceneScriptingFloatChannel, const [0,1]}
          Trim_ColorR {type=MovieSceneScriptingFloatChannel, keys [[0,1],[5,0]]}
          Trim_ColorG {type=MovieSceneScriptingFloatChannel, keys [[0,1],[5,0.6867]]}
          Trim_ColorB {type=MovieSceneScriptingFloatChannel, keys [[0,1],[5,0.0887]]}
          Trim_ColorA {type=MovieSceneScriptingFloatChannel, const [1,2]}
         ]
        }]
      }]
    }
    Text_Action {
     id=#21
     possessed_class=CommonTextBlock
     tracks [{
       type=MovieSceneFloatTrack
       display_name=Render Opacity
       sections [{type=MovieSceneFloatSection, end_frame=9, start_frame=0, channels [None {type=MovieSceneScriptingFloatChannel, const [0,1]}]}]
      }]
    }
    Image_Success {
     id=#22
     possessed_class=Image
     tracks [{
       type=MovieSceneByteTrack
       display_name=Visibility
       sections [{type=MovieSceneByteSection, end_frame=3, start_frame=0, channels [None {type=MovieSceneScriptingByteChannel, const [4,1]}]}]
      }]
    }
    Image_Success {
     id=#23
     display_name=Overlay_ResultIcon
     possessed_class=Image
     tracks [
      {
       type=MovieScene2DTransformTrack
       sections [{
         type=MovieScene2DTransformSection
         end_frame=7
         start_frame=0
         channels [
          Scale.X {type=MovieSceneScriptingFloatChannel, keys [[0,2],[7,1]]}
          Scale.Y {type=MovieSceneScriptingFloatChannel, keys [[0,2],[7,1]]}
         ]
        }]
      }
      {
       type=MovieSceneFloatTrack
       display_name=Render Opacity
       sections [{type=MovieSceneFloatSection, end_frame=9, start_frame=0, channels [None {type=MovieSceneScriptingFloatChannel, keys [[0,0],[9.0003,1]]}]}]
      }
     ]
    }
   ]
   tracks [{
     type=MovieSceneEventTrack
     sections [{
       type=MovieSceneEventTriggerSection
       channels [None {type=MovieSceneScriptingEventChannel, const ["<Struct 'MovieSceneEvent' (0x4687a4d0500) {}>", 1]}]
      }]
    }]
  }
 }
 Anim_TapAccept {
  binding_count=6
  length_seconds=0.45
  object=&self:Anim_TapAccept
  sequence {
   bindings [
    Image_Fail {
     id=#16
     possessed_class=Image
     tracks [{
       type=MovieSceneByteTrack
       display_name=Visibility
       sections [{type=MovieSceneByteSection, end_frame=3, start_frame=0, channels [None {type=MovieSceneScriptingByteChannel, const [1,1]}]}]
      }]
    }
    Image_BackgroundMat {
     id=#19
     possessed_class=Image
     tracks [{
       type=MovieSceneWidgetMaterialTrack
       sections [{
         type=MovieSceneParameterSection
         end_frame=9
         start_frame=0
         channels [
          LerpToGreen {type=MovieSceneScriptingFloatChannel, keys [[0,0],[9,1]]}
          LeftOffset {type=MovieSceneScriptingFloatChannel, keys [[0,0.915],[5,0.175]]}
          LerpToRed {type=MovieSceneScriptingFloatChannel, const [0,1]}
          LerpAngleChange {type=MovieSceneScriptingFloatChannel, keys [[0,0],[5,1]]}
          Trim_ColorR {type=MovieSceneScriptingFloatChannel, keys [[0,0.0021],[5,0]]}
          Trim_ColorG {type=MovieSceneScriptingFloatChannel, keys [[0,0.0612],[5,0.6867]]}
          Trim_ColorB {type=MovieSceneScriptingFloatChannel, keys [[0,0.5029],[5,0.0887]]}
          Trim_ColorA {type=MovieSceneScriptingFloatChannel, const [1,2]}
         ]
        }]
      }]
    }
    Image_Trim {
     id=#20
     possessed_class=Image
     tracks [{
       type=MovieSceneWidgetMaterialTrack
       sections [{
         type=MovieSceneParameterSection
         end_frame=9
         start_frame=0
         channels [
          LeftOffset {type=MovieSceneScriptingFloatChannel, keys [[0,0.985],[5,0.15]]}
          LerpAngleChange {type=MovieSceneScriptingFloatChannel, keys [[0,0],[5,1]]}
          LerpToGreen {type=MovieSceneScriptingFloatChannel, keys [[0,0],[9,1]]}
          LerpToRed {type=MovieSceneScriptingFloatChannel, const [0,1]}
          Trim_ColorR {type=MovieSceneScriptingFloatChannel, keys [[0,0.0021],[5,0]]}
          Trim_ColorG {type=MovieSceneScriptingFloatChannel, keys [[0,0.0612],[5,0.6867]]}
          Trim_ColorB {type=MovieSceneScriptingFloatChannel, keys [[0,0.5029],[5,0.0887]]}
          Trim_ColorA {type=MovieSceneScriptingFloatChannel, const [1,2]}
         ]
        }]
      }]
    }
    Text_Action {
     id=#21
     possessed_class=CommonTextBlock
     tracks [{
       type=MovieSceneFloatTrack
       display_name=Render Opacity
       sections [{type=MovieSceneFloatSection, end_frame=9, start_frame=0, channels [None {type=MovieSceneScriptingFloatChannel, const [0,1]}]}]
      }]
    }
    Image_Success {
     id=#22
     possessed_class=Image
     tracks [{
       type=MovieSceneByteTrack
       display_name=Visibility
       sections [{type=MovieSceneByteSection, end_frame=3, start_frame=0, channels [None {type=MovieSceneScriptingByteChannel, const [4,1]}]}]
      }]
    }
    Image_Success {
     id=#23
     display_name=Overlay_ResultIcon
     possessed_class=Image
     tracks [
      {
       type=MovieScene2DTransformTrack
       sections [{
         type=MovieScene2DTransformSection
         end_frame=7
         start_frame=0
         channels [
          Scale.X {type=MovieSceneScriptingFloatChannel, keys [[0,2],[7,1]]}
          Scale.Y {type=MovieSceneScriptingFloatChannel, keys [[0,2],[7,1]]}
         ]
        }]
      }
      {
       type=MovieSceneFloatTrack
       display_name=Render Opacity
       sections [{type=MovieSceneFloatSection, end_frame=9, start_frame=0, channels [None {type=MovieSceneScriptingFloatChannel, keys [[0,0],[9.0003,1]]}]}]
      }
     ]
    }
   ]
   tracks [{
     type=MovieSceneEventTrack
     sections [{
       type=MovieSceneEventTriggerSection
       channels [None {type=MovieSceneScriptingEventChannel, const ["<Struct 'MovieSceneEvent' (0x4687a4d06e0) {}>", 1]}]
      }]
    }]
  }
 }
]
graphs [
 CleanupBeforeClosing {nodes [
   CleanupBeforeClosing() {is=title K2Node_FunctionEntry}
   > Clear and Invalidate Timer by Handle(Handle=Get AnimationDelayTimer() -> (AnimationDelayTimer:Timer Handle Structure))
   > SetVisibility(self=Get Switcher_InputButtons() -> (Switcher_InputButtons:Common Widget Switcher Legacy Object Reference), InVisibility=Collapsed)
  ]}
 DetermineIntroAnim {nodes [
   DetermineIntroAnim() {is=title K2Node_FunctionEntry}
   n1 = Break S UI Toast Property!(S_UI_ToastProperty:"S UI Toast Property Structure (by ref)"=Get Data() -> (Data:S UI Toast Property Structure)) -> (
    IntroAnimationType_14_#27:E_PositiveNegativeNeutral Enum
    UseComplexAnim_21_#28:bool=~
   ) {is=title K2Node_BreakStruct}
   n2 = Break S UI Toast Property!(S_UI_ToastProperty=Get Data()) -> (UseComplexAnim_21_#28=~)
   n3 = Select(
    Option 0:Widget Animation Object Reference=Get Anim_IntroNegativeInfoSimple() -> (Anim_IntroNegativeInfoSimple:Widget Animation Object Reference)
    Option 1:Widget Animation Object Reference=Get Anim_IntroNegativeInfo() -> (Anim_IntroNegativeInfo:Widget Animation Object Reference)
    Index=@n2
   ) -> (ReturnValue:Widget Animation Object Reference)
   n4 = Select(
    Option 0=Get Anim_IntroNeutralInfoSimple() -> (Anim_IntroNeutralInfoSimple:Widget Animation Object Reference)
    Option 1=Get Anim_IntroNeutralInfo() -> (Anim_IntroNeutralInfo:Widget Animation Object Reference)
    Index=@n2
   )
   n5 = Select(
    Option 0=Get Anim_IntroPositiveInfoSimple() -> (Anim_IntroPositiveInfoSimple:Widget Animation Object Reference)
    Option 1=Get Anim_IntroPositiveInfo() -> (Anim_IntroPositiveInfo:Widget Animation Object Reference)
    Index=@n2
   )
   n6 = Get CachedNotification() -> (CachedNotification:Fort UINotification Object Reference)
   > Return Node!(
    execute
    ReturnValue:Widget Animation Object Reference=Select(
     Option 0=Select!(
      NewEnumerator0:Widget Animation Object Reference=@n4
      NewEnumerator1:Widget Animation Object Reference=@n3
      NewEnumerator2:Widget Animation Object Reference=@n5
      Index:E_PositiveNegativeNeutral Enum=@n1
     ) -> (ReturnValue)
     Option 1=Get Anim_IntroAction() -> (Anim_IntroAction:Widget Animation Object Reference)
     Index=Get bHasAction!(self:Fort UINotification Object Reference=@n6) -> (bHasAction:bool)
    )
   )
  ]}
 EventGraph {nodes [
   "On Button Clicked (Button_Hold)"!() -> (OutputDelegate, then, Button:Common Button Legacy Object Reference) {is=title K2Node_ComponentBoundEvent}
   > TakeAction()
   > CleanupBeforeClosing() {is=title K2Node_CallFunction}
   > Delay(Duration=0.75)
   > PlayOutro()
   "On Button Hovered (Button_Cursor)"!() -> (OutputDelegate, then, Button:Common Button Legacy Object Reference) {is=title K2Node_ComponentBoundEvent}
   n1 = Get CommonInputSubsystem()
   n2 = AND Boolean(A="Equal (Enum)"(A=GetCurrentInputType(self=@n1), B=MouseAndKeyboard), B=Get bHasAction(self=Get CachedNotification()))
   > Branch(Condition=@n2)
   > SetVectorParameterValue(self=Get TrimMat() -> (TrimMat:Material Instance Dynamic Object Reference), ParameterName=Trim_Color, Value=(R=1,G=1,B=1,A=1))
   > SetColorAndOpacity!(
    execute
    self=Get Text_Action() -> (Text_Action:Common Text Object Reference)
    InColorAndOpacity_SpecifiedColor:Linear Color Structure=(R=0.0024,G=0.0612,B=0.5029,A=1)
    InColorAndOpacity_ColorUseRule:ESlateColorStylingMode Enum
   ) -> (then)
   "On Button Unhovered (Button_Cursor)"!() -> (OutputDelegate, then, Button:Common Button Legacy Object Reference) {is=title K2Node_ComponentBoundEvent}
   n3 = Get IsActionTaken() -> (IsActionTaken:bool)
   > Branch(Condition=AND Boolean(A=@n2, B=NOT Boolean(A=@n3)))
   > SetVectorParameterValue(self=Get TrimMat(), ParameterName=Trim_Color, Value=(R=0.0024,G=0.0612,B=0.5029,A=1))
   > SetColorAndOpacity(self=Get Text_Action(), InColorAndOpacity_SpecifiedColor=(R=1,G=1,B=1,A=1))
   "Hold Action Started (Button_Cursor)"() {is=title K2Node_ComponentBoundEvent}
   > HoldPressed(self:Fort Social Notification Widget Object Reference)
   "Hold Action Ended (Button_Cursor)"() {is=title K2Node_ComponentBoundEvent}
   > HoldReleased(self:Fort Social Notification Widget Object Reference)
   "On Button Clicked (Button_Cursor)"!() -> (OutputDelegate, then, Button:Common Button Legacy Object Reference) {is=title K2Node_ComponentBoundEvent}
   > n4 = SpawnSound2D!(
    execute
    Sound:Sound Base Object Reference=Get ButtonClickSound() -> (ButtonClickSound:Sound Base Object Reference)
    VolumeMultiplier:float=1
    PitchMultiplier:float=1
    StartTime:float=0
    ConcurrencySettings:Sound Concurrency Object Reference
    bPersistAcrossLevelTransition:bool=false
    bAutoDestroy:bool=true
   ) -> (then, ReturnValue:Audio Component Object Reference)
   SetVisibility(execute=[@n4.then, @n6.then], self=Get InputKeybind() -> (InputKeybind:Overlay Object Reference), InVisibility=Hidden)
   n5 = GetCurrentInputType(self=Get CommonInputSubsystem())
   > PlayAnimation(
    InAnimation=Select!(
     Option 0:Widget Animation Object Reference=Get Anim_TapAccept() -> (Anim_TapAccept:Widget Animation Object Reference)
     Option 1:Widget Animation Object Reference=Get Anim_ClickAccept() -> (Anim_ClickAccept:Widget Animation Object Reference)
     Index:bool="Equal (Enum)"(A=@n5, B=MouseAndKeyboard)
    ) -> (ReturnValue)
    PlaybackSpeed=1.5
   )
   > TakeAction()
   > CleanupBeforeClosing()
   > Delay(Duration=0.33)
   > PlayOutro()
   n6 = "Hold Action Completed (Button_Cursor)"() {is=title K2Node_ComponentBoundEvent}
   HandleOutroFinished() {is=title K2Node_CustomEvent}
   > ResetMaterials()
   > SetupAnimBindings!(execute, self, Unbind:bool=true) -> (then)
   > NotifyCloseAnimationFinished(self:Fort Toast Widget Object Reference)
   HandleIntroStarted() {is=title K2Node_CustomEvent}
   n7 = DetermineIntroAnim!() -> (ReturnValue:Widget Animation Object Reference) {is=title K2Node_CallFunction}
   > Delay(Duration=float / float(A=GetEndTime(self=@n7), B=15))
   > SetDescriptionTextWrapValue!(execute, self) -> (then, IsSizeEmpty:bool)
   HandleIntroFinished() {is=title K2Node_CustomEvent}
   n8 = GetDurationTime!(self:Fort Social Notification Widget Object Reference) -> (ReturnValue:float)
   > n9 = Set Timer by Event(Delegate=Create Event(), Time=@n8, bLooping=false)
   > Set AnimationDelayTimer(AnimationDelayTimer:Timer Handle Structure=@n9) -> (Output_Get:Timer Handle Structure)
   > Branch(Condition=Get bHasAction(self=Get CachedNotification()))
   > PlayAnimation(InAnimation=Get Anim_TimeToRespond() -> (Anim_TimeToRespond:Widget Animation Object Reference), PlaybackSpeed=float / float(A=1, B=@n8))
   PlayOutro() {is=title K2Node_CustomEvent}
   > PlayAnimation(InAnimation=Get Anim_Outro() -> (Anim_Outro:Widget Animation Object Reference))
   > SetVisibility(self=Get Switcher_InputButtons(), InVisibility=Collapsed)
   n10 = Event OnNewAvatarsEnabledChanged!() -> (OutputDelegate, then, bUsingNewAvatars:bool) {is=title K2Node_Event}
   > n11 = Set UsingNewAvatars(UsingNewAvatars:bool=@n10.bUsingNewAvatars) -> (Output_Get:bool)
   > SetVisibility(
    self=Get PlayerAvatar() -> (PlayerAvatar:Fort Player Avatar Object Reference)
    InVisibility=Select(Option 0:ESlateVisibility Enum=Collapsed, Option 1:ESlateVisibility Enum=HitTestInvisible, Index=@n11) -> (ReturnValue:ESlateVisibility Enum)
   )
   > SetVisibility(
    self=Get Avatar_MemberIcon() -> (Avatar_MemberIcon:Fort Social Avatar Icon Object Reference)
    InVisibility=Select(Option 0=HitTestInvisible, Option 1=Collapsed, Index=@n11)
   )
   Event OnPlayOpenAnimation() {is=title K2Node_Event}
   > UpdateData!(execute, self, Notification:Fort UINotification Object Reference=Get CachedNotification()) -> (then)
   n12 = Break S UI Toast Property(S_UI_ToastProperty=Get Data()) -> (UseComplexAnim_21_#28=~)
   n13 = Get CachedNotification()
   > PlayAnimation(
    InAnimation=DetermineIntroAnim()
    PlaybackSpeed=Select(Option 0:double=Select(Option 0=2, Option 1=1, Index=@n12), Option 1:double=1.5, Index=Get bHasAction(self=@n13)) -> (ReturnValue:double)
   )
   > Set IsActionTaken(IsActionTaken:bool=false) -> (Output_Get:bool)
   > PlayNotificationSound()
   Event OnUserStartedHoldActionNotification() {is=title K2Node_Event}
   > Pause Timer by Handle!(execute, Handle:Timer Handle Structure=Get AnimationDelayTimer()) -> (then)
   > PlayAnimation(InAnimation=Get Anim_HoldProgress() -> (Anim_HoldProgress:Widget Animation Object Reference))
   Event OnMouseEnter()
   > Pause Timer by Handle(Handle=Get AnimationDelayTimer())
   n14 = Get Anim_TimeToRespond()
   > Branch(Condition=IsAnimationPlaying(InAnimation=@n14))
   > PauseAnimation!(execute, self:User Widget Object Reference, InAnimation:Widget Animation Object Reference=@n14) -> (then, ReturnValue:float)
   Event OnMouseLeave!() -> (OutputDelegate, then, MouseEvent:"Pointer Event Structure (by ref)") {is=title K2Node_Event}
   > Unpause Timer by Handle!(execute, Handle:Timer Handle Structure=Get AnimationDelayTimer()) -> (then)
   n15 = GetAnimationCurrentTime!(self:User Widget Object Reference, InAnimation:Widget Animation Object Reference=Get Anim_TimeToRespond()) -> (ReturnValue:float)
   > Branch(Condition="In Range (Float)"!(Value:double=@n15, Min:double=0, Max:double=1, InclusiveMin:bool=false, InclusiveMax:bool=false) -> (ReturnValue:bool))
   > PlayAnimation(InAnimation=Get Anim_TimeToRespond(), StartAtTime=@n15, PlaybackSpeed=float / float(A=1, B=GetDurationTime()))
   n16 = Event OnCursorModeChanged!() -> (OutputDelegate, then, bCursorModeEnabled:bool) {is=title K2Node_Event}
   n17 = IsUsingTouch(PlayerController=GetOwningPlayer())
   > SetActiveWidget(
    self:Widget Switcher Object Reference=Get Switcher_InputButtons()
    Widget=Select(
     Option 0:Widget Object Reference=Get Button_Hold() -> (Button_Hold:Icon Text Button Object Reference)
     Option 1:Widget Object Reference=Get Button_Cursor() -> (Button_Cursor:Icon Text Button Hold Object Reference)
     Index=OR Boolean(A=@n16.bCursorModeEnabled, B=@n17)
    ) -> (ReturnValue:Widget Object Reference)
   )
   > InputBasedStyling()
   > Unpause Timer by Handle(Handle=Get AnimationDelayTimer())
   Event OnCancelNotification() {is=title K2Node_Event}
   n18 = "Not Equal (Enum)"!(
    A:ESlateVisibility Enum=GetVisibility!(self:Widget Object Reference=Get Switcher_InputButtons()) -> (ReturnValue:ESlateVisibility Enum=Visible)
    B:ESlateVisibility Enum=Collapsed
   ) -> (ReturnValue:bool)
   > Branch(Condition=@n18)
   > n19 = StopAnimation(InAnimation=Get Anim_HoldProgress())
   CleanupBeforeClosing(execute=[@n20.then, @n19.then])
   > Delay(Duration=0.05)
   > PlayOutro() {is=title K2Node_CallFunction}
   Event OnTakeActionNotification() {is=title K2Node_Event}
   > Branch(Condition=@n18)
   > n20 = TakeAction()
   Event Construct()
   > n21 = GetDynamicMaterial(self=Get Image_BackgroundMat() -> (Image_BackgroundMat:Image Object Reference))
   > Set BackgroundMat(BackgroundMat:Material Instance Dynamic Object Reference=@n21) -> (Output_Get:Material Instance Dynamic Object Reference)
   > n22 = GetDynamicMaterial(self=Get Image_Trim() -> (Image_Trim:Image Object Reference))
   > Set TrimMat(TrimMat:Material Instance Dynamic Object Reference=@n22) -> (Output_Get:Material Instance Dynamic Object Reference)
   Event OnInputModeChanged!() -> (OutputDelegate, then, NewInputType:ECommonInputType Enum) {is=title K2Node_Event}
   > SetActiveWidgetIndex(self=Get Switcher_InputButtons(), Index=GetActiveWidgetIndex!() -> (ReturnValue:int=0))
   > InputBasedStyling()
   > Unpause Timer by Handle(Handle=Get AnimationDelayTimer())
   n23 = Event OnNotificationSet!() -> (OutputDelegate, then, Notification:Fort UINotification Object Reference) {is=title K2Node_Event}
   > UpdateData(Notification=@n23.Notification)
   > UpdateToast!(execute, self, Notification:Fort UINotification Object Reference=@n23.Notification) -> (then)
   > SetupAnimBindings(Unbind=false)
   n24 = Event OnNotificationUpdated!() -> (OutputDelegate, then, Notification:Fort UINotification Object Reference) {is=title K2Node_Event}
   > UpdateData(Notification=@n24.Notification)
   > UpdateToast(Notification=@n24.Notification)
  ]}
 GetActiveWidgetIndex {nodes [
   GetActiveWidgetIndex() {is=title K2Node_FunctionEntry}
   n1 = NOT Boolean(A=IsUsingTouch(PlayerController=GetOwningPlayer()))
   n2 = GetIsInGame!(self:Fort Social Notification Widget Object Reference) -> (ReturnValue:bool)
   > Return Node(ReturnValue:int=SelectInt(A=0, B=1, bPickA=AND Boolean(A=@n2, B=@n1)))
  ]}
 GetInputSwitcherVisibility {nodes [
   GetInputSwitcherVisibility() {is=title K2Node_FunctionEntry}
   n1 = Get CachedNotification()
   > Return Node(ReturnValue:ESlateVisibility Enum=Select(Option 0:ESlateVisibility Enum=Collapsed, Option 1:ESlateVisibility Enum=SelfHitTestInvisible, Index=Get bHasAction(self=@n1)) -> (ReturnValue:ESlateVisibility Enum))
  ]}
 HandleNotificationType {nodes [
   HandleNotificationType() {is=title K2Node_FunctionEntry}
   > UpdateIcons()
   > n1 = Get!(execute:Exec) -> (then:Exec, else:Exec, CachedNotification:Fort UINotification Object Reference)
   n2 = Get CustomActionLabel!(self:Fort UINotification Object Reference=@n1) -> (CustomActionLabel:Text)
   > n3 = Branch(Condition=TextIsEmpty(InText=@n2))
   > n4 = Switch on EFortNotificationType!(
    execute
    Selection:EFortNotificationType Enum=Get NotificationType!(self:Fort UINotification Object Reference=@n1) -> (NotificationType:EFortNotificationType Enum=Default)
   ) -> (
    Default:Exec
    Power:Exec
    HealthWarning:Exec
    Error:Exec
    GiftSent:Exec
    VoiceChannel:Exec
    DonutChallenge:Exec
    SocialNotification:Exec
    BattlePassPageUnlock:Exec
    DownloadOnDemandError:Exec
    DownloadCompleted:Exec
    CompeteStart:Exec
    Playtime:Exec
    EpicPartySocialNotification:Exec
    AgeVerification:Exec
    Max:Exec
   ) {is=title K2Node_SwitchEnum}
   n5 = MakeLiteralText(Value="LOCTEXT(\"GO!\")")
   n6 = Get CommonInputSubsystem()
   n7 = MakeLiteralText(Value="LOCTEXT(\"HOLD\")")
   Return Node!(
    execute=@n4.CompeteStart
    Return:Text=Select!(MouseAndKeyboard:Text=@n5, Gamepad:Text=@n7, Touch:Text=@n5, Count:Text=@n7, Index:ECommonInputType Enum=GetCurrentInputType(self=@n6)) -> (ReturnValue:Text)
   )
   n8 = Cast To FortUIFriendNotification!(execute=[@n4.EpicPartySocialNotification, @n4.SocialNotification], Object=Get CachedNotification()) -> (then, CastFailed, AsFort UIFriend Notification:Fort UIFriend Notification Object Reference) {is=title K2Node_DynamicCast}
   n9 = MakeLiteralText(Value="LOCTEXT(\"ACCEPT\")")
   n10 = MakeLiteralText(Value="LOCTEXT(\"OPEN\")")
   n11 = Select!(
    Default:Text=@n10
    FriendRequest:Text=@n9
    PartyInviteSent:Text
    PartyInvite:Text=@n9
    InviteFriendToParty:Text=MakeLiteralText(Value="LOCTEXT(\"INVITE\")")
    PartyRequest:Text=@n9
    AutoImportFriendSuggestion:Text=MakeLiteralText(Value="LOCTEXT(\"ENABLE\")")
    PartyMemberCreated:Text=@n10
    PartyMemberLeft:Text
    PartyJoin:Text
    EpicPartyInvite:Text=@n9
    EpicPartyInviteSent:Text
    EpicPartyMemberJoined:Text
    EpicPartyMemberLeft:Text
    EpicPartyJoined:Text
    InviteFriendToEpicParty:Text
    EpicPartyJoin:Text
    GroupInvite:Text=MakeLiteralText(Value="LOCTEXT(\"VIEW\")")
    EpicPartyJoinabilityChanged:Text
    EpicPartyJoinRequestSent:Text
    EpicPartyJoinRequestReceived:Text
    EpicPartyJoinRequestAccepted:Text
    EpicPartyPromotedToLeader:Text
    PartyInvite1v1:Text=@n9
    PartyInvite1v1Sent:Text
    PartyMemberSittingOut:Text
    Index:EFortUIFriendNotificationType Enum=Get FriendActionType!(self:Fort UIFriend Notification Object Reference=@n8) -> (FriendActionType:EFortUIFriendNotificationType Enum=Default)
   ) -> (ReturnValue)
   n12 = GetCurrentInputType(self=Get CommonInputSubsystem())
   n13 = MakeLiteralText(Value="LOCTEXT(\"HOLD\")")
   n14 = Select!(Option 0:Text=@n11, Option 1:Text=@n13, Index:bool=GetIsInGame()) -> (ReturnValue)
   > Return Node(Return=Select(
     Option 0=Select!(MouseAndKeyboard:Text=@n14, Gamepad:Text=@n13, Touch:Text=@n11, Count:Text=@n13, Index:ECommonInputType Enum=@n12) -> (ReturnValue)
     Index=TextIsEmpty(InText=@n11)
    ))
   n15 = Cast To FortUIPlaytimeNotification!(execute=@n4.Playtime, Object=Get CachedNotification()) -> (then, CastFailed, AsFort UIPlaytime Notification:Fort UIPlaytime Notification Object Reference) {is=title K2Node_DynamicCast}
   n16 = GetCurrentInputType(self=Get CommonInputSubsystem())
   n17 = MakeLiteralText(Value="LOCTEXT(\"HOLD\")")
   n18 = MakeLiteralText(Value="LOCTEXT(\"VIEW\")")
   n19 = Select!(Option 0:Text=@n18, Option 1:Text=@n17, Index:bool=GetIsInGame()) -> (ReturnValue)
   > Return Node(Return=Select!(
     Default:Text
     Power:Text
     HealthWarning:Text
     Error:Text
     GiftSent:Text
     VoiceChannel:Text
     DonutChallenge:Text
     SocialNotification:Text
     BattlePassPageUnlock:Text
     DownloadOnDemandError:Text
     DownloadCompleted:Text
     CompeteStart:Text
     Playtime:Text=Select!(MouseAndKeyboard:Text=@n19, Gamepad:Text=@n17, Touch:Text=@n18, Count:Text=@n17, Index:ECommonInputType Enum=@n16) -> (ReturnValue)
     EpicPartySocialNotification:Text
     AgeVerification:Text
     Max:Text
     Index:EFortNotificationType Enum=Get NotificationType(self=@n15) -> (NotificationType=Default)
    ) -> (ReturnValue))
   Return Node(execute=@n3.else, Return=@n2)
  ]}
 InputBasedStyling {nodes [
   InputBasedStyling() {is=title K2Node_FunctionEntry}
   > Set L_IsInGame(L_IsInGame:bool=GetIsInGame()) -> (Output_Get:bool)
   > Set L_InputType(L_InputType:ECommonInputType Enum=GetCurrentInputType(self=Get CommonInputSubsystem())) -> (Output_Get:ECommonInputType Enum)
   > n1 = HandleNotificationType!(execute, self) -> (then, Return:Text) {is=title K2Node_CallFunction}
   > "SetText (Text)"(self=Get Text_Action(), InText=@n1)
   n2 = Get L_IsInGame() -> (L_IsInGame:bool)
   > SetRenderTranslation!(
    execute
    self=Get Text_Action()
    Translation_X:double=Select(
     MouseAndKeyboard:double=Select!(Option 0:double=0, Option 1:double=-6, Index:bool=@n2) -> (ReturnValue)
     Gamepad:double=-6
     Touch:double=0
     Count:double=-6
     Index=Get L_InputType() -> (L_InputType:ECommonInputType Enum)
    ) -> (ReturnValue:double)
    Translation_Y:double=0
   ) -> (then)
   n3 = Get L_IsInGame()
   > SetVisibility(
    self=Get InputKeybind()
    InVisibility=Select!(
     MouseAndKeyboard:ESlateVisibility Enum=Select!(Option 0:ESlateVisibility Enum=Collapsed, Option 1:ESlateVisibility Enum=SelfHitTestInvisible, Index:bool=@n3) -> (ReturnValue)
     Gamepad:ESlateVisibility Enum=SelfHitTestInvisible
     Touch:ESlateVisibility Enum=Collapsed
     Count:ESlateVisibility Enum=SelfHitTestInvisible
     Index:ECommonInputType Enum=Get L_InputType()
    ) -> (ReturnValue:ESlateVisibility Enum)
   )
  ]}
 OnFinishedToast {nodes [OnFinishedToast() {is=title K2Node_FunctionEntry}]}
 PlayNotificationSound {nodes [
   PlayNotificationSound() {is=title K2Node_FunctionEntry}
   > n1 = Sequence()
   > n2 = Cast To FriendNotification!(execute, Object=Get CachedNotification()) -> (then, CastFailed, AsFriend Notification:Friend Notification Object Reference) {is=title K2Node_DynamicCast}
   > n3 = Switch on EFortUIFriendNotificationType!(execute, Selection:EFortUIFriendNotificationType Enum=Get FriendActionType(self=@n2) -> (FriendActionType=Default)) -> (
    Default:Exec
    FriendRequest:Exec
    PartyInviteSent:Exec
    PartyInvite:Exec
    InviteFriendToParty:Exec
    PartyRequest:Exec
    AutoImportFriendSuggestion:Exec
    PartyMemberCreated:Exec
    PartyMemberLeft:Exec
    PartyJoin:Exec
    EpicPartyInvite:Exec
    EpicPartyInviteSent:Exec
    EpicPartyMemberJoined:Exec
    EpicPartyMemberLeft:Exec
    EpicPartyJoined:Exec
    InviteFriendToEpicParty:Exec
    EpicPartyJoin:Exec
    GroupInvite:Exec
    EpicPartyJoinabilityChanged:Exec
    EpicPartyJoinRequestSent:Exec
    EpicPartyJoinRequestReceived:Exec
    EpicPartyJoinRequestAccepted:Exec
    EpicPartyPromotedToLeader:Exec
    PartyInvite1v1:Exec
    PartyInvite1v1Sent:Exec
    PartyMemberSittingOut:Exec
   ) {is=title K2Node_SwitchEnum}
   PlaySound2D(
    execute=[
     @n3.EpicPartyInvite
     @n3.EpicPartyJoin
     @n3.EpicPartyJoinRequestAccepted
     @n3.EpicPartyJoinRequestReceived
     @n3.EpicPartyJoined
     @n3.EpicPartyMemberJoined
    ]
    Sound=/Game/Athena/Sounds/Notifications/MSS_UI_Notification_Friend_Received
   )
   > Return Node!(execute)
   PlaySound2D(
    execute=[@n3.EpicPartyInviteSent, @n3.EpicPartyJoinRequestSent, @n3.InviteFriendToEpicParty]
    Sound=/Game/Athena/Sounds/Notifications/MSS_UI_Notification_Sent
   )
   > Return Node()
   PlaySound2D(execute=@n3.EpicPartyJoinabilityChanged, Sound=/Game/Athena/Sounds/Notifications/MSS_UI_Notification_Joinability_Changed)
   > Return Node()
   PlaySound2D(execute=@n3.EpicPartyMemberLeft, Sound=/Game/Athena/Sounds/Notifications/MSS_UI_Notification_Friend_Left)
   > Return Node()
   PlaySound2D(execute=@n1.then_1, Sound=/Game/Sounds/Fort_UI_Sounds/FrontEnd/Fort_UI_Party_Invited_Cue)
  ]}
 ResetMaterials {nodes [
   ResetMaterials() {is=title K2Node_FunctionEntry}
   > n1 = Sequence!(execute) -> (then_0, then_1, then_2:Exec, then_3:Exec)
   > SetVectorParameterValue(self=Get TrimMat(), ParameterName=Trim_Color, Value=(R=0.0024,G=0.0612,B=0.5029,A=1))
   > SetScalarParameterValue(self=Get TrimMat(), ParameterName=HideProgressBar, Value=0)
   > SetScalarParameterValue(self=Get TrimMat(), ParameterName=LerpAngleChange, Value=0)
   > SetScalarParameterValue(self=Get TrimMat(), ParameterName=LerpToGreen, Value=0)
   > SetScalarParameterValue(self=Get TrimMat(), ParameterName=LerpToRed, Value=0)
   > SetScalarParameterValue(self=Get TrimMat(), ParameterName=Angle, Value=0)
   > SetScalarParameterValue(self=Get TrimMat(), ParameterName=LeftOffset, Value=0.985)
   > SetScalarParameterValue(self=Get TrimMat(), ParameterName=RightOffset, Value=0)
   > SetScalarParameterValue(self=Get TrimMat(), ParameterName=Sharpness, Value=0.985)
   > SetScalarParameterValue(self=Get TrimMat(), ParameterName=Burst, Value=0)
   SetScalarParameterValue(
    execute=@n1.then_1
    self=Get BackgroundMat() -> (BackgroundMat:Material Instance Dynamic Object Reference)
    ParameterName=HideProgressBar
    Value=0
   )
   > SetScalarParameterValue(self=Get BackgroundMat(), ParameterName=Progress, Value=0)
   > SetScalarParameterValue(self=Get BackgroundMat(), ParameterName=LerpToGreen, Value=0)
   > SetScalarParameterValue(self=Get BackgroundMat(), ParameterName=LerpToRed, Value=0)
   > SetScalarParameterValue(self=Get BackgroundMat(), ParameterName=Burst, Value=0)
   SetRenderOpacity(execute=@n1.then_2, self=Get Overlay_ResultIcon() -> (Overlay_ResultIcon:Overlay Object Reference), InOpacity=0)
   > SetVisibility(self=Get InputKeybind(), InVisibility=SelfHitTestInvisible)
   > SetRenderOpacity(self=Get InputKeybind(), InOpacity=1)
   > SetRenderScale!(execute, self:Widget Object Reference=Get InputKeybind(), Scale:Vector 2D Structure=(X=1,Y=1)) -> (then)
   > n2 = GetDynamicMaterial(self=Get Image_ProgressCircle() -> (Image_ProgressCircle:Image Object Reference))
   > SetScalarParameterValue(self=@n2, ParameterName=Progress, Value=0)
   > SetRenderOpacity(self=Get Text_Action(), InOpacity=1)
   > SetColorAndOpacity(self=Get Text_Action(), InColorAndOpacity_SpecifiedColor=(R=1,G=1,B=1,A=1))
   SetBrushResourceObject!(
    execute=@n1.then_3
    self:Image Object Reference=Get Image_Lazy() -> (Image_Lazy:Fort Lazy Image Object Reference)
    ResourceObject:Object Reference
   ) -> (then)
   > SetVisibility(
    self=[Get SidebarSocialTagList() -> (SidebarSocialTagList:Sidebar Social Tag List Base Object Reference), Get Image_Lazy()]
    InVisibility=Collapsed
   )
   > SetVisibility(
    self=Select(
     Option 0:Widget Object Reference=Get Avatar_MemberIcon()
     Option 1:Widget Object Reference=Get PlayerAvatar()
     Index=Get UsingNewAvatars() -> (UsingNewAvatars:bool)
    ) -> (ReturnValue:Widget Object Reference)
    InVisibility=SelfHitTestInvisible
   )
  ]}
 Sequencer Events {nodes [
   HoldComplete() {is=title K2Node_CustomEvent}
   > StopAnimation(InAnimation=Get Anim_TimeToRespond())
   > SetScalarParameterValue(self=Get BackgroundMat(), ParameterName=Progress, Value=1)
  ]}
 SetDescriptionTextWrapValue {nodes [
   SetDescriptionTextWrapValue() {is=title K2Node_FunctionEntry}
   n1 = GetLocalSize!(Geometry:"Geometry Structure (by ref)"=GetPaintSpaceGeometry!(self:Widget Object Reference=Get VerticalBox_Info() -> (VerticalBox_Info:Vertical Box Object Reference)) -> (ReturnValue:"Geometry Structure (by ref)")) -> (ReturnValue:Vector 2D Structure)
   > n2 = Branch(Condition=IsZero2D!(A:"Vector 2D Structure (by ref)"=@n1) -> (ReturnValue:bool))
   > Return Node!(execute, IsSizeEmpty:bool=true)
   n3 = BreakVector2D(InVec=@n1)
   n4 = Get DefaultWrapTextValue() -> (DefaultWrapTextValue:float)
   Set WrapTextAt!(
    execute=@n2.else
    WrapTextAt:float=Select(Option 0:float=Get DefaultWrapTextValue(), Option 1:float=@n3, Index="float > float"(A=@n3, B=@n4)) -> (ReturnValue:float)
    self:Text Layout Widget Object Reference=Get RichText_Description() -> (RichText_Description:Common Rich Text Block Object Reference)
   ) -> (then, Output_Get:float)
   > Return Node(IsSizeEmpty=false)
  ]}
 SetupAnimBindings {nodes [
   SetupAnimBindings!() -> (then, Unbind:bool=~) {is=title K2Node_FunctionEntry}
   > Get()
   > n1 = Branch(Condition=Get Unbind() -> (Unbind:bool))
   > UnbindFromAnimationFinished!(
    execute
    self:Widget Animation Object Reference=DetermineIntroAnim()
    Widget:User Widget Object Reference=Self-Reference()
    Delegate:Delegate=Create Event()
   ) -> (then)
   > UnbindFromAnimationFinished(self=Get Anim_Outro(), Widget=Self-Reference(), Delegate=Create Event())
   n2 = Create Event()
   > UnbindFromAnimationStarted!(
    execute
    self:User Widget Object Reference=Self-Reference()
    Animation:Widget Animation Object Reference=DetermineIntroAnim()
    Delegate:Delegate=@n2
   ) -> (then)
   BindToAnimationFinished!(
    execute=@n1.else
    self:Widget Animation Object Reference=DetermineIntroAnim()
    Widget:User Widget Object Reference=Self-Reference()
    Delegate=Create Event()
   ) -> (then)
   > BindToAnimationFinished(self=Get Anim_Outro(), Widget=Self-Reference(), Delegate=Create Event())
   > BindToAnimationStarted!(
    execute
    self:User Widget Object Reference=Self-Reference()
    Animation:Widget Animation Object Reference=DetermineIntroAnim()
    Delegate:Delegate=@n2
   ) -> (then)
  ]}
 TakeAction {nodes [
   TakeAction() {is=title K2Node_FunctionEntry}
   > Branch(Condition=Get bHasAction(self=Get CachedNotification()))
   > n1 = Branch(Condition=Get IsActionTaken())
   Set IsActionTaken(execute=@n1.else, IsActionTaken=true)
   > TakeAction(self:Fort UINotification Object Reference=Get CachedNotification()) {is=title K2Node_CallFunction}
  ]}
 UpdateData {nodes [
   n1 = UpdateData!() -> (then, Notification:Fort UINotification Object Reference) {is=title K2Node_FunctionEntry}
   n2 = Get Data!(self:BP Toast Notification Asset Object Reference=Get ToastProperty() -> (ToastProperty:BP Toast Notification Asset Object Reference)) -> (Data:Map of EFortNotificationType Enums to S UI Toast Property Structures=Default)
   n3 = Find(TargetMap:Map of EFortNotificationType Enums to S UI Toast Property Structures=@n2, Key:"EFortNotificationType Enum (by ref)"=Default) -> (Value:S UI Toast Property Structure)
   n4 = Find(TargetMap=@n2, Key=Get NotificationType(self=@n1) -> (NotificationType=Default))
   > Branch(Condition=OR Boolean(A=@n3.ReturnValue, B=@n4.ReturnValue))
   > Set Data(Data:S UI Toast Property Structure=Select(Option 0:S UI Toast Property Structure=@n3, Option 1:S UI Toast Property Structure=@n4, Index=@n4.ReturnValue) -> (ReturnValue:S UI Toast Property Structure)) -> (Output_Get:S UI Toast Property Structure)
  ]}
 UpdateIcons {nodes [
   UpdateIcons() {is=title K2Node_FunctionEntry}
   > n1 = Sequence!(execute) -> (then_0, then_1, then_2)
   > n2 = Get()
   n3 = Get DisplayAsset!(self:Fort UINotification Object Reference=@n2) -> (DisplayAsset:Soft Object Reference)
   > n4 = Branch(Condition=IsValidSoftObjectReference(SoftObjectReference=@n3))
   > Set Local_IconDisplayAsset(Local_IconDisplayAsset:Soft Object Reference=@n3) -> (Output_Get:Soft Object Reference)
   n5 = Switch on EFortNotificationType(execute=@n4.else, Selection=Get NotificationType(self=Get CachedNotification()) -> (NotificationType=Default))
   Set Local_IconDisplayAsset(execute=@n5.CompeteStart, Local_IconDisplayAsset=Get CompeteIcon() -> (CompeteIcon:Soft Object Reference))
   > Set bLocal_ShowIcon(bLocal_ShowIcon:bool=true) -> (Output_Get:bool)
   Set Local_IconDisplayAsset(
    execute=@n5.DownloadOnDemandError
    Local_IconDisplayAsset=Get DownloadOnDemandErrorIcon() -> (DownloadOnDemandErrorIcon:Soft Object Reference)
   )
   Set Local_IconDisplayAsset(execute=@n5.Playtime, Local_IconDisplayAsset=Get PlaytimeIcon() -> (PlaytimeIcon:Soft Object Reference))
   > Set bLocal_ShowIcon(bLocal_ShowIcon=true)
   Set bLocal_ShowAvatar(execute=[@n5.Default, @n5.DownloadCompleted, @n5.Error, @n5.HealthWarning, @n5.Power], bLocal_ShowAvatar:bool=false) -> (Output_Get:bool)
   n6 = Set bLocal_ShowIcon(
    execute=@n1.then_1
    bLocal_ShowIcon=IsValidSoftObjectReference(SoftObjectReference=Get Local_IconDisplayAsset() -> (Local_IconDisplayAsset:Soft Object Reference))
   )
   > Branch(Condition=@n6)
   > Set bLocal_ShowAvatar(bLocal_ShowAvatar=false)
   > SetBrushFromLazyDisplayAsset!(
    execute
    self:Common Lazy Image Object Reference=Get Image_Lazy()
    LazyObject:"Soft Object Reference (by ref)"=Get Local_IconDisplayAsset()
    bMatchTextureSize:bool=false
   ) -> (then)
   SetVisibility(
    execute=@n1.then_2
    self=Get Image_Lazy()
    InVisibility=Select(
     Option 0:ESlateVisibility Enum=Collapsed
     Option 1:ESlateVisibility Enum=SelfHitTestInvisible
     Index=Get bLocal_ShowIcon() -> (bLocal_ShowIcon:bool)
    ) -> (ReturnValue:ESlateVisibility Enum)
   )
   > n7 = Branch(Condition=Get bLocal_ShowAvatar() -> (bLocal_ShowAvatar:bool))
   > SetVisibility(
    self=Select(Option 0:Widget Object Reference=Get Avatar_MemberIcon(), Option 1:Widget Object Reference=Get PlayerAvatar(), Index=Get UsingNewAvatars()) -> (ReturnValue:Widget Object Reference)
    InVisibility=SelfHitTestInvisible
   )
   SetVisibility(execute=@n7.else, self=[Get PlayerAvatar(), Get Avatar_MemberIcon()], InVisibility=Collapsed)
  ]}
 UpdateToast {nodes [
   UpdateToast!() -> (then, Notification:Fort UINotification Object Reference) {is=title K2Node_FunctionEntry}
   > SetText(
    self=Get RichText_Title() -> (RichText_Title:Common Rich Text Block Object Reference)
    InText=Get Title!(self:Fort UINotification Object Reference=Get Notification() -> (Notification:Fort UINotification Object Reference)) -> (Title:Text)
   )
   > SetText(self=Get RichText_Description(), InText=Get Description!(self:Fort UINotification Object Reference=Get Notification()) -> (Description:Text))
   > SetVisibility(
    self=Get Switcher_InputButtons()
    InVisibility=GetInputSwitcherVisibility!() -> (ReturnValue:ESlateVisibility Enum=Visible) {is=title K2Node_CallFunction}
   )
   > SetActiveWidgetIndex(self=Get Switcher_InputButtons(), Index=GetActiveWidgetIndex() -> (ReturnValue=0) {is=title K2Node_CallFunction})
   > InputBasedStyling() {is=title K2Node_CallFunction}
  ]}
]
variables [
 AnimationDelayTimer {type=TimerHandle}
 BackgroundMat {type=MaterialInstanceDynamic}
 ButtonClickSound {type=SoundBase, default=/Game/Sounds/Fort_UI_Sounds/FrontEnd/Social/UI_Social_Toast_Accept_Click_Cue}
 CompeteIcon {type=Object, default=/Game/UI/Foundation/Toasts/Materials/MI_UI_ToastIcon_Compete}
 Data {type=S_UI_ToastProperty, default {displayTimeOverride=0, introAnimationType=Neutral, useComplexAnim=false}}
 DownloadOnDemandErrorIcon {type=Object, default=/Game/UI/Foundation/Toasts/Materials/T_UI_icon_downloadmessage}
 IsActionTaken {type=boolean}
 OnFinishedToast {}
 PlaytimeIcon {type=Object, default=/FortUILibrary/Textures/MaterialInstances/MI_UIKit_Icon_UpdatePending}
 ToastProperty {type=BP_ToastNotificationAsset_C, default=/Game/UI/Foundation/Toasts/Data/DA_UI_ToastNotificationPoperty}
 TrimMat {type=MaterialInstanceDynamic}
 UsingNewAvatars {type=boolean}
]
widgets {
 info {inheritedWidgetCount=4, parentClass=/Script/FortniteUI.FortSocialNotificationWidget, rootWidgetClass=&GridPanel, widgetCount=32}
 widgets [
  Root {widgetClassPath=&GridPanel, widget=Root}
  Image_ToastCushion {
   widgetClassPath=&Image
   bIsVariable=true
   parent=Root
   widget=Image_ToastCushion
   properties {
    visibility=SelfHitTestInvisible
    brush {
     bIsDynamicallyLoaded=false
     drawAs=Image
     imageType=NoImage
     mirroring=NoMirror
     resourceName=None
     resourceObject=/Game/UI/Foundation/Toasts/Materials/M_UI_ToastCushion
     tiling=NoTile
     imageSize=<258 104>
     margin=<ltrb 0 0 0 0>
     outlineSettings {
      bUseBrushTransparency=false
      roundingType=HalfHeightRadius
      width=0
      color {colorUseRule=UseColor_Specified, specifiedColor=<rgba 0 0 0 0>}
      cornerRadii=<0 0 0 0>
     }
     tintColor {colorUseRule=UseColor_Specified, specifiedColor=<rgba 1 1 1 1>}
     uVRegion {bIsValid=false, max=<0 0>, min=<0 0>}
    }
   }
   slot {class=GridSlot, properties {horizontalAlignment=HAlign_Left, nudge=<-120 0>}}
  }
  Image_BackgroundMat {
   widgetClassPath=&Image
   bIsVariable=true
   parent=Root
   widget=Image_BackgroundMat
   properties {
    visibility=SelfHitTestInvisible
    brush {
     bIsDynamicallyLoaded=false
     drawAs=Image
     imageType=NoImage
     mirroring=NoMirror
     resourceName=None
     resourceObject=/Game/UI/Foundation/Toasts/Materials/I_UI_ToastBackground
     tiling=NoTile
     imageSize=<32 32>
     margin=<ltrb 0 0 0 0>
     outlineSettings {
      bUseBrushTransparency=false
      roundingType=HalfHeightRadius
      width=0
      color {colorUseRule=UseColor_Specified, specifiedColor=<rgba 0 0 0 0>}
      cornerRadii=<0 0 0 0>
     }
     tintColor {colorUseRule=UseColor_Specified, specifiedColor=<rgba 1 1 1 1>}
     uVRegion {bIsValid=false, max=<0 0>, min=<0 0>}
    }
    renderTransform {angle=0, scale=<1 1>, shear=<-10 0>, translation=<0 0>}
    renderTransformPivot=<0 0.5>
   }
   slot {class=GridSlot, properties {padding=<ltrb 77 0 0 0>}}
  }
  Content {widgetClassPath=&HorizontalBox, parent=Root, widget=Content, slot {class=GridSlot}}
  Spacer_1 {widgetClassPath=&Spacer, parent=Content, widget=Spacer_1, properties {size=<96 1>}, slot {class=HorizontalBoxSlot}}
  Icon {
   widgetClassPath=&Overlay
   parent=Content
   widget=Icon
   slot {class=HorizontalBoxSlot, properties {horizontalAlignment=HAlign_Left, verticalAlignment=VAlign_Center}}
  }
  Image_TextureIcon {
   widgetClassPath=&Image
   bIsVariable=true
   parent=Icon
   widget=Image_TextureIcon
   properties {
    visibility=Collapsed
    brush {
     bIsDynamicallyLoaded=false
     drawAs=Image
     imageType=NoImage
     mirroring=NoMirror
     resourceName=None
     resourceObject=&T_UI_icon_downloadmessage
     tiling=NoTile
     imageSize=<68 68>
     margin=<ltrb 0 0 0 0>
     outlineSettings {
      bUseBrushTransparency=false
      roundingType=HalfHeightRadius
      width=0
      color {colorUseRule=UseColor_Specified, specifiedColor=<rgba 0 0 0 0>}
      cornerRadii=<0 0 0 0>
     }
     tintColor {colorUseRule=UseColor_Specified, specifiedColor=<rgba 1 1 1 1>}
     uVRegion {bIsValid=false, max=<0 0>, min=<0 0>}
    }
   }
   slot {class=OverlaySlot, properties {horizontalAlignment=HAlign_Fill, verticalAlignment=VAlign_Fill}}
  }
  Image_Lazy {
   widgetClassPath=&FortLazyImage
   bIsVariable=true
   parent=Icon
   widget=Image_Lazy
   properties {brush {
     bIsDynamicallyLoaded=false
     drawAs=Image
     imageType=NoImage
     mirroring=NoMirror
     resourceName=None
     resourceObject=null
     tiling=NoTile
     imageSize=<68 68>
     margin=<ltrb 0 0 0 0>
     outlineSettings {
      bUseBrushTransparency=false
      roundingType=HalfHeightRadius
      width=0
      color {colorUseRule=UseColor_Specified, specifiedColor=<rgba 0 0 0 0>}
      cornerRadii=<0 0 0 0>
     }
     tintColor {colorUseRule=UseColor_Specified, specifiedColor=<rgba 1 1 1 1>}
     uVRegion {bIsValid=false, max=<0 0>, min=<0 0>}
    }}
   slot {class=OverlaySlot, properties {horizontalAlignment=HAlign_Fill, verticalAlignment=VAlign_Fill}}
  }
  Avatar_MemberIcon {
   widgetClassPath=/Script/FortniteUI.FortSocialAvatarIcon
   bInherited=true
   bIsVariable=true
   parent=Icon
   widget=Avatar_MemberIcon
   properties {
    avatarMaterial=/Game/UI/SocialPanel/Art/Materials/UI_UserPortrait_Buffet
    materialTextureParamName=SmallPreview
    brush {
     bIsDynamicallyLoaded=false
     drawAs=Image
     imageType=NoImage
     mirroring=NoMirror
     resourceName=None
     resourceObject=&UI_UserPortrait_Buffet
     tiling=NoTile
     imageSize=<68 68>
     margin=<ltrb 0 0 0 0>
     outlineSettings {
      bUseBrushTransparency=false
      roundingType=HalfHeightRadius
      width=0
      color {colorUseRule=UseColor_Specified, specifiedColor=<rgba 0 0 0 0>}
      cornerRadii=<0 0 0 0>
     }
     tintColor {colorUseRule=UseColor_Specified, specifiedColor=<rgba 1 1 1 1>}
     uVRegion {bIsValid=false, max=<0 0>, min=<0 0>}
    }
    loadingBackgroundBrush {
     bIsDynamicallyLoaded=false
     drawAs=Image
     imageType=NoImage
     mirroring=NoMirror
     resourceName=None
     resourceObject=/FortUILibrary/UIKit/Throbber/MI_UIKit_Monochrome_Throbber
     tiling=NoTile
     imageSize=<32 32>
     margin=<ltrb 0 0 0 0>
     outlineSettings {
      bUseBrushTransparency=false
      roundingType=HalfHeightRadius
      width=0
      color {colorUseRule=UseColor_Specified, specifiedColor=<rgba 0 0 0 0>}
      cornerRadii=<0 0 0 0>
     }
     tintColor {colorUseRule=UseColor_Specified, specifiedColor=<rgba 1 1 1 1>}
     uVRegion {bIsValid=false, max=<0 0>, min=<0 0>}
    }
   }
   slot {class=OverlaySlot, properties {horizontalAlignment=HAlign_Fill, verticalAlignment=VAlign_Fill}}
  }
  PlayerAvatar {
   widgetClassPath=/Game/UI/Foundation/Profile/WBP_PlayerAvatar
   bInherited=true
   bIsVariable=true
   parent=Icon
   widget=PlayerAvatar
   properties {showOnlinePresence=false}
   slot {class=OverlaySlot, properties {horizontalAlignment=HAlign_Fill, verticalAlignment=VAlign_Fill}}
  }
  Spacer {widgetClassPath=&Spacer, parent=Content, widget=Spacer, properties {size=<12 1>}, slot {class=HorizontalBoxSlot}}
  VerticalBox_Info {
   widgetClassPath=&VerticalBox
   bIsVariable=true
   parent=Content
   widget=VerticalBox_Info
   slot {class=HorizontalBoxSlot, properties {verticalAlignment=VAlign_Center, padding=<ltrb 0 12 0 12>}}
  }
  RichText_Title {
   widgetClassPath=&CommonRichTextBlock
   bIsVariable=true
   parent=VerticalBox_Info
   widget=RichText_Title
   properties {
    bAutoCollapseWithEmptyText=true
    bIsScrollingEnabled=false
    bOverrideDefaultStyle=true
    lineHeightPercentage=0.9
    mobileTextBlockScale=1.1
    text=PARTY INVITE
    textStyleSet=&/ToastTextStyleTableTitle
    textTransformPolicy=ToUpper
    defaultTextStyleOverride {
     overflowPolicy=Clip
     transformPolicy=None
     colorAndOpacity {colorUseRule=UseColor_Specified, specifiedColor=<rgba 0.004 0.004 0.004 1>}
     font {
      bForceMonospaced=false
      bMaterialIsStencil=false
      fontMaterial=null
      fontObject=&BurbankSmall
      letterSpacing=-75
      monospacedWidth=1
      size=20
      skewAmount=0
      typefaceFontName=Black
      outlineSettings {
       bApplyOutlineToDropShadows=false
       bMiteredCorners=false
       bSeparateFillAlpha=false
       outlineMaterial=null
       outlineSize=0
       outlineColor=<rgba 0 0 0 1>
      }
     }
     highlightColor {colorUseRule=UseColor_Specified, specifiedColor=<rgba 0 0 0 1>}
     highlightShape {
      bIsDynamicallyLoaded=false
      drawAs=Image
      imageType=NoImage
      mirroring=NoMirror
      resourceName=None
      resourceObject=null
      tiling=NoTile
      imageSize=<32 32>
      margin=<ltrb 0 0 0 0>
      outlineSettings {
       bUseBrushTransparency=false
       roundingType=HalfHeightRadius
       width=0
       color {colorUseRule=UseColor_Specified, specifiedColor=<rgba 0 0 0 0>}
       cornerRadii=<0 0 0 0>
      }
      tintColor {colorUseRule=UseColor_Specified, specifiedColor=<rgba 1 1 1 1>}
      uVRegion {bIsValid=false, max=<0 0>, min=<0 0>}
     }
     selectedBackgroundColor {colorUseRule=UseColor_Foreground, specifiedColor=<rgba 1 0 1 1>}
     shadowColorAndOpacity=<rgba 0 0 0 1>
     shadowOffset=<0 0>
     strikeBrush {
      bIsDynamicallyLoaded=false
      drawAs=Image
      imageType=NoImage
      mirroring=NoMirror
      resourceName=None
      resourceObject=null
      tiling=NoTile
      imageSize=<32 32>
      margin=<ltrb 0 0 0 0>
      outlineSettings {
       bUseBrushTransparency=false
       roundingType=HalfHeightRadius
       width=0
       color {colorUseRule=UseColor_Specified, specifiedColor=<rgba 0 0 0 0>}
       cornerRadii=<0 0 0 0>
      }
      tintColor {colorUseRule=UseColor_Specified, specifiedColor=<rgba 1 1 1 1>}
      uVRegion {bIsValid=false, max=<0 0>, min=<0 0>}
     }
     underlineBrush {
      bIsDynamicallyLoaded=false
      drawAs=Image
      imageType=NoImage
      mirroring=NoMirror
      resourceName=None
      resourceObject=null
      tiling=NoTile
      imageSize=<32 32>
      margin=<ltrb 0 0 0 0>
      outlineSettings {
       bUseBrushTransparency=false
       roundingType=HalfHeightRadius
       width=0
       color {colorUseRule=UseColor_Specified, specifiedColor=<rgba 0 0 0 0>}
       cornerRadii=<0 0 0 0>
      }
      tintColor {colorUseRule=UseColor_Specified, specifiedColor=<rgba 1 1 1 1>}
      uVRegion {bIsValid=false, max=<0 0>, min=<0 0>}
     }
    }
   }
   slot {class=VerticalBoxSlot, properties {verticalAlignment=VAlign_Center}}
  }
  HBox_Desc {
   widgetClassPath=&HorizontalBox
   parent=VerticalBox_Info
   widget=HBox_Desc
   slot {class=VerticalBoxSlot, properties {verticalAlignment=VAlign_Center, padding=<ltrb 0 0 0 4>}}
  }
  WBP_RebootRallyIcon {
   widgetClassPath=/Game/UI/Sidebar/RebootRally/WBP_RebootRallyIcon
   bIsVariable=true
   parent=HBox_Desc
   widget=WBP_RebootRallyIcon
   properties {visibility=Collapsed}
   slot {class=HorizontalBoxSlot, properties {horizontalAlignment=HAlign_Left, verticalAlignment=VAlign_Center, padding=<ltrb -3 -3 0 -3>}}
  }
  RichText_Description {
   widgetClassPath=&CommonRichTextBlock
   bIsVariable=true
   parent=HBox_Desc
   widget=RichText_Description
   properties {
    bAutoCollapseWithEmptyText=true
    bIsScrollingEnabled=false
    bOverrideDefaultStyle=true
    lineHeightPercentage=0.9
    mobileTextBlockScale=1.2
    text="<Username>Thom.Yorke</> invited you to a party"
    textStyleSet=&/ToastTextStyleTableDescription
    wrapTextAt=545
    defaultTextStyleOverride {
     overflowPolicy=Clip
     transformPolicy=None
     colorAndOpacity {colorUseRule=UseColor_Specified, specifiedColor=<rgba 0.0578 0.0578 0.0578 1>}
     font {
      bForceMonospaced=false
      bMaterialIsStencil=false
      fontMaterial=null
      fontObject=&BurbankSmall
      letterSpacing=-75
      monospacedWidth=1
      size=16
      skewAmount=0
      typefaceFontName=Bold
      outlineSettings {
       bApplyOutlineToDropShadows=false
       bMiteredCorners=false
       bSeparateFillAlpha=false
       outlineMaterial=null
       outlineSize=0
       outlineColor=<rgba 0 0 0 1>
      }
     }
     highlightColor {colorUseRule=UseColor_Specified, specifiedColor=<rgba 0 0 0 1>}
     highlightShape {
      bIsDynamicallyLoaded=false
      drawAs=Image
      imageType=NoImage
      mirroring=NoMirror
      resourceName=None
      resourceObject=null
      tiling=NoTile
      imageSize=<32 32>
      margin=<ltrb 0 0 0 0>
      outlineSettings {
       bUseBrushTransparency=false
       roundingType=HalfHeightRadius
       width=0
       color {colorUseRule=UseColor_Specified, specifiedColor=<rgba 0 0 0 0>}
       cornerRadii=<0 0 0 0>
      }
      tintColor {colorUseRule=UseColor_Specified, specifiedColor=<rgba 1 1 1 1>}
      uVRegion {bIsValid=false, max=<0 0>, min=<0 0>}
     }
     selectedBackgroundColor {colorUseRule=UseColor_Foreground, specifiedColor=<rgba 1 0 1 1>}
     shadowColorAndOpacity=<rgba 0 0 0 1>
     shadowOffset=<0 0>
     strikeBrush {
      bIsDynamicallyLoaded=false
      drawAs=Image
      imageType=NoImage
      mirroring=NoMirror
      resourceName=None
      resourceObject=null
      tiling=NoTile
      imageSize=<32 32>
      margin=<ltrb 0 0 0 0>
      outlineSettings {
       bUseBrushTransparency=false
       roundingType=HalfHeightRadius
       width=0
       color {colorUseRule=UseColor_Specified, specifiedColor=<rgba 0 0 0 0>}
       cornerRadii=<0 0 0 0>
      }
      tintColor {colorUseRule=UseColor_Specified, specifiedColor=<rgba 1 1 1 1>}
      uVRegion {bIsValid=false, max=<0 0>, min=<0 0>}
     }
     underlineBrush {
      bIsDynamicallyLoaded=false
      drawAs=Image
      imageType=NoImage
      mirroring=NoMirror
      resourceName=None
      resourceObject=null
      tiling=NoTile
      imageSize=<32 32>
      margin=<ltrb 0 0 0 0>
      outlineSettings {
       bUseBrushTransparency=false
       roundingType=HalfHeightRadius
       width=0
       color {colorUseRule=UseColor_Specified, specifiedColor=<rgba 0 0 0 0>}
       cornerRadii=<0 0 0 0>
      }
      tintColor {colorUseRule=UseColor_Specified, specifiedColor=<rgba 1 1 1 1>}
      uVRegion {bIsValid=false, max=<0 0>, min=<0 0>}
     }
    }
   }
   slot {class=HorizontalBoxSlot, properties {verticalAlignment=VAlign_Center, size {sizeRule=Fill, value=1}}}
  }
  SidebarSocialTagList {
   widgetClassPath=/Game/UI/Sidebar/Tags/SidebarSocialTagList
   bInherited=true
   bIsVariable=true
   parent=VerticalBox_Info
   widget=SidebarSocialTagList
   properties {bAreTagsSelectable=false, visibility=Collapsed}
   slot {class=VerticalBoxSlot, properties {horizontalAlignment=HAlign_Left, padding=<ltrb 0 2 0 0>}}
  }
  Spacer_32x {widgetClassPath=&Spacer, parent=Content, widget=Spacer_32x, properties {size=<32 1>}, slot {class=HorizontalBoxSlot}}
  Spacer_HandleShear {widgetClassPath=&Spacer, parent=Content, widget=Spacer_HandleShear, properties {size=<13 1>}, slot {class=HorizontalBoxSlot}}
  Interaction {
   widgetClassPath=&GridPanel
   parent=Root
   widget=Interaction
   properties {columnFill [1], rowFill [1]}
   slot {class=GridSlot, properties {column=1, padding=<ltrb 0 0 10 0>}}
  }
  Image_Trim {
   widgetClassPath=&Image
   bIsVariable=true
   parent=Interaction
   widget=Image_Trim
   properties {
    visibility=SelfHitTestInvisible
    brush {
     bIsDynamicallyLoaded=false
     drawAs=Image
     imageType=NoImage
     mirroring=NoMirror
     resourceName=None
     resourceObject=/Game/UI/Foundation/Toasts/Materials/MI_UI_SkewBox
     tiling=NoTile
     imageSize=<128 32>
     margin=<ltrb 0 0 0 0>
     outlineSettings {
      bUseBrushTransparency=false
      roundingType=HalfHeightRadius
      width=0
      color {colorUseRule=UseColor_Specified, specifiedColor=<rgba 0 0 0 0>}
      cornerRadii=<0 0 0 0>
     }
     tintColor {colorUseRule=UseColor_Specified, specifiedColor=<rgba 1 1 1 1>}
     uVRegion {bIsValid=false, max=<0 0>, min=<0 0>}
    }
    renderTransform {angle=0, scale=<1 1>, shear=<-10 0>, translation=<0 0>}
    renderTransformPivot=<0 0.5>
   }
   slot {class=GridSlot}
  }
  Overlay_ResultIcon {
   widgetClassPath=&Overlay
   bIsVariable=true
   parent=Interaction
   widget=Overlay_ResultIcon
   properties {renderOpacity=0}
   slot {class=GridSlot}
  }
  Image_Success {
   widgetClassPath=&Image
   bIsVariable=true
   parent=Overlay_ResultIcon
   widget=Image_Success
   properties {
    visibility=SelfHitTestInvisible
    brush {
     bIsDynamicallyLoaded=false
     drawAs=Image
     imageType=NoImage
     mirroring=NoMirror
     resourceName=None
     resourceObject=/Game/UI/Foundation/Textures/Icons/Checks/T-Icon-Check-128
     tiling=NoTile
     imageSize=<64 64>
     margin=<ltrb 0 0 0 0>
     outlineSettings {
      bUseBrushTransparency=false
      roundingType=HalfHeightRadius
      width=0
      color {colorUseRule=UseColor_Specified, specifiedColor=<rgba 0 0 0 0>}
      cornerRadii=<0 0 0 0>
     }
     tintColor {colorUseRule=UseColor_Specified, specifiedColor=<rgba 1 1 1 1>}
     uVRegion {bIsValid=false, max=<0 0>, min=<0 0>}
    }
    colorAndOpacity=<rgba 0.0331 1 0.227 1>
   }
   slot {class=OverlaySlot, properties {horizontalAlignment=HAlign_Center, verticalAlignment=VAlign_Center}}
  }
  Image_Fail {
   widgetClassPath=&Image
   bIsVariable=true
   parent=Overlay_ResultIcon
   widget=Image_Fail
   properties {
    visibility=SelfHitTestInvisible
    brush {
     bIsDynamicallyLoaded=false
     drawAs=Image
     imageType=NoImage
     mirroring=NoMirror
     resourceName=None
     resourceObject=/Game/Athena/UI/InGame/Creative/PropertyEditor/Icon_Picker/Textures/64x64/T_UI_IconLibrary_exclamation_64
     tiling=NoTile
     imageSize=<64 64>
     margin=<ltrb 0 0 0 0>
     outlineSettings {
      bUseBrushTransparency=false
      roundingType=HalfHeightRadius
      width=0
      color {colorUseRule=UseColor_Specified, specifiedColor=<rgba 0 0 0 0>}
      cornerRadii=<0 0 0 0>
     }
     tintColor {colorUseRule=UseColor_Specified, specifiedColor=<rgba 1 1 1 1>}
     uVRegion {bIsValid=false, max=<0 0>, min=<0 0>}
    }
    colorAndOpacity=<rgba 0.8148 0.0307 0.0844 1>
    renderTransform {angle=0, scale=<1 1>, shear=<-13 0>, translation=<0 0>}
   }
   slot {class=OverlaySlot, properties {horizontalAlignment=HAlign_Center, verticalAlignment=VAlign_Center}}
  }
  VerticalBox_HoldAction {
   widgetClassPath=&VerticalBox
   bIsVariable=true
   parent=Interaction
   widget=VerticalBox_HoldAction
   properties {renderTransform {angle=0, scale=<1 1>, shear=<0 0>, translation=<3 0>}}
   slot {class=GridSlot, properties {verticalAlignment=VAlign_Center}}
  }
  InputKeybind {
   widgetClassPath=&Overlay
   bIsVariable=true
   parent=VerticalBox_HoldAction
   widget=InputKeybind
   slot {
    class=VerticalBoxSlot
    properties {horizontalAlignment=HAlign_Center, verticalAlignment=VAlign_Bottom, padding=<ltrb 0 5 0 0>, size {sizeRule=Fill, value=1}}
   }
  }
  Image_ProgressCircle {
   widgetClassPath=&Image
   bIsVariable=true
   parent=InputKeybind
   widget=Image_ProgressCircle
   properties {
    visibility=SelfHitTestInvisible
    brush {
     bIsDynamicallyLoaded=false
     drawAs=Image
     imageType=NoImage
     mirroring=NoMirror
     resourceName=None
     resourceObject=/Game/UI/Foundation/Toasts/Materials/MI_UI_RadialProgress
     tiling=NoTile
     imageSize=<64 64>
     margin=<ltrb 0 0 0 0>
     outlineSettings {
      bUseBrushTransparency=false
      roundingType=HalfHeightRadius
      width=0
      color {colorUseRule=UseColor_Specified, specifiedColor=<rgba 0 0 0 0>}
      cornerRadii=<0 0 0 0>
     }
     tintColor {colorUseRule=UseColor_Specified, specifiedColor=<rgba 1 1 1 1>}
     uVRegion {bIsValid=false, max=<0 0>, min=<0 0>}
    }
   }
   slot {class=OverlaySlot, properties {horizontalAlignment=HAlign_Center, verticalAlignment=VAlign_Center}}
  }
  Action_HoldKeybindWidget {
   widgetClassPath=/FortUILibrary/Input/KeybindWidget
   bInherited=true
   bIsVariable=true
   parent=InputKeybind
   widget=Action_HoldKeybindWidget
   properties {bUseKeyboardTextureForKeys=true, designTimeKey=Gamepad_Special_Right, forcedHoldKeybindStatus=ForcedHold}
   slot {class=OverlaySlot, properties {horizontalAlignment=HAlign_Center, verticalAlignment=VAlign_Center}}
  }
  Text_Action {
   widgetClassPath=&CommonTextBlock
   bIsVariable=true
   parent=VerticalBox_HoldAction
   widget=Text_Action
   properties {
    justification=Center
    mobileFontSizeMultiplier=1.3
    text=HOLD
    font {
     bForceMonospaced=false
     bMaterialIsStencil=false
     fontMaterial=null
     fontObject=&Burbank
     letterSpacing=0
     monospacedWidth=1
     size=16
     skewAmount=0
     typefaceFontName=Burbank-PlayerPower
     outlineSettings {
      bApplyOutlineToDropShadows=false
      bMiteredCorners=false
      bSeparateFillAlpha=false
      outlineMaterial=null
      outlineSize=0
      outlineColor=<rgba 0 0 0 1>
     }
    }
    renderTransform {angle=0, scale=<1 1>, shear=<-10 0>, translation=<-6 0>}
    shadowOffset=<0 0>
   }
   slot {class=VerticalBoxSlot, properties {horizontalAlignment=HAlign_Center, verticalAlignment=VAlign_Top, padding=<ltrb 8 0 10 0>}}
  }
  Switcher_InputButtons {
   widgetClassPath=&CommonWidgetSwitcherLegacy
   bIsVariable=true
   parent=Interaction
   widget=Switcher_InputButtons
   properties {renderOpacity=0, renderTransform {angle=0, scale=<1 1>, shear=<-10 0>, translation=<0 0>}}
   slot {class=GridSlot}
  }
  Button_Hold {
   widgetClassPath=&IconTextButton
   bIsVariable=true
   parent=Switcher_InputButtons
   widget=Button_Hold
   properties {
    bIgnoreInputActionWidgetText=true
    bIsFocusable=false
    bIsPersistentBinding=true
    buttonText=""
    onButtonBaseClicked=(&self:WidgetTree.Button_Hold.HandleOnButtonClicked)
    onButtonBaseDoubleClicked=(&self:WidgetTree.Button_Hold.HandleOnButtonDoubleClicked)
    onButtonBaseHovered=(&self:WidgetTree.Button_Hold.HandleOnButtonHovered)
    onButtonBaseUnhovered=(&self:WidgetTree.Button_Hold.HandleOnButtonUnhovered)
    onSelectedChangedBase=(&self:WidgetTree.Button_Hold.HandleOnSelectedChanged)
    style=null
   }
   slot {class=WidgetSwitcherSlot}
  }
  Button_Cursor {
   widgetClassPath=/FortUILibrary/Foundation/Buttons/IconTextButtonHold
   bIsVariable=true
   parent=Switcher_InputButtons
   widget=Button_Cursor
   properties {
    bIgnoreInputActionWidgetText=true
    bIsFocusable=false
    bIsPersistentBinding=true
    buttonText=""
    inputPriority=3
    onButtonBaseClicked=(&self:WidgetTree.Button_Cursor.HandleOnButtonClicked)
    onButtonBaseDoubleClicked=(&self:WidgetTree.Button_Cursor.HandleOnButtonDoubleClicked)
    onButtonBaseHovered=(&self:WidgetTree.Button_Cursor.HandleOnButtonHovered)
    onButtonBaseUnhovered=(&self:WidgetTree.Button_Cursor.HandleOnButtonUnhovered)
    onSelectedChangedBase=(&self:WidgetTree.Button_Cursor.HandleOnSelectedChanged)
    style=null
    triggeringInputAction {dataTable=&InputActionDataTable, rowName=ToastOpen}
   }
   slot {class=WidgetSwitcherSlot}
  }
 ]
}
