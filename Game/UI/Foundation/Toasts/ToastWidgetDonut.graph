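# Comic-bubble toast notification, derived from FortToastWidget, themed for the Donut (Deadpool) challenges.
# Notification events update the description, restyle for health warnings and append "Open menu to accept." to in-zone friend requests and party invites, while the title is never set here.
# Opening plays a pop-in and sound, then a display timer starts when it finishes; the outro plays Outro_New reversed, with animation-finished bindings managed on Construct and Destruct.
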WidgetBlueprint /Game/UI/Foundation/Toasts/ToastWidgetDonut
animations [
 Intro {
  binding_count=1
  length_seconds=0.15
  object=&self:Intro
  sequence {bindings [Overlay_1 {
     id=#1
     possessed_class=Overlay
     tracks [
      {
       type=MovieSceneFloatTrack
       display_name=Render Opacity
       sections [{
         type=MovieSceneFloatSection
         end_frame=3
         start_frame=0
         channels [None {type=MovieSceneScriptingFloatChannel, keys [[0,0],[2.694,1],[3,1]]}]
        }]
      }
      {
       type=MovieScene2DTransformTrack
       sections [{
         type=MovieScene2DTransformSection
         end_frame=3
         start_frame=0
         channels [
          Translation.X {type=MovieSceneScriptingFloatChannel}
          Translation.Y {type=MovieSceneScriptingFloatChannel}
          Angle {type=MovieSceneScriptingFloatChannel}
          Scale.X {type=MovieSceneScriptingFloatChannel, keys [[0,0.4],[2.3963,1.0447],[3,1]]}
          Scale.Y {type=MovieSceneScriptingFloatChannel, keys [[0,0.5],[2.3963,1.0447],[3,1]]}
          Shear.X {type=MovieSceneScriptingFloatChannel}
          Shear.Y {type=MovieSceneScriptingFloatChannel}
         ]
        }]
      }
     ]
    }]}
 }
 Outro_New {
  binding_count=1
  length_seconds=0.25
  object=&self:Outro_New
  sequence {bindings [Overlay_1 {
     id=#1
     possessed_class=Overlay
     tracks [
      {
       type=MovieSceneFloatTrack
       display_name=Render Opacity
       sections [{type=MovieSceneFloatSection, end_frame=5, start_frame=0, channels [None {type=MovieSceneScriptingFloatChannel, keys [[0,0],[5,1]]}]}]
      }
      {
       type=MovieScene2DTransformTrack
       sections [{
         type=MovieScene2DTransformSection
         end_frame=5
         start_frame=0
         channels [
          Translation.X {type=MovieSceneScriptingFloatChannel}
          Translation.Y {type=MovieSceneScriptingFloatChannel}
          Angle {type=MovieSceneScriptingFloatChannel}
          Scale.X {type=MovieSceneScriptingFloatChannel, keys [[0,0.4],[5,1]]}
          Scale.Y {type=MovieSceneScriptingFloatChannel, keys [[0,0.5],[5,1]]}
          Shear.X {type=MovieSceneScriptingFloatChannel}
          Shear.Y {type=MovieSceneScriptingFloatChannel}
         ]
        }]
      }
     ]
    }]}
 }
]
graphs [
 ApplyAdditionalStyling {nodes [
   ApplyAdditionalStyling() {is=title K2Node_FunctionEntry}
   n1 = "Equal (Enum)"(
    A:EFortNotificationType Enum=Get NotificationType!(self:Fort UINotification Object Reference=Get CachedNotification() -> (CachedNotification:Fort UINotification Object Reference)) -> (NotificationType:EFortNotificationType Enum=Default)
    B:EFortNotificationType Enum=HealthWarning
   )
   > Branch(Condition=@n1)
   > SetStyle!(
    execute
    self:Common Border Object Reference=Get CommonBorder_MainWidget() -> (CommonBorder_MainWidget:Common Border Object Reference)
    InStyle:Common Border Style Class Reference=Select(
     Option 0:Common Border Style Class Reference=/FortUILibrary/Styles/Border/Border-TabM-Solid-White100pc.Border-TabM-Solid-White100pc_C
     Option 1:Common Border Style Class Reference=/FortUILibrary/Styles/Border/Border-SolidBG-RewardBack.Border-SolidBG-RewardBack_C
     Index=@n1
    ) -> (ReturnValue:Common Border Style Class Reference)
   ) -> (then)
   > SetStyle(
    self:Common Text Object Reference=Get Title() -> (Title:Common Text Object Reference)
    InStyle:Common Text Style Class Reference=Select(
     Option 0:Common Text Style Class Reference=/FortUILibrary/Styles/Typography/Foundation/TextStyle-Base-XXS-B-Black.TextStyle-Base-XXS-B-Black_C
     Option 1:Common Text Style Class Reference=/FortUILibrary/Styles/Typography/Foundation/TextStyle-Base-XXS-B-Red.TextStyle-Base-XXS-B-Red_C
     Index=@n1
    ) -> (ReturnValue:Common Text Style Class Reference)
   )
   > SetStyle(
    self=Get Description() -> (Description:Common Text Object Reference)
    InStyle=Select(
     Option 0=/FortUILibrary/Styles/Typography/Foundation/TextStyle-Base-XXS-Black.TextStyle-Base-XXS-Black_C
     Option 1=/FortUILibrary/Styles/Typography/Foundation/TextStyle-Base-XXS-Red.TextStyle-Base-XXS-Red_C
     Index=@n1
    )
   )
  ]}
 EventGraph {nodes [
   HandleOutroFinished() {is=title K2Node_CustomEvent}
   > NotifyCloseAnimationFinished(self:Fort Toast Widget Object Reference)
   HandleIntroFinished() {is=title K2Node_CustomEvent}
   > Set Timer by Event(
    Delegate=Create Event()
    Time=GetDisplayTime!(self:Fort Toast Widget Object Reference) -> (bOutDisplayTimeOverridden:bool, ReturnValue:float).ReturnValue
    bLooping=false
   )
   PlayOutro() {is=title K2Node_CustomEvent}
   > PlayAnimation(InAnimation=Get Outro_New() -> (Outro_New:Widget Animation Object Reference), PlayMode=Reverse, PlaybackSpeed=1.5)
   Event OnPlayOpenAnimation() {is=title K2Node_Event}
   > StartIntro()
   n1 = Event OnNotificationSet!() -> (OutputDelegate, then, Notification:Fort UINotification Object Reference) {is=title K2Node_Event}
   > UpdateToast!(execute, self, Notification:Fort UINotification Object Reference=@n1.Notification) -> (then)
   n2 = Event OnNotificationUpdated!() -> (OutputDelegate, then, Notification:Fort UINotification Object Reference) {is=title K2Node_Event}
   > UpdateToast(Notification=@n2.Notification)
   Event Construct()
   > SetupAnimBindings!(execute, self, Unbind:bool=false) -> (then)
   Event Destruct()
   > SetupAnimBindings(Unbind=true)
  ]}
 GetOpenButtonVisibility {nodes [
   GetOpenButtonVisibility() {is=title K2Node_FunctionEntry}
   n1 = NOT Boolean(A=IsInZone!(self:Fort Global UIContext Object Reference=Get FortGlobalUIContext() -> (ReturnValue:Fort Global UIContext Object Reference) {is=title K2Node_GetSubsystem}) -> (ReturnValue:bool))
   n2 = Get bHasAction!(self:Fort UINotification Object Reference=Get CachedNotification()) -> (bHasAction:bool)
   > Return Node!(
    execute
    ReturnValue:ESlateVisibility Enum=Select(Option 0:ESlateVisibility Enum=Collapsed, Option 1:ESlateVisibility Enum=Visible, Index=AND Boolean(A=@n2, B=@n1)) -> (ReturnValue:ESlateVisibility Enum)
   )
  ]}
 OnFinishedToast {nodes [OnFinishedToast() {is=title K2Node_FunctionEntry}]}
 SetImage {nodes [SetImage() {is=title K2Node_FunctionEntry}]}
 SetupAnimBindings {nodes [
   n1 = SetupAnimBindings!() -> (then, Unbind:bool=~) {is=title K2Node_FunctionEntry}
   > n2 = Branch(Condition=@n1)
   > UnbindFromAnimationFinished!(
    execute
    self:Widget Animation Object Reference=Get Intro()
    Widget:User Widget Object Reference=Self-Reference()
    Delegate:Delegate=Create Event()
   ) -> (then)
   > UnbindFromAnimationFinished(self=Get Outro_New(), Widget=Self-Reference(), Delegate=Create Event())
   BindToAnimationFinished!(
    execute=@n2.else
    self:Widget Animation Object Reference=Get Intro()
    Widget:User Widget Object Reference=Self-Reference()
    Delegate=Create Event()
   ) -> (then)
   > BindToAnimationFinished(self=Get Outro_New(), Widget=Self-Reference(), Delegate=Create Event())
  ]}
 StartIntro {nodes [
   StartIntro() {is=title K2Node_FunctionEntry}
   > PlayAnimation(InAnimation=Get Intro())
   > Set bActionTaken(bActionTaken:bool=false) -> (Output_Get:bool)
   > PlaySound2D(Sound=/Game/Sounds/Fort_UI_Sounds/FrontEnd/Fort_UI_Party_Invited_Cue)
  ]}
 UpdateToast {nodes [
   n1 = UpdateToast!() -> (then, Notification:Fort UINotification Object Reference) {is=title K2Node_FunctionEntry}
   > "SetText (Text)"(self=Get Description(), InText=Get Description!(self:Fort UINotification Object Reference=@n1) -> (Description:Text))
   > SetImage() {is=title K2Node_CallFunction}
   > ApplyAdditionalStyling() {is=title K2Node_CallFunction}
   > Branch(Condition=Get bHasAction(self=Get Notification() -> (Notification:Fort UINotification Object Reference)))
   > n2 = Cast To FortUIFriendNotification!(execute, Object=Get Notification()) -> (then, CastFailed, AsFort UIFriend Notification:Fort UIFriend Notification Object Reference) {is=title K2Node_DynamicCast}
   > n3 = Switch on EFortUIFriendNotificationType!(
    execute
    Selection:EFortUIFriendNotificationType Enum=Get FriendActionType!(self:Fort UIFriend Notification Object Reference=@n2) -> (FriendActionType:EFortUIFriendNotificationType Enum=Default)
   ) -> (
    Default:Exec
    FriendRequest:Exec
    PartyInviteSent:Exec
    PartyInvite:Exec
    InviteFriendToParty:Exec
    PartyRequest:Exec
    AutoImportFriendSuggestion:Exec
    PartyMemberCreated:Exec
    PartyMemberLeft:Exec
    PartyJoin:Exec
    EpicPartyInvite:Exec
    EpicPartyInviteSent:Exec
    EpicPartyMemberJoined:Exec
    EpicPartyMemberLeft:Exec
    EpicPartyJoined:Exec
    InviteFriendToEpicParty:Exec
    EpicPartyJoin:Exec
    GroupInvite:Exec
    EpicPartyJoinabilityChanged:Exec
    EpicPartyJoinRequestSent:Exec
    EpicPartyJoinRequestReceived:Exec
    EpicPartyJoinRequestAccepted:Exec
    EpicPartyPromotedToLeader:Exec
    PartyInvite1v1:Exec
    PartyInvite1v1Sent:Exec
    PartyMemberSittingOut:Exec
   ) {is=title K2Node_SwitchEnum}
   Branch(execute=[@n3.FriendRequest, @n3.PartyInvite], Condition=IsInZone(self=Get FortGlobalUIContext()))
   n4 = Get Notification()
   > "SetText (Text)"(
    self=Get Description!() -> (Description:Common Text Object Reference)
    InText=Format Text!(
     Format="LOCTEXT(\"{0} {1}\")"
     0:Text=Get Description!(self:Fort UINotification Object Reference=@n4) -> (Description:Text)
     1:Text=MakeLiteralText(Value="LOCTEXT(\"Open menu to accept.\")")
    ) -> (Result)
   )
  ]}
]
variables [
 AnimationDelayTimer {type=TimerHandle}
 bActionTaken {type=boolean}
 OnFinishedToast {}
]
widgets {
 info {inheritedWidgetCount=0, parentClass=/Script/FortniteUI.FortToastWidget, rootWidgetClass=&Overlay, widgetCount=12}
 widgets [
  Overlay_1 {widgetClassPath=&Overlay, widget=Overlay_1, properties {renderTransformPivot=<0 1>}}
  B_Dropshadow {
   widgetClassPath=&Border
   parent=Overlay_1
   widget=B_Dropshadow
   properties {
    visibility=SelfHitTestInvisible
    background {
     bIsDynamicallyLoaded=false
     drawAs=Image
     imageType=NoImage
     mirroring=NoMirror
     resourceName=None
     resourceObject=/FortUILibrary/Styles/Border/Materials/M_Border_DropShadow
     tiling=NoTile
     imageSize=<32 32>
     margin=<ltrb 0 0 0 0>
     outlineSettings {
      bUseBrushTransparency=false
      roundingType=HalfHeightRadius
      width=0
      color {colorUseRule=UseColor_Specified, specifiedColor=<rgba 0 0 0 0>}
      cornerRadii=<0 0 0 0>
     }
     tintColor {colorUseRule=UseColor_Specified, specifiedColor=<rgba 1 1 1 1>}
     uVRegion {bIsValid=false, max=<0 0>, min=<0 0>}
    }
    renderTransform {angle=0, scale=<1 1>, shear=<0 0>, translation=<5 8>}
    renderTransformPivot=<0 0.5>
   }
   slot {class=OverlaySlot, properties {horizontalAlignment=HAlign_Fill, verticalAlignment=VAlign_Fill}}
  }
  HorizontalBox_1 {
   widgetClassPath=&HorizontalBox
   parent=Overlay_1
   widget=HorizontalBox_1
   slot {class=OverlaySlot, properties {horizontalAlignment=HAlign_Fill, verticalAlignment=VAlign_Fill}}
  }
  SizeBox_4 {
   widgetClassPath=&SizeBox
   parent=HorizontalBox_1
   widget=SizeBox_4
   properties {
    bOverride_MaxDesiredWidth=true
    bOverride_MinDesiredWidth=true
    maxDesiredHeight=150
    maxDesiredWidth=750
    minDesiredHeight=150
    minDesiredWidth=600
    widthOverride=600
   }
   slot {class=HorizontalBoxSlot}
  }
  Border_0 {
   widgetClassPath=&Border
   parent=SizeBox_4
   widget=Border_0
   properties {brushColor=<rgba 0 0 0 1>, padding=<ltrb 4 4 4 4>}
   slot {class=SizeBoxSlot}
  }
  CommonBorder_MainWidget {
   widgetClassPath=&CommonBorder
   bIsVariable=true
   parent=Border_0
   widget=CommonBorder_MainWidget
   properties {
    background {
     bIsDynamicallyLoaded=false
     drawAs=Box
     imageType=NoImage
     mirroring=NoMirror
     resourceName=None
     resourceObject=/Game/Environments/FrontEnd/BackPlates/UITakeovers/M_UI_ComicBubble_Inst
     tiling=NoTile
     imageSize=<52 36>
     margin=<ltrb 0 0 0 0>
     outlineSettings {
      bUseBrushTransparency=false
      roundingType=HalfHeightRadius
      width=0
      color {colorUseRule=UseColor_Specified, specifiedColor=<rgba 0 0 0 0>}
      cornerRadii=<0 0 0 0>
     }
     tintColor {colorUseRule=UseColor_Specified, specifiedColor=<rgba 1 1 1 1>}
     uVRegion {bIsValid=false, max=<0 0>, min=<0 0>}
    }
    brushColor=<rgba 1 1 1 0.9>
    padding=<ltrb 0 0 0 0>
    renderTransformPivot=<0 0.5>
   }
   slot {class=BorderSlot, properties {padding=<ltrb 4 4 4 4>}}
  }
  HorizontalBox_0 {
   widgetClassPath=&HorizontalBox
   parent=CommonBorder_MainWidget
   widget=HorizontalBox_0
   properties {clipping=ClipToBounds}
   slot {class=BorderSlot, properties {padding=<ltrb 0 0 0 0>}}
  }
  Image_134 {
   widgetClassPath=&Image
   bIsVariable=true
   parent=HorizontalBox_0
   widget=Image_134
   properties {
    brush {
     bIsDynamicallyLoaded=false
     drawAs=Image
     imageType=NoImage
     mirroring=NoMirror
     resourceName=None
     resourceObject=/Game/Athena/UI/Challenges/Art/Donut/M_UI_DonutIcons_Toast
     tiling=NoTile
     imageSize=<70 70>
     margin=<ltrb 0 0 0 0>
     outlineSettings {
      bUseBrushTransparency=false
      roundingType=HalfHeightRadius
      width=0
      color {colorUseRule=UseColor_Specified, specifiedColor=<rgba 0 0 0 0>}
      cornerRadii=<0 0 0 0>
     }
     tintColor {colorUseRule=UseColor_Specified, specifiedColor=<rgba 1 1 1 1>}
     uVRegion {bIsValid=false, max=<0 0>, min=<0 0>}
    }
    colorAndOpacity=<rgba 0 0 0 1>
    renderTransform {angle=3, scale=<1 1>, shear=<0 0>, translation=<2 5>}
   }
   slot {class=HorizontalBoxSlot, properties {horizontalAlignment=HAlign_Center, verticalAlignment=VAlign_Center, padding=<ltrb 0 -20 0 -20>}}
  }
  VerticalBox_0 {
   widgetClassPath=&VerticalBox
   parent=HorizontalBox_0
   widget=VerticalBox_0
   slot {class=HorizontalBoxSlot, properties {padding=<ltrb -10 10 10 10>, size {sizeRule=Fill, value=1}}}
  }
  TextBlock_0 {
   widgetClassPath=&TextBlock
   bIsVariable=true
   parent=VerticalBox_0
   widget=TextBlock_0
   properties {
    text=PartyInvite
    visibility=Collapsed
    font {
     bForceMonospaced=false
     bMaterialIsStencil=false
     fontMaterial=null
     fontObject=&Burbank
     letterSpacing=0
     monospacedWidth=1
     size=24
     skewAmount=0
     typefaceFontName=Burbank-Black
     outlineSettings {
      bApplyOutlineToDropShadows=true
      bMiteredCorners=false
      bSeparateFillAlpha=false
      outlineMaterial=null
      outlineSize=2
      outlineColor=<rgba 0 0 0 1>
     }
    }
   }
   slot {class=VerticalBoxSlot}
  }
  Title {
   widgetClassPath=&CommonTextBlock
   bIsVariable=true
   parent=VerticalBox_0
   widget=Title
   properties {
    justification=Center
    mobileFontSizeMultiplier=1.35
    scrollStyle=&TextScrollStyle-Base_C
    text=Deadpool Challenge Completed
    textTransformPolicy=ToUpper
    colorAndOpacity {colorUseRule=UseColor_Specified, specifiedColor=<rgba 0 0 0 1>}
    font {
     bForceMonospaced=false
     bMaterialIsStencil=false
     fontMaterial=null
     fontObject=&BurbankSmall
     letterSpacing=0
     monospacedWidth=1
     size=20
     skewAmount=0
     typefaceFontName=Black
     outlineSettings {
      bApplyOutlineToDropShadows=false
      bMiteredCorners=false
      bSeparateFillAlpha=false
      outlineMaterial=null
      outlineSize=0
      outlineColor=<rgba 0 0 0 1>
     }
    }
    shadowOffset=<0 0>
   }
   slot {class=VerticalBoxSlot, properties {size {sizeRule=Fill, value=1}}}
  }
  Description {
   widgetClassPath=&CommonTextBlock
   bIsVariable=true
   parent=VerticalBox_0
   widget=Description
   properties {
    autoWrapText=true
    bAutoCollapseWithEmptyText=true
    justification=Center
    lineHeightPercentage=0.85
    mobileFontSizeMultiplier=1.34
    scrollStyle=&TextScrollStyle-Base_C
    text="Visit Deadpool's room to view more!"
    textTransformPolicy=ToUpper
    colorAndOpacity {colorUseRule=UseColor_Specified, specifiedColor=<rgba 0 0 0 1>}
    font {
     bForceMonospaced=false
     bMaterialIsStencil=false
     fontMaterial=null
     fontObject=&BurbankSmall
     letterSpacing=0
     monospacedWidth=1
     size=16
     skewAmount=0
     typefaceFontName=Bold
     outlineSettings {
      bApplyOutlineToDropShadows=false
      bMiteredCorners=false
      bSeparateFillAlpha=false
      outlineMaterial=null
      outlineSize=0
      outlineColor=<rgba 0 0 0 1>
     }
    }
    shadowOffset=<0 0>
   }
   slot {class=VerticalBoxSlot, properties {padding=<ltrb 0 5 0 0>}}
  }
 ]
}
