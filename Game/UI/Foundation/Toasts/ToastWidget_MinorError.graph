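# Fortnite toast widget that shows minor-error notifications with a warning icon, message, note and optional Open button.
# UpdateToast fills text from the notification, picks a random apology title (collapsed by default) and restyles for HealthWarning.
# After the animated intro, a display-time timer runs, pausing on hover.
# Open takes the action once, clears the timer and plays the outro, an intro-style animation run in reverse.
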WidgetBlueprint /Game/UI/Foundation/Toasts/ToastWidget_MinorError
animations [
 Intro {
  binding_count=5
  display_label=""
  length_seconds=0.75
  object=&self:Intro
  sequence {
   bindings [
    B_NotificationImage {
     id=#1
     possessed_class=CommonBorder
     tracks [{
       type=MovieScene2DTransformTrack
       sections [{
         type=MovieScene2DTransformSection
         end_frame=75
         start_frame=0
         channels [
          Translation.X {type=MovieSceneScriptingFloatChannel}
          Translation.Y {type=MovieSceneScriptingFloatChannel}
          Angle {type=MovieSceneScriptingFloatChannel, keys [[0,0],[25,0],[75,-360]]}
          Scale.X {type=MovieSceneScriptingFloatChannel, keys [[0,0],[14.9983,0.8471],[17.9983,1.3111],[21.9983,1.2062],[23.9983,1.0246],[25.9983,1]]}
          Scale.Y {type=MovieSceneScriptingFloatChannel, keys [[0,0],[10.9983,1.7484],[17.9983,0.795],[21.9983,1.2062],[23.9983,1.0246],[25.9983,1]]}
          Shear.X {type=MovieSceneScriptingFloatChannel}
          Shear.Y {type=MovieSceneScriptingFloatChannel}
         ]
        }]
      }]
    }
    B_NotificationImageExtended {
     id=#2
     possessed_class=CommonBorder
     tracks [{
       type=MovieScene2DTransformTrack
       sections [{
         type=MovieScene2DTransformSection
         end_frame=75
         start_frame=0
         channels [
          Translation.X {type=MovieSceneScriptingFloatChannel}
          Translation.Y {type=MovieSceneScriptingFloatChannel}
          Angle {type=MovieSceneScriptingFloatChannel}
          Scale.X {type=MovieSceneScriptingFloatChannel}
          Scale.Y {type=MovieSceneScriptingFloatChannel, thinned_from=6, keys [[0,0],[25,0],[42.9983,1.1427],[52.9983,1],[75,1]]}
          Shear.X {type=MovieSceneScriptingFloatChannel}
          Shear.Y {type=MovieSceneScriptingFloatChannel}
         ]
        }]
      }]
    }
    HorizontalBox_1 {
     id=#3
     possessed_class=HorizontalBox
     tracks [{
       type=MovieScene2DTransformTrack
       sections [{
         type=MovieScene2DTransformSection
         end_frame=75
         start_frame=0
         channels [
          Translation.X {type=MovieSceneScriptingFloatChannel}
          Translation.Y {type=MovieSceneScriptingFloatChannel}
          Angle {type=MovieSceneScriptingFloatChannel}
          Scale.X {type=MovieSceneScriptingFloatChannel, const [1,5]}
          Scale.Y {type=MovieSceneScriptingFloatChannel, keys [[0,1],[34.9983,1],[44.9983,1.1489],[62,0.8419],[75,1]]}
          Shear.X {type=MovieSceneScriptingFloatChannel}
          Shear.Y {type=MovieSceneScriptingFloatChannel}
         ]
        }]
      }]
    }
    B_Dropshadow {
     id=#4
     possessed_class=Border
     tracks [{
       type=MovieScene2DTransformTrack
       sections [{
         type=MovieScene2DTransformSection
         end_frame=75
         start_frame=0
         channels [
          Translation.X {type=MovieSceneScriptingFloatChannel}
          Translation.Y {type=MovieSceneScriptingFloatChannel}
          Angle {type=MovieSceneScriptingFloatChannel}
          Scale.X {type=MovieSceneScriptingFloatChannel, keys [[0,0],[52.9983,0],[62.9983,1],[64.9983,1.05],[69.9983,0.96],[75,1]]}
          Scale.Y {type=MovieSceneScriptingFloatChannel}
          Shear.X {type=MovieSceneScriptingFloatChannel}
          Shear.Y {type=MovieSceneScriptingFloatChannel}
         ]
        }]
      }]
    }
    CommonBorder_MainWidget {
     id=#5
     possessed_class=CommonBorder
     tracks [
      {
       type=MovieScene2DTransformTrack
       sections [{
         type=MovieScene2DTransformSection
         end_frame=75
         start_frame=0
         channels [
          Translation.X {type=MovieSceneScriptingFloatChannel}
          Translation.Y {type=MovieSceneScriptingFloatChannel}
          Angle {type=MovieSceneScriptingFloatChannel}
          Scale.X {type=MovieSceneScriptingFloatChannel, keys [[0,0],[52.9983,0],[62.9983,1],[69.9983,0.9461],[75,1]]}
          Scale.Y {type=MovieSceneScriptingFloatChannel, const [1,2]}
          Shear.X {type=MovieSceneScriptingFloatChannel}
          Shear.Y {type=MovieSceneScriptingFloatChannel}
         ]
        }]
      }
      {
       type=MovieSceneColorTrack
       display_name=Content Color and Opacity
       sections [{
         type=MovieSceneColorSection
         end_frame=75
         start_frame=0
         channels [
          Color.R {type=MovieSceneScriptingFloatChannel}
          Color.G {type=MovieSceneScriptingFloatChannel}
          Color.B {type=MovieSceneScriptingFloatChannel}
          Color.A {type=MovieSceneScriptingFloatChannel, keys [[0,0],[57.9983,0],[68.9983,1],[75,1]]}
         ]
        }]
      }
     ]
    }
   ]
   display_rate {numerator=100}
  }
 }
 Outro {
  binding_count=1
  display_label=""
  length_seconds=0.5
  object=&self:Outro
  sequence {bindings [RootVerticalBox {
     id=#6
     possessed_class=VerticalBox
     tracks [{
       type=MovieScene2DTransformTrack
       sections [{
         type=MovieScene2DTransformSection
         end_frame=10
         start_frame=0
         channels [
          Translation.X {type=MovieSceneScriptingFloatChannel, keys [[0,0],[10,-600]]}
          Translation.Y {type=MovieSceneScriptingFloatChannel, const [0,1]}
          Angle {type=MovieSceneScriptingFloatChannel, const [0,1]}
          Scale.X {type=MovieSceneScriptingFloatChannel, const [1,1]}
          Scale.Y {type=MovieSceneScriptingFloatChannel, const [1,1]}
          Shear.X {type=MovieSceneScriptingFloatChannel, const [0,1]}
          Shear.Y {type=MovieSceneScriptingFloatChannel, const [0,1]}
         ]
        }]
      }]
    }]}
 }
 Intro_new2 {
  binding_count=3
  display_label=""
  length_seconds=0.55
  object=&self:Intro_new2
  sequence {
   bindings [
    B_NotificationImage {
     id=#7
     possessed_class=CommonBorder
     tracks [
      {
       type=MovieScene2DTransformTrack
       sections [{
         type=MovieScene2DTransformSection
         end_frame=55
         start_frame=0
         channels [
          Translation.X {type=MovieSceneScriptingFloatChannel, keys [[0,40],[14.9983,75],[25,75],[34.9983,0]]}
          Translation.Y {type=MovieSceneScriptingFloatChannel, const [0,4]}
          Angle {type=MovieSceneScriptingFloatChannel}
          Scale.X {type=MovieSceneScriptingFloatChannel, const [1,1]}
          Scale.Y {type=MovieSceneScriptingFloatChannel, const [1,1]}
          Shear.X {type=MovieSceneScriptingFloatChannel}
          Shear.Y {type=MovieSceneScriptingFloatChannel}
         ]
        }]
      }
      {
       type=MovieSceneFloatTrack
       display_name=Render Opacity
       sections [{type=MovieSceneFloatSection, end_frame=15, start_frame=0, channels [None {type=MovieSceneScriptingFloatChannel, keys [[0,0],[14.9983,1]]}]}]
      }
     ]
    }
    B_NotificationImageExtended {
     id=#8
     possessed_class=CommonBorder
     tracks [{
       type=MovieScene2DTransformTrack
       sections [{
         type=MovieScene2DTransformSection
         end_frame=55
         start_frame=0
         channels [
          Translation.X {type=MovieSceneScriptingFloatChannel}
          Translation.Y {type=MovieSceneScriptingFloatChannel}
          Angle {type=MovieSceneScriptingFloatChannel}
          Scale.X {type=MovieSceneScriptingFloatChannel, keys [[0,0],[34.9983,0],[44.9983,1],[55,1]]}
          Scale.Y {type=MovieSceneScriptingFloatChannel, const [1,4]}
          Shear.X {type=MovieSceneScriptingFloatChannel}
          Shear.Y {type=MovieSceneScriptingFloatChannel}
         ]
        }]
      }]
    }
    CommonBorder_MainWidget {
     id=#9
     possessed_class=CommonBorder
     tracks [{
       type=MovieScene2DTransformTrack
       sections [{
         type=MovieScene2DTransformSection
         end_frame=55
         start_frame=0
         channels [
          Translation.X {type=MovieSceneScriptingFloatChannel}
          Translation.Y {type=MovieSceneScriptingFloatChannel}
          Angle {type=MovieSceneScriptingFloatChannel}
          Scale.X {type=MovieSceneScriptingFloatChannel, keys [[0,0],[44.9983,0],[55,1]]}
          Scale.Y {type=MovieSceneScriptingFloatChannel, const [1,3]}
          Shear.X {type=MovieSceneScriptingFloatChannel}
          Shear.Y {type=MovieSceneScriptingFloatChannel}
         ]
        }]
      }]
    }
   ]
   display_rate {numerator=100}
  }
 }
 Intro_4114 {
  binding_count=4
  display_label=""
  length_seconds=0.75
  object=&self:Intro_4114
  sequence {
   bindings [
    B_NotificationImage {
     id=#1
     possessed_class=CommonBorder
     tracks [{
       type=MovieScene2DTransformTrack
       sections [{
         type=MovieScene2DTransformSection
         end_frame=75
         start_frame=0
         channels [
          Translation.X {type=MovieSceneScriptingFloatChannel}
          Translation.Y {type=MovieSceneScriptingFloatChannel}
          Angle {type=MovieSceneScriptingFloatChannel, keys [[0,0],[25,0],[75,-360]]}
          Scale.X {type=MovieSceneScriptingFloatChannel, keys [[0,0],[14.9983,0.8471],[17.9983,1.3111],[21.9983,1.2062],[23.9983,1.0246],[25.9983,1]]}
          Scale.Y {type=MovieSceneScriptingFloatChannel, keys [[0,0],[10.9983,1.7484],[17.9983,0.795],[21.9983,1.2062],[23.9983,1.0246],[25.9983,1]]}
          Shear.X {type=MovieSceneScriptingFloatChannel}
          Shear.Y {type=MovieSceneScriptingFloatChannel}
         ]
        }]
      }]
    }
    B_NotificationImageExtended {
     id=#2
     possessed_class=CommonBorder
     tracks [{
       type=MovieScene2DTransformTrack
       sections [{
         type=MovieScene2DTransformSection
         end_frame=75
         start_frame=0
         channels [
          Translation.X {type=MovieSceneScriptingFloatChannel}
          Translation.Y {type=MovieSceneScriptingFloatChannel}
          Angle {type=MovieSceneScriptingFloatChannel}
          Scale.X {type=MovieSceneScriptingFloatChannel}
          Scale.Y {type=MovieSceneScriptingFloatChannel, thinned_from=6, keys [[0,0],[25,0],[42.9983,1.1427],[52.9983,1],[75,1]]}
          Shear.X {type=MovieSceneScriptingFloatChannel}
          Shear.Y {type=MovieSceneScriptingFloatChannel}
         ]
        }]
      }]
    }
    HorizontalBox_1 {
     id=#3
     possessed_class=HorizontalBox
     tracks [{
       type=MovieScene2DTransformTrack
       sections [{
         type=MovieScene2DTransformSection
         end_frame=75
         start_frame=0
         channels [
          Translation.X {type=MovieSceneScriptingFloatChannel}
          Translation.Y {type=MovieSceneScriptingFloatChannel}
          Angle {type=MovieSceneScriptingFloatChannel}
          Scale.X {type=MovieSceneScriptingFloatChannel, const [1,5]}
          Scale.Y {type=MovieSceneScriptingFloatChannel, keys [[0,1],[34.9983,1],[44.9983,1.1489],[62,0.8419],[75,1]]}
          Shear.X {type=MovieSceneScriptingFloatChannel}
          Shear.Y {type=MovieSceneScriptingFloatChannel}
         ]
        }]
      }]
    }
    CommonBorder_MainWidget {
     id=#5
     possessed_class=CommonBorder
     tracks [
      {
       type=MovieScene2DTransformTrack
       sections [{
         type=MovieScene2DTransformSection
         end_frame=75
         start_frame=0
         channels [
          Translation.X {type=MovieSceneScriptingFloatChannel}
          Translation.Y {type=MovieSceneScriptingFloatChannel}
          Angle {type=MovieSceneScriptingFloatChannel}
          Scale.X {type=MovieSceneScriptingFloatChannel, keys [[0,0],[52.9983,0],[62.9983,1],[69.9983,0.9461],[75,1]]}
          Scale.Y {type=MovieSceneScriptingFloatChannel, const [1,2]}
          Shear.X {type=MovieSceneScriptingFloatChannel}
          Shear.Y {type=MovieSceneScriptingFloatChannel}
         ]
        }]
      }
      {
       type=MovieSceneColorTrack
       display_name=Content Color and Opacity
       sections [{
         type=MovieSceneColorSection
         end_frame=75
         start_frame=0
         channels [
          Color.R {type=MovieSceneScriptingFloatChannel}
          Color.G {type=MovieSceneScriptingFloatChannel}
          Color.B {type=MovieSceneScriptingFloatChannel}
          Color.A {type=MovieSceneScriptingFloatChannel, keys [[0,0],[57.9983,0],[68.9983,1],[75,1]]}
         ]
        }]
      }
     ]
    }
   ]
   display_rate {numerator=100}
  }
 }
 Intro_212 {
  binding_count=5
  display_label=""
  length_seconds=0.75
  object=&self:Intro_212
  sequence {
   bindings [
    B_NotificationImage {
     id=#1
     possessed_class=CommonBorder
     tracks [{
       type=MovieScene2DTransformTrack
       sections [{
         type=MovieScene2DTransformSection
         end_frame=75
         start_frame=0
         channels [
          Translation.X {type=MovieSceneScriptingFloatChannel}
          Translation.Y {type=MovieSceneScriptingFloatChannel}
          Angle {type=MovieSceneScriptingFloatChannel, keys [[0,0],[25,0],[75,-360]]}
          Scale.X {type=MovieSceneScriptingFloatChannel, keys [[0,0],[14.9983,0.8471],[17.9983,1.3111],[21.9983,1.2062],[23.9983,1.0246],[25.9983,1]]}
          Scale.Y {type=MovieSceneScriptingFloatChannel, keys [[0,0],[10.9983,1.7484],[17.9983,0.795],[21.9983,1.2062],[23.9983,1.0246],[25.9983,1]]}
          Shear.X {type=MovieSceneScriptingFloatChannel}
          Shear.Y {type=MovieSceneScriptingFloatChannel}
         ]
        }]
      }]
    }
    B_NotificationImageExtended {
     id=#2
     possessed_class=CommonBorder
     tracks [{
       type=MovieScene2DTransformTrack
       sections [{
         type=MovieScene2DTransformSection
         end_frame=75
         start_frame=0
         channels [
          Translation.X {type=MovieSceneScriptingFloatChannel}
          Translation.Y {type=MovieSceneScriptingFloatChannel}
          Angle {type=MovieSceneScriptingFloatChannel}
          Scale.X {type=MovieSceneScriptingFloatChannel}
          Scale.Y {type=MovieSceneScriptingFloatChannel, thinned_from=6, keys [[0,0],[25,0],[42.9983,1.1427],[52.9983,1],[75,1]]}
          Shear.X {type=MovieSceneScriptingFloatChannel}
          Shear.Y {type=MovieSceneScriptingFloatChannel}
         ]
        }]
      }]
    }
    HorizontalBox_1 {
     id=#3
     possessed_class=HorizontalBox
     tracks [{
       type=MovieScene2DTransformTrack
       sections [{
         type=MovieScene2DTransformSection
         end_frame=75
         start_frame=0
         channels [
          Translation.X {type=MovieSceneScriptingFloatChannel}
          Translation.Y {type=MovieSceneScriptingFloatChannel}
          Angle {type=MovieSceneScriptingFloatChannel}
          Scale.X {type=MovieSceneScriptingFloatChannel, const [1,5]}
          Scale.Y {type=MovieSceneScriptingFloatChannel, keys [[0,1],[34.9983,1],[44.9983,1.1489],[62,0.8419],[75,1]]}
          Shear.X {type=MovieSceneScriptingFloatChannel}
          Shear.Y {type=MovieSceneScriptingFloatChannel}
         ]
        }]
      }]
    }
    B_Dropshadow {
     id=#4
     possessed_class=Border
     tracks [{
       type=MovieScene2DTransformTrack
       sections [{
         type=MovieScene2DTransformSection
         end_frame=75
         start_frame=0
         channels [
          Translation.X {type=MovieSceneScriptingFloatChannel}
          Translation.Y {type=MovieSceneScriptingFloatChannel}
          Angle {type=MovieSceneScriptingFloatChannel}
          Scale.X {type=MovieSceneScriptingFloatChannel, keys [[0,0],[52.9983,0],[62.9983,1],[64.9983,1.05],[69.9983,0.96],[75,1]]}
          Scale.Y {type=MovieSceneScriptingFloatChannel}
          Shear.X {type=MovieSceneScriptingFloatChannel}
          Shear.Y {type=MovieSceneScriptingFloatChannel}
         ]
        }]
      }]
    }
    CommonBorder_MainWidget {
     id=#5
     possessed_class=CommonBorder
     tracks [
      {
       type=MovieScene2DTransformTrack
       sections [{
         type=MovieScene2DTransformSection
         end_frame=75
         start_frame=0
         channels [
          Translation.X {type=MovieSceneScriptingFloatChannel}
          Translation.Y {type=MovieSceneScriptingFloatChannel}
          Angle {type=MovieSceneScriptingFloatChannel}
          Scale.X {type=MovieSceneScriptingFloatChannel, keys [[0,0],[52.9983,0],[62.9983,1],[69.9983,0.9461],[75,1]]}
          Scale.Y {type=MovieSceneScriptingFloatChannel, const [1,2]}
          Shear.X {type=MovieSceneScriptingFloatChannel}
          Shear.Y {type=MovieSceneScriptingFloatChannel}
         ]
        }]
      }
      {
       type=MovieSceneColorTrack
       display_name=Content Color and Opacity
       sections [{
         type=MovieSceneColorSection
         end_frame=75
         start_frame=0
         channels [
          Color.R {type=MovieSceneScriptingFloatChannel}
          Color.G {type=MovieSceneScriptingFloatChannel}
          Color.B {type=MovieSceneScriptingFloatChannel}
          Color.A {type=MovieSceneScriptingFloatChannel, keys [[0,0],[57.9983,0],[68.9983,1],[75,1]]}
         ]
        }]
      }
     ]
    }
   ]
   display_rate {numerator=100}
  }
 }
 Outro_New {
  binding_count=5
  display_label=""
  length_seconds=0.63
  object=&self:Outro_New
  sequence {
   bindings [
    B_NotificationImage {
     id=#1
     possessed_class=CommonBorder
     tracks [{
       type=MovieScene2DTransformTrack
       sections [{
         type=MovieScene2DTransformSection
         end_frame=75
         start_frame=0
         channels [
          Translation.X {type=MovieSceneScriptingFloatChannel}
          Translation.Y {type=MovieSceneScriptingFloatChannel}
          Angle {type=MovieSceneScriptingFloatChannel, const [-360,2]}
          Scale.X {type=MovieSceneScriptingFloatChannel, keys [[0,0],[14.9983,0.8471],[17.9983,1.3111],[21.9983,1.4447],[23.9983,1.0246],[25.9983,1]]}
          Scale.Y {type=MovieSceneScriptingFloatChannel, keys [[0,0],[10.9983,1.7484],[17.9983,0.795],[21.9983,0.7927],[23.9983,1.0246],[25.9983,1]]}
          Shear.X {type=MovieSceneScriptingFloatChannel}
          Shear.Y {type=MovieSceneScriptingFloatChannel}
         ]
        }]
      }]
    }
    B_NotificationImageExtended {
     id=#2
     possessed_class=CommonBorder
     tracks [{
       type=MovieScene2DTransformTrack
       sections [{
         type=MovieScene2DTransformSection
         end_frame=75
         start_frame=0
         channels [
          Translation.X {type=MovieSceneScriptingFloatChannel}
          Translation.Y {type=MovieSceneScriptingFloatChannel}
          Angle {type=MovieSceneScriptingFloatChannel}
          Scale.X {type=MovieSceneScriptingFloatChannel}
          Scale.Y {type=MovieSceneScriptingFloatChannel, thinned_from=6, keys [[0,0],[16.9983,0],[31,1.1427],[40.9983,1],[62.9983,1]]}
          Shear.X {type=MovieSceneScriptingFloatChannel}
          Shear.Y {type=MovieSceneScriptingFloatChannel}
         ]
        }]
      }]
    }
    HorizontalBox_1 {
     id=#3
     possessed_class=HorizontalBox
     tracks [{
       type=MovieScene2DTransformTrack
       sections [{
         type=MovieScene2DTransformSection
         end_frame=75
         start_frame=0
         channels [
          Translation.X {type=MovieSceneScriptingFloatChannel}
          Translation.Y {type=MovieSceneScriptingFloatChannel}
          Angle {type=MovieSceneScriptingFloatChannel}
          Scale.X {type=MovieSceneScriptingFloatChannel, const [1,5]}
          Scale.Y {type=MovieSceneScriptingFloatChannel, keys [[0,1],[26.9983,1],[36.9983,1.1489],[62.9983,1]]}
          Shear.X {type=MovieSceneScriptingFloatChannel}
          Shear.Y {type=MovieSceneScriptingFloatChannel}
         ]
        }]
      }]
    }
    B_Dropshadow {
     id=#4
     possessed_class=Border
     tracks [{
       type=MovieScene2DTransformTrack
       sections [{
         type=MovieScene2DTransformSection
         end_frame=75
         start_frame=0
         channels [
          Translation.X {type=MovieSceneScriptingFloatChannel}
          Translation.Y {type=MovieSceneScriptingFloatChannel}
          Angle {type=MovieSceneScriptingFloatChannel}
          Scale.X {type=MovieSceneScriptingFloatChannel, keys [[0,0],[40.9983,0],[50.9983,1],[52.9983,1.05],[58,0.96],[62.9983,1]]}
          Scale.Y {type=MovieSceneScriptingFloatChannel}
          Shear.X {type=MovieSceneScriptingFloatChannel}
          Shear.Y {type=MovieSceneScriptingFloatChannel}
         ]
        }]
      }]
    }
    CommonBorder_MainWidget {
     id=#5
     possessed_class=CommonBorder
     tracks [
      {
       type=MovieScene2DTransformTrack
       sections [{
         type=MovieScene2DTransformSection
         end_frame=75
         start_frame=0
         channels [
          Translation.X {type=MovieSceneScriptingFloatChannel}
          Translation.Y {type=MovieSceneScriptingFloatChannel}
          Angle {type=MovieSceneScriptingFloatChannel}
          Scale.X {type=MovieSceneScriptingFloatChannel, keys [[0,0],[40.9983,0],[50.9983,1],[62.9983,1]]}
          Scale.Y {type=MovieSceneScriptingFloatChannel, const [1,2]}
          Shear.X {type=MovieSceneScriptingFloatChannel}
          Shear.Y {type=MovieSceneScriptingFloatChannel}
         ]
        }]
      }
      {
       type=MovieSceneColorTrack
       display_name=Content Color and Opacity
       sections [{
         type=MovieSceneColorSection
         end_frame=75
         start_frame=0
         channels [
          Color.R {type=MovieSceneScriptingFloatChannel}
          Color.G {type=MovieSceneScriptingFloatChannel}
          Color.B {type=MovieSceneScriptingFloatChannel}
          Color.A {type=MovieSceneScriptingFloatChannel, keys [[0,0],[46,0],[57,1],[62.9983,1]]}
         ]
        }]
      }
     ]
    }
   ]
   display_rate {numerator=100}
  }
 }
]
graphs [
 ApplyAdditionalStyling {nodes [
   ApplyAdditionalStyling() {is=title K2Node_FunctionEntry}
   n1 = "Equal (Enum)"(
    A:EFortNotificationType Enum=Get NotificationType!(self:Fort UINotification Object Reference=Get CachedNotification() -> (CachedNotification:Fort UINotification Object Reference)) -> (NotificationType:EFortNotificationType Enum=Default)
    B:EFortNotificationType Enum=HealthWarning
   )
   > Branch(Condition=@n1)
   > SetStyle!(
    execute
    self:Common Border Object Reference=Get CommonBorder_MainWidget() -> (CommonBorder_MainWidget:Common Border Object Reference)
    InStyle:Common Border Style Class Reference=Select(
     Option 0:Common Border Style Class Reference=/FortUILibrary/Styles/Border/Border-TabM-Solid-White100pc.Border-TabM-Solid-White100pc_C
     Option 1:Common Border Style Class Reference=/FortUILibrary/Styles/Border/Border-SolidBG-RewardBack.Border-SolidBG-RewardBack_C
     Index=@n1
    ) -> (ReturnValue:Common Border Style Class Reference)
   ) -> (then)
   > SetStyle(
    self:Common Text Object Reference=Get Title() -> (Title:Common Text Object Reference)
    InStyle:Common Text Style Class Reference=Select(
     Option 0:Common Text Style Class Reference=/FortUILibrary/Styles/Typography/Foundation/TextStyle-Base-XXS-B-Black.TextStyle-Base-XXS-B-Black_C
     Option 1:Common Text Style Class Reference=/FortUILibrary/Styles/Typography/Foundation/TextStyle-Base-XXS-B-Red.TextStyle-Base-XXS-B-Red_C
     Index=@n1
    ) -> (ReturnValue:Common Text Style Class Reference)
   )
   > SetStyle(
    self=Get Description() -> (Description:Common Text Object Reference)
    InStyle=Select(
     Option 0=/FortUILibrary/Styles/Typography/Foundation/TextStyle-Base-XXS-Black.TextStyle-Base-XXS-Black_C
     Option 1=/FortUILibrary/Styles/Typography/Foundation/TextStyle-Base-XXS-Red.TextStyle-Base-XXS-Red_C
     Index=@n1
    )
   )
  ]}
 EventGraph {nodes [
   "On Button Clicked (OpenButton)"!() -> (OutputDelegate, then, Button:Common Button Legacy Object Reference) {is=title K2Node_ComponentBoundEvent}
   > Branch(Condition=Get bHasAction!(self:Fort UINotification Object Reference=Get CachedNotification()) -> (bHasAction:bool))
   > n1 = Branch(Condition=Get bActionTaken() -> (bActionTaken:bool))
   Set bActionTaken(execute=@n1.else, bActionTaken:bool=true) -> (Output_Get:bool)
   > TakeAction(self:Fort UINotification Object Reference=Get CachedNotification())
   > Clear and Invalidate Timer by Handle(Handle=Get AnimationDelayTimer() -> (AnimationDelayTimer:Timer Handle Structure))
   > Delay(Duration=0.2)
   > PlayOutro()
   HandleOutroFinished() {is=title K2Node_CustomEvent}
   > NotifyCloseAnimationFinished(self:Fort Toast Widget Object Reference)
   HandleIntroFinished() {is=title K2Node_CustomEvent}
   > n2 = Set Timer by Event(
    Delegate=Create Event()
    Time=GetDisplayTime!(self:Fort Toast Widget Object Reference) -> (bOutDisplayTimeOverridden:bool, ReturnValue:float).ReturnValue
    bLooping=false
   )
   > Set AnimationDelayTimer(AnimationDelayTimer:Timer Handle Structure=@n2) -> (Output_Get:Timer Handle Structure)
   PlayOutro() {is=title K2Node_CustomEvent}
   > PlayAnimation(InAnimation=Get Outro_New() -> (Outro_New:Widget Animation Object Reference), PlayMode=Reverse, PlaybackSpeed=1.5)
   Event OnPlayOpenAnimation() {is=title K2Node_Event}
   > StartIntro()
   n3 = Event OnNotificationSet!() -> (OutputDelegate, then, Notification:Fort UINotification Object Reference) {is=title K2Node_Event}
   > UpdateToast!(execute, self, Notification:Fort UINotification Object Reference=@n3.Notification) -> (then)
   Event OnMouseEnter()
   > Pause Timer by Handle!(execute, Handle:Timer Handle Structure=Get AnimationDelayTimer()) -> (then)
   Event OnMouseLeave!() -> (OutputDelegate, then, MouseEvent:"Pointer Event Structure (by ref)") {is=title K2Node_Event}
   > Unpause Timer by Handle!(execute, Handle:Timer Handle Structure=Get AnimationDelayTimer()) -> (then)
   n4 = Event OnNotificationUpdated!() -> (OutputDelegate, then, Notification:Fort UINotification Object Reference) {is=title K2Node_Event}
   > UpdateToast(Notification=@n4.Notification)
   Event Construct()
   > SetupAnimBindings!(execute, self, Unbind:bool=false) -> (then)
   Event Destruct()
   > SetupAnimBindings(Unbind=true)
  ]}
 GetOpenButtonVisibility {nodes [
   GetOpenButtonVisibility() {is=title K2Node_FunctionEntry}
   n1 = Get CachedNotification()
   > Return Node!(
    execute
    ReturnValue:ESlateVisibility Enum=Select(Option 0:ESlateVisibility Enum=Collapsed, Option 1:ESlateVisibility Enum=Visible, Index=Get bHasAction(self=@n1)) -> (ReturnValue:ESlateVisibility Enum)
   )
  ]}
 OnFinishedToast {nodes [OnFinishedToast() {is=title K2Node_FunctionEntry}]}
 SetupAnimBindings {nodes [
   n1 = SetupAnimBindings!() -> (then, Unbind:bool=~) {is=title K2Node_FunctionEntry}
   > n2 = Branch(Condition=@n1)
   > UnbindFromAnimationFinished!(execute, self:User Widget Object Reference, Animation:Widget Animation Object Reference=Get Intro(), Delegate:Delegate=Create Event()) -> (then)
   > UnbindFromAnimationFinished(Animation=Get Outro_New(), Delegate=Create Event())
   BindToAnimationFinished(execute=@n2.else, Animation=Get Intro(), Delegate=Create Event())
   > BindToAnimationFinished(Animation=Get Outro_New(), Delegate=Create Event())
  ]}
 StartIntro {nodes [
   StartIntro() {is=title K2Node_FunctionEntry}
   > PlayAnimation(InAnimation=Get Intro())
   > Set bActionTaken(bActionTaken=false)
  ]}
 UpdateToast {nodes [
   UpdateToast!() -> (then, Notification:Fort UINotification Object Reference) {is=title K2Node_FunctionEntry}
   n1 = Get NotificationImage() -> (NotificationImage:Image Object Reference)
   > SetBrush(
    self=@n1
    InBrush=Make Slate Brush!(
     TintColor:Slate Color Structure=(SpecifiedColor=(R=1,G=1,B=1,A=1),ColorUseRule=UseColor_Specified)
     DrawAs:ESlateBrushDrawType Enum=Image
     Tiling:ESlateBrushTileType Enum=NoTile
     Mirroring:ESlateBrushMirrorType Enum=NoMirror
     ImageSize:Vector 2D Structure=(X=64,Y=64)
     Margin:Margin Structure=(Left=0,Top=0,Right=0,Bottom=0)
     ResourceObject:Object Reference=/Game/UI/Foundation/Shell/Textures/T-Icon-Warning-64
     OutlineSettings:Slate Brush Outline Settings Structure=(CornerRadii=(X=0,Y=0,Z=0,W=0),Color=(SpecifiedColor=(R=0,G=0,B=0,A=0),ColorUseRule=UseColor_Specified),Width=0,RoundingType=HalfHeightRadius,bUseBrushTransparency=False)
    ) -> (SlateBrush:Slate Brush Structure) {is=title K2Node_MakeStruct}
   )
   n2 = Get NotificationType(self=Get Notification() -> (Notification:Fort UINotification Object Reference)) -> (NotificationType=Default)
   > SetColorAndOpacity(
    self:Image Object Reference=@n1
    InColorAndOpacity:Linear Color Structure=Select(
     Option 0:Linear Color Structure=Make Linear Color!(R:float=1, G:float=0.9544, B:float=0, A:float=1) -> (LinearColor:Linear Color Structure) {is=title K2Node_MakeStruct}
     Option 1:Linear Color Structure=Make Linear Color(R=1, G=1, B=1, A=1)
     Index="Equal (Enum)"(A=@n2, B=GiftSent)
    ) -> (ReturnValue:Linear Color Structure)
   )
   > "SetText (Text)"(
    self=Get Title()
    InText=Select!(
     Option 0:Text="LOCTEXT(\"Oops! There is a small issue.\")"
     Option 1:Text="LOCTEXT(\"Doh! Minor error encountered.\")"
     Option 2:Text="LOCTEXT(\"Oh no! Something didn\\'t work.\")"
     Index:int=RandomIntegerInRange!(Min:int=0, Max:int=2) -> (ReturnValue:int=0)
    ) -> (ReturnValue:Text)
   )
   > "SetText (Text)"(self=Get Description(), InText=Get Title!(self:Fort UINotification Object Reference=Get Notification()) -> (Title:Text))
   > "SetText (Text)"(
    self=Get ProgrammerNote() -> (ProgrammerNote:Common Text Object Reference)
    InText=Get Description!(self:Fort UINotification Object Reference=Get Notification()) -> (Description:Text)
   )
   > SetVisibility(
    self=Get OpenButton() -> (OpenButton:Icon Text Button Object Reference)
    InVisibility=GetOpenButtonVisibility!() -> (ReturnValue:ESlateVisibility Enum=Visible) {is=title K2Node_CallFunction}
   )
   > ApplyAdditionalStyling() {is=title K2Node_CallFunction}
  ]}
]
variables [
 AnimationDelayTimer {type=TimerHandle}
 bActionTaken {type=boolean}
 OnFinishedToast {}
]
widgets {
 info {inheritedWidgetCount=0, parentClass=/Script/FortniteUI.FortToastWidget, rootWidgetClass=&VerticalBox, widgetCount=18}
 widgets [
  RootVerticalBox {widgetClassPath=&VerticalBox, widget=RootVerticalBox, properties {renderTransformPivot=<0 0.5>}}
  SizeBox_2 {
   widgetClassPath=&SizeBox
   parent=RootVerticalBox
   widget=SizeBox_2
   properties {bOverride_WidthOverride=true, maxDesiredHeight=250, maxDesiredWidth=700, minDesiredWidth=400, widthOverride=700}
   slot {class=VerticalBoxSlot}
  }
  VerticalBox_1 {widgetClassPath=&VerticalBox, parent=SizeBox_2, widget=VerticalBox_1, slot {class=SizeBoxSlot}}
  Overlay_1 {widgetClassPath=&Overlay, parent=VerticalBox_1, widget=Overlay_1, slot {class=VerticalBoxSlot}}
  B_Dropshadow {
   widgetClassPath=&Border
   parent=Overlay_1
   widget=B_Dropshadow
   properties {
    background {
     bIsDynamicallyLoaded=false
     drawAs=Image
     imageType=NoImage
     mirroring=NoMirror
     resourceName=None
     resourceObject=/FortUILibrary/Styles/Border/Materials/M_Border_DropShadow
     tiling=NoTile
     imageSize=<32 32>
     margin=<ltrb 0 0 0 0>
     outlineSettings {
      bUseBrushTransparency=false
      roundingType=HalfHeightRadius
      width=0
      color {colorUseRule=UseColor_Specified, specifiedColor=<rgba 0 0 0 0>}
      cornerRadii=<0 0 0 0>
     }
     tintColor {colorUseRule=UseColor_Specified, specifiedColor=<rgba 1 1 1 1>}
     uVRegion {bIsValid=false, max=<0 0>, min=<0 0>}
    }
    renderTransform {angle=0, scale=<1 1>, shear=<0 0>, translation=<5 8>}
    renderTransformPivot=<0 0.5>
   }
   slot {class=OverlaySlot, properties {horizontalAlignment=HAlign_Fill, verticalAlignment=VAlign_Fill}}
  }
  HorizontalBox_1 {
   widgetClassPath=&HorizontalBox
   parent=Overlay_1
   widget=HorizontalBox_1
   slot {class=OverlaySlot, properties {horizontalAlignment=HAlign_Fill, verticalAlignment=VAlign_Fill}}
  }
  Overlay_0 {widgetClassPath=&Overlay, parent=HorizontalBox_1, widget=Overlay_0, slot {class=HorizontalBoxSlot}}
  B_NotificationImageExtended {
   widgetClassPath=&CommonBorder
   bIsVariable=true
   parent=Overlay_0
   widget=B_NotificationImageExtended
   properties {
    horizontalAlignment=HAlign_Center
    style=&Border-SolidBG-RewardBack_C
    verticalAlignment=VAlign_Center
    renderTransformPivot=<0 0.5>
   }
   slot {class=OverlaySlot, properties {horizontalAlignment=HAlign_Fill, verticalAlignment=VAlign_Fill}}
  }
  B_NotificationImage {
   widgetClassPath=&CommonBorder
   bIsVariable=true
   parent=Overlay_0
   widget=B_NotificationImage
   properties {
    style=&Border-SolidBG-RewardBack_C
    brushColor=<rgba 1 1 1 0>
    padding=<ltrb 0 0 0 0>
    renderTransform {angle=-360, scale=<1 1>, shear=<0 0>, translation=<0 0>}
   }
   slot {class=OverlaySlot, properties {horizontalAlignment=HAlign_Center, verticalAlignment=VAlign_Center, padding=<ltrb 4 0 4 0>}}
  }
  SizeBox_0 {
   widgetClassPath=&SizeBox
   parent=B_NotificationImage
   widget=SizeBox_0
   properties {bOverride_HeightOverride=true, bOverride_WidthOverride=true, heightOverride=40, widthOverride=40}
   slot {class=BorderSlot, properties {padding=<ltrb 0 0 0 0>}}
  }
  NotificationImage {
   widgetClassPath=&Image
   bIsVariable=true
   parent=SizeBox_0
   widget=NotificationImage
   properties {
    brush {
     bIsDynamicallyLoaded=false
     drawAs=Image
     imageType=NoImage
     mirroring=NoMirror
     resourceName=None
     resourceObject=&T-Icon-Warning-64
     tiling=NoTile
     imageSize=<64 64>
     margin=<ltrb 0 0 0 0>
     outlineSettings {
      bUseBrushTransparency=false
      roundingType=HalfHeightRadius
      width=0
      color {colorUseRule=UseColor_Specified, specifiedColor=<rgba 0 0 0 0>}
      cornerRadii=<0 0 0 0>
     }
     tintColor {colorUseRule=UseColor_Specified, specifiedColor=<rgba 1 1 1 1>}
     uVRegion {bIsValid=false, max=<0 0>, min=<0 0>}
    }
    colorAndOpacity=<rgba 1 0.9544 0 1>
   }
   slot {class=SizeBoxSlot}
  }
  CommonBorder_MainWidget {
   widgetClassPath=&CommonBorder
   bIsVariable=true
   parent=HorizontalBox_1
   widget=CommonBorder_MainWidget
   properties {
    background {
     bIsDynamicallyLoaded=false
     drawAs=Box
     imageType=NoImage
     mirroring=NoMirror
     resourceName=None
     resourceObject=null
     tiling=NoTile
     imageSize=<52 36>
     margin=<ltrb 0.35 0 0.35 0>
     outlineSettings {
      bUseBrushTransparency=false
      roundingType=HalfHeightRadius
      width=0
      color {colorUseRule=UseColor_Specified, specifiedColor=<rgba 0 0 0 0>}
      cornerRadii=<0 0 0 0>
     }
     tintColor {colorUseRule=UseColor_Specified, specifiedColor=<rgba 1 1 1 0.95>}
     uVRegion {bIsValid=false, max=<0 0>, min=<0 0>}
    }
    padding=<ltrb 0 0 0 0>
    renderTransformPivot=<0 0.5>
   }
   slot {class=HorizontalBoxSlot, properties {size {sizeRule=Fill, value=1}}}
  }
  HorizontalBox_0 {
   widgetClassPath=&HorizontalBox
   parent=CommonBorder_MainWidget
   widget=HorizontalBox_0
   slot {class=BorderSlot, properties {padding=<ltrb 0 0 0 0>}}
  }
  VerticalBox_0 {
   widgetClassPath=&VerticalBox
   parent=HorizontalBox_0
   widget=VerticalBox_0
   slot {class=HorizontalBoxSlot, properties {padding=<ltrb 20 10 10 10>, size {sizeRule=Fill, value=1}}}
  }
  Title {
   widgetClassPath=&CommonTextBlock
   bIsVariable=true
   parent=VerticalBox_0
   widget=Title
   properties {
    autoWrapText=true
    style=/FortUILibrary/Styles/Typography/Foundation/TextStyle-Header-S-Black.TextStyle-Header-S-Black_C
    text=Minor Error.
    visibility=Collapsed
    colorAndOpacity {colorUseRule=UseColor_Specified, specifiedColor=<rgba 0 0 0 1>}
    font {
     bForceMonospaced=false
     bMaterialIsStencil=false
     fontMaterial=null
     fontObject=&Burbank
     letterSpacing=0
     monospacedWidth=1
     size=24
     skewAmount=0
     typefaceFontName=Burbank-Black
     outlineSettings {
      bApplyOutlineToDropShadows=false
      bMiteredCorners=false
      bSeparateFillAlpha=false
      outlineMaterial=null
      outlineSize=0
      outlineColor=<rgba 0 0 0 1>
     }
    }
    shadowOffset=<0 0>
   }
   slot {class=VerticalBoxSlot}
  }
  Description {
   widgetClassPath=&CommonTextBlock
   bIsVariable=true
   parent=VerticalBox_0
   widget=Description
   properties {
    autoWrapText=true
    bAutoCollapseWithEmptyText=true
    style=/FortUILibrary/Styles/Typography/Foundation/TextStyle-Base-XS-B-Black.TextStyle-Base-XS-B-Black_C
    text="This should be a friendly message that is Helpful, Honest, Humble and Human. Seriously, its important!"
    colorAndOpacity {colorUseRule=UseColor_Specified, specifiedColor=<rgba 0 0 0 1>}
    font {
     bForceMonospaced=false
     bMaterialIsStencil=false
     fontMaterial=null
     fontObject=&NotoSans
     letterSpacing=0
     monospacedWidth=1
     size=16
     skewAmount=0
     typefaceFontName=NotoSans-Bold
     outlineSettings {
      bApplyOutlineToDropShadows=false
      bMiteredCorners=false
      bSeparateFillAlpha=false
      outlineMaterial=null
      outlineSize=0
      outlineColor=<rgba 0 0 0 1>
     }
    }
    shadowOffset=<0 0>
   }
   slot {class=VerticalBoxSlot}
  }
  ProgrammerNote {
   widgetClassPath=&CommonTextBlock
   bIsVariable=true
   parent=VerticalBox_0
   widget=ProgrammerNote
   properties {
    autoWrapText=true
    bAutoCollapseWithEmptyText=true
    style=&TextStyle-Base-XXS-Black_C
    text=Minor Error.
    visibility=Visible
    colorAndOpacity {colorUseRule=UseColor_Specified, specifiedColor=<rgba 0 0 0 1>}
    font {
     bForceMonospaced=false
     bMaterialIsStencil=false
     fontMaterial=null
     fontObject=&NotoSans
     letterSpacing=0
     monospacedWidth=1
     size=14
     skewAmount=0
     typefaceFontName=NotoSans-Regular
     outlineSettings {
      bApplyOutlineToDropShadows=false
      bMiteredCorners=false
      bSeparateFillAlpha=false
      outlineMaterial=null
      outlineSize=0
      outlineColor=<rgba 0 0 0 1>
     }
    }
    shadowOffset=<0 0>
   }
   slot {class=VerticalBoxSlot}
  }
  OpenButton {
   widgetClassPath=&IconTextButton
   bIsVariable=true
   parent=HorizontalBox_0
   widget=OpenButton
   properties {
    bIsPersistentBinding=true
    buttonText=Open
    controllerInputStyle=/FortUILibrary/Foundation/Buttons/ButtonStyles/Primary/ButtonStyle-Primary-M.ButtonStyle-Primary-M_C
    onButtonBaseClicked=(&self:WidgetTree.OpenButton.HandleOnButtonClicked)
    onButtonBaseDoubleClicked=(&self:WidgetTree.OpenButton.HandleOnButtonDoubleClicked)
    onButtonBaseHovered=(&self:WidgetTree.OpenButton.HandleOnButtonHovered)
    onButtonBaseUnhovered=(&self:WidgetTree.OpenButton.HandleOnButtonUnhovered)
    onSelectedChangedBase=(&self:WidgetTree.OpenButton.HandleOnSelectedChanged)
    visibility=Collapsed
    triggeringInputAction {dataTable=&InputActionDataTable, rowName=ToastOpen}
   }
   slot {class=HorizontalBoxSlot, properties {horizontalAlignment=HAlign_Left, padding=<ltrb 8 8 8 8>}}
  }
 ]
}
