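# HUD popup, derived from FortVoiceChannelChangePopup, that announces a voice channel change with a mic icon, three text lines and switchable action or hold-to-join-game-chat buttons.
# Construct starts an endlessly looping mic pulse and binds the popup animation's finish event.
# Opening plays the popup animation; closing plays it in reverse, with a flag collapsing the widget on finish.
# Input mode changes swap keyboard and gamepad keybind prompts; material color/progress setter nodes are unconnected to events.
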WidgetBlueprint /Game/UI/InGame/HUD/VoiceChannelChangePopup
animations [
 Anim_SimplePopUp {
  binding_count=4
  length_seconds=0.2
  object=&self:Anim_SimplePopUp
  sequence {bindings [
    IconSB {
     id=#1
     possessed_class=SizeBox
     tracks [
      {
       type=MovieScene2DTransformTrack
       sections [{
         type=MovieScene2DTransformSection
         channels [
          Translation.X {type=MovieSceneScriptingFloatChannel, keys [[0,8],[3,-6],[4,0]]}
          Translation.Y {type=MovieSceneScriptingFloatChannel, const [0,1]}
          Angle {type=MovieSceneScriptingFloatChannel}
          Scale.X {type=MovieSceneScriptingFloatChannel, keys [[0,0.875],[3,1.125],[4,1]]}
          Scale.Y {type=MovieSceneScriptingFloatChannel, keys [[0,0.875],[3,1.125],[4,1]]}
          Shear.X {type=MovieSceneScriptingFloatChannel}
          Shear.Y {type=MovieSceneScriptingFloatChannel}
         ]
        }]
      }
      {
       type=MovieSceneFloatTrack
       display_name=Render Opacity
       sections [{type=MovieSceneFloatSection, end_frame=4, start_frame=0, channels [None {type=MovieSceneScriptingFloatChannel, keys [[0,0],[2,1]]}]}]
      }
     ]
    }
    ContentVB {
     id=#2
     possessed_class=VerticalBox
     tracks [{
       type=MovieScene2DTransformTrack
       sections [{
         type=MovieScene2DTransformSection
         channels [
          Translation.X {type=MovieSceneScriptingFloatChannel, keys [[0,-32],[4,0]]}
          Translation.Y {type=MovieSceneScriptingFloatChannel}
          Angle {type=MovieSceneScriptingFloatChannel}
          Scale.X {type=MovieSceneScriptingFloatChannel}
          Scale.Y {type=MovieSceneScriptingFloatChannel}
          Shear.X {type=MovieSceneScriptingFloatChannel}
          Shear.Y {type=MovieSceneScriptingFloatChannel}
         ]
        }]
      }]
    }
    AnimatedInputBorder {
     id=#3
     possessed_class=Border
     tracks [
      {
       type=MovieSceneColorTrack
       display_name=Brush Color
       sections [{
         type=MovieSceneColorSection
         end_frame=4
         start_frame=0
         channels [
          Color.R {type=MovieSceneScriptingFloatChannel}
          Color.G {type=MovieSceneScriptingFloatChannel}
          Color.B {type=MovieSceneScriptingFloatChannel}
          Color.A {type=MovieSceneScriptingFloatChannel, keys [[0,0],[2,1]]}
         ]
        }]
      }
      {
       type=MovieSceneWidgetMaterialTrack
       sections [{
         type=MovieSceneParameterSection
         end_frame=4
         start_frame=0
         channels [AnimateStateChange {type=MovieSceneScriptingFloatChannel, keys [[0,0],[3,1],[4,1]]}]
        }]
      }
     ]
    }
    Border_ChannelInfo {
     id=#4
     possessed_class=Border
     tracks [
      {
       type=MovieSceneFloatVectorTrack
       display_name=Desired Size Scale
       sections [{
         type=MovieSceneFloatVectorSection
         channels [
          Vector.X {type=MovieSceneScriptingFloatChannel, keys [[0,0],[4,1]]}
          Vector.Y {type=MovieSceneScriptingFloatChannel}
         ]
        }]
      }
      {
       type=MovieSceneFloatTrack
       display_name=Render Opacity
       sections [{type=MovieSceneFloatSection, channels [None {type=MovieSceneScriptingFloatChannel, keys [[0,0],[4,1]]}]}]
      }
     ]
    }
   ]}
 }
 AmbientIconPulse {
  binding_count=1
  length_seconds=0.8
  object=&self:AmbientIconPulse
  sequence {
   bindings [MicPulse {
     id=#5
     possessed_class=Image
     tracks [
      {
       type=MovieSceneFloatTrack
       display_name=Render Opacity
       sections [{type=MovieSceneFloatSection, channels [None {type=MovieSceneScriptingFloatChannel, keys [[0,1],[40,0],[80,0]]}]}]
      }
      {
       type=MovieScene2DTransformTrack
       sections [{
         type=MovieScene2DTransformSection
         channels [
          Translation.X {type=MovieSceneScriptingFloatChannel}
          Translation.Y {type=MovieSceneScriptingFloatChannel}
          Angle {type=MovieSceneScriptingFloatChannel}
          Scale.X {type=MovieSceneScriptingFloatChannel, keys [[0,1],[40,2.5]]}
          Scale.Y {type=MovieSceneScriptingFloatChannel, keys [[0,1],[40,2.5]]}
          Shear.X {type=MovieSceneScriptingFloatChannel}
          Shear.Y {type=MovieSceneScriptingFloatChannel}
         ]
        }]
      }
     ]
    }]
   display_rate {numerator=100}
  }
 }
]
graphs [EventGraph {nodes [
   n1 = GetDynamicMaterial()
   n2 = Get M_Param_ColorA() -> (M_Param_ColorA:Name=None)
   n3 = Get ColorA() -> (ColorA:Linear Color Structure)
   > SetVectorParameterValue(self=@n1, ParameterName=@n2, Value=@n3)
   n4 = Get M_Param_ColorB() -> (M_Param_ColorB:Name=None)
   n5 = Get ColorB() -> (ColorB:Linear Color Structure)
   > SetVectorParameterValue(self=@n1, ParameterName=@n4, Value=@n5)
   n6 = Get M_Param_Progress() -> (M_Param_Progress:Name=None)
   > SetScalarParameterValue(self=@n1, ParameterName=@n6, Value=1)
   n7 = GetDynamicMaterial()
   > SetVectorParameterValue(self=@n7, ParameterName=@n2, Value=@n3)
   > SetVectorParameterValue(self=@n7, ParameterName=@n4, Value=@n5)
   > SetScalarParameterValue(self=@n7, ParameterName=@n6, Value=1)
   n8 = Popup Animation Finished() {is=title K2Node_CustomEvent}
   > Branch(Condition=Get CollapseOnAnimationFinish() -> (CollapseOnAnimationFinish:bool))
   > Set CollapseOnAnimationFinish(CollapseOnAnimationFinish:bool=false) -> (Output_Get:bool)
   > SetVisibility(InVisibility=Collapsed)
   Event OnPlayOpenAnimation() {is=title K2Node_Event}
   > Set CollapseOnAnimationFinish(CollapseOnAnimationFinish=false)
   n9 = Get Anim_SimplePopUp() -> (Anim_SimplePopUp:Widget Animation Object Reference)
   > PlayAnimation(InAnimation=@n9)
   Event Construct()
   > PlayAnimation(InAnimation=Get AmbientIconPulse() -> (AmbientIconPulse:Widget Animation Object Reference), NumLoopsToPlay=0)
   > BindToAnimationFinished(Animation=Get Anim_SimplePopUp(), Delegate=@n8)
   n10 = Event OnInputModeChanged!() -> (OutputDelegate, then, NewInputType:ECommonInputType Enum) {is=title K2Node_Event}
   n11 = Get KBKeybindWrapper() -> (KBKeybindWrapper:Horizontal Box Object Reference)
   > SetActiveWidget(
    self:Widget Switcher Object Reference=Get Switcher_KeybindWidgets() -> (Switcher_KeybindWidgets:Common Widget Switcher Legacy Object Reference)
    Widget=Select!(
     MouseAndKeyboard:Widget Object Reference=@n11
     Gamepad:Widget Object Reference=Get GamepadKeybindWrapper() -> (GamepadKeybindWrapper:Horizontal Box Object Reference)
     Touch:Widget Object Reference=@n11
     Count:Widget Object Reference=@n11
     Index:ECommonInputType Enum=@n10.NewInputType
    ) -> (ReturnValue:Widget Object Reference)
   )
   Event OnPlayCloseAnimation() {is=title K2Node_Event}
   > Set CollapseOnAnimationFinish(CollapseOnAnimationFinish=true)
   > PlayAnimation(InAnimation=@n9, PlayMode=Reverse)
   Event PreConstruct()
  ]}]
variables [
 CollapseOnAnimationFinish {type=boolean}
 ColorA {type=LinearColor, default=<rgba 0 0.7916 1 1>}
 ColorB {type=LinearColor, default=<rgba 0.1719 0.8275 1 1>}
 M_Param_ColorA {type=string, default=ColorA}
 M_Param_ColorB {type=string, default=ColorB}
 M_Param_Progress {type=string, default=Progress}
]
widgets {
 info {
  inheritedWidgetCount=8
  namedSlotCount=1
  parentClass=/Script/FortniteUI.FortVoiceChannelChangePopup
  rootWidgetClass=&VerticalBox
  widgetCount=24
 }
 widgets [
  VerticalBox_0 {widgetClassPath=&VerticalBox, widget=VerticalBox_0}
  Border_ChannelInfo {
   widgetClassPath=&Border
   parent=VerticalBox_0
   widget=Border_ChannelInfo
   properties {
    verticalAlignment=VAlign_Center
    background {
     bIsDynamicallyLoaded=false
     drawAs=Image
     imageType=NoImage
     mirroring=NoMirror
     resourceName=None
     resourceObject=/Game/UI/SocialPanel/Art/Materials/M_UI_SocialUserItemBorder_ChannelNotification
     tiling=NoTile
     imageSize=<32 32>
     margin=<ltrb 0 0 0 0>
     outlineSettings {
      bUseBrushTransparency=false
      roundingType=HalfHeightRadius
      width=0
      color {colorUseRule=UseColor_Specified, specifiedColor=<rgba 0 0 0 0>}
      cornerRadii=<0 0 0 0>
     }
     tintColor {colorUseRule=UseColor_Specified, specifiedColor=<rgba 1 1 1 1>}
     uVRegion {bIsValid=false, max=<0 0>, min=<0 0>}
    }
    padding=<ltrb 0 0 0 0>
   }
   slot {class=VerticalBoxSlot}
  }
  VerticalBox_2 {
   widgetClassPath=&VerticalBox
   parent=Border_ChannelInfo
   widget=VerticalBox_2
   slot {class=BorderSlot, properties {verticalAlignment=VAlign_Center, padding=<ltrb 0 0 0 0>}}
  }
  HorizontalBox_3 {widgetClassPath=&HorizontalBox, parent=VerticalBox_2, widget=HorizontalBox_3, slot {class=VerticalBoxSlot}}
  AnimatedInputBorder {
   widgetClassPath=&Border
   parent=HorizontalBox_3
   widget=AnimatedInputBorder
   properties {
    verticalAlignment=VAlign_Center
    visibility=SelfHitTestInvisible
    background {
     bIsDynamicallyLoaded=false
     drawAs=Image
     imageType=NoImage
     mirroring=NoMirror
     resourceName=None
     resourceObject=/Game/UI/SocialPanel/Art/Materials/M_UI_UserHover_Inst_Flipped
     tiling=NoTile
     imageSize=<64 64>
     margin=<ltrb 1 0 0 0>
     outlineSettings {
      bUseBrushTransparency=false
      roundingType=HalfHeightRadius
      width=0
      color {colorUseRule=UseColor_Specified, specifiedColor=<rgba 0 0 0 0>}
      cornerRadii=<0 0 0 0>
     }
     tintColor {colorUseRule=UseColor_Specified, specifiedColor=<rgba 1 1 1 0.3>}
     uVRegion {bIsValid=false, max=<0 0>, min=<0 0>}
    }
    brushColor=<rgba 1 1 1 0.5>
    padding=<ltrb 12 0 16 0>
   }
   slot {class=HorizontalBoxSlot, properties {horizontalAlignment=HAlign_Left}}
  }
  IconSB {
   widgetClassPath=&SizeBox
   parent=AnimatedInputBorder
   widget=IconSB
   properties {bOverride_HeightOverride=true, bOverride_WidthOverride=true, heightOverride=48, widthOverride=48}
   slot {class=BorderSlot, properties {verticalAlignment=VAlign_Center, padding=<ltrb 12 0 16 0>}}
  }
  Overlay_0 {widgetClassPath=&Overlay, parent=IconSB, widget=Overlay_0, slot {class=SizeBoxSlot}}
  MicPulse {
   widgetClassPath=&Image
   bIsVariable=true
   parent=Overlay_0
   widget=MicPulse
   properties {brush {
     bIsDynamicallyLoaded=false
     drawAs=Image
     imageType=NoImage
     mirroring=NoMirror
     resourceName=None
     resourceObject=/Game/UI/SocialPanel/Art/Textures/T_UI_Icon_Mic_64
     tiling=NoTile
     imageSize=<64 64>
     margin=<ltrb 0 0 0 0>
     outlineSettings {
      bUseBrushTransparency=false
      roundingType=HalfHeightRadius
      width=0
      color {colorUseRule=UseColor_Specified, specifiedColor=<rgba 0 0 0 0>}
      cornerRadii=<0 0 0 0>
     }
     tintColor {colorUseRule=UseColor_Specified, specifiedColor=<rgba 0.0723 0.8714 1 1>}
     uVRegion {bIsValid=false, max=<0 0>, min=<0 0>}
    }}
   slot {class=OverlaySlot, properties {horizontalAlignment=HAlign_Fill, verticalAlignment=VAlign_Fill}}
  }
  Image_Mic {
   widgetClassPath=&Image
   bIsVariable=true
   parent=Overlay_0
   widget=Image_Mic
   properties {brush {
     bIsDynamicallyLoaded=false
     drawAs=Image
     imageType=NoImage
     mirroring=NoMirror
     resourceName=None
     resourceObject=&T_UI_Icon_Mic_64
     tiling=NoTile
     imageSize=<64 64>
     margin=<ltrb 0 0 0 0>
     outlineSettings {
      bUseBrushTransparency=false
      roundingType=HalfHeightRadius
      width=0
      color {colorUseRule=UseColor_Specified, specifiedColor=<rgba 0 0 0 0>}
      cornerRadii=<0 0 0 0>
     }
     tintColor {colorUseRule=UseColor_Specified, specifiedColor=<rgba 1 1 1 1>}
     uVRegion {bIsValid=false, max=<0 0>, min=<0 0>}
    }}
   slot {class=OverlaySlot, properties {horizontalAlignment=HAlign_Fill, verticalAlignment=VAlign_Fill}}
  }
  ContentVB {
   widgetClassPath=&VerticalBox
   parent=HorizontalBox_3
   widget=ContentVB
   properties {flowDirectionPreference=Culture}
   slot {class=HorizontalBoxSlot, properties {padding=<ltrb 16 8 16 8>, size {sizeRule=Fill, value=1}}}
  }
  Text_TopLine {
   widgetClassPath=&CommonRichTextBlock
   bInherited=true
   bIsVariable=true
   parent=ContentVB
   widget=Text_TopLine
   properties {
    bOverrideDefaultStyle=true
    mobileTextBlockScale=1.2
    text=Joined
    textTransformPolicy=ToUpper
    defaultTextStyleOverride {
     overflowPolicy=Clip
     transformPolicy=None
     colorAndOpacity {colorUseRule=UseColor_Specified, specifiedColor=<rgba 0.4452 1 1 1>}
     font {
      bForceMonospaced=false
      bMaterialIsStencil=false
      fontMaterial=null
      fontObject=&HeadingNow
      letterSpacing=0
      monospacedWidth=1
      size=18
      skewAmount=0
      typefaceFontName=75 Medium
      outlineSettings {
       bApplyOutlineToDropShadows=false
       bMiteredCorners=false
       bSeparateFillAlpha=false
       outlineMaterial=null
       outlineSize=0
       outlineColor=<rgba 0 0 0 1>
      }
     }
     highlightColor {colorUseRule=UseColor_Specified, specifiedColor=<rgba 0 0 0 1>}
     highlightShape {
      bIsDynamicallyLoaded=false
      drawAs=Image
      imageType=NoImage
      mirroring=NoMirror
      resourceName=None
      resourceObject=null
      tiling=NoTile
      imageSize=<32 32>
      margin=<ltrb 0 0 0 0>
      outlineSettings {
       bUseBrushTransparency=false
       roundingType=HalfHeightRadius
       width=0
       color {colorUseRule=UseColor_Specified, specifiedColor=<rgba 0 0 0 0>}
       cornerRadii=<0 0 0 0>
      }
      tintColor {colorUseRule=UseColor_Specified, specifiedColor=<rgba 1 1 1 1>}
      uVRegion {bIsValid=false, max=<0 0>, min=<0 0>}
     }
     selectedBackgroundColor {colorUseRule=UseColor_Foreground, specifiedColor=<rgba 1 0 1 1>}
     shadowColorAndOpacity=<rgba 0 0 0 1>
     shadowOffset=<0 0>
     strikeBrush {
      bIsDynamicallyLoaded=false
      drawAs=Image
      imageType=NoImage
      mirroring=NoMirror
      resourceName=None
      resourceObject=null
      tiling=NoTile
      imageSize=<32 32>
      margin=<ltrb 0 0 0 0>
      outlineSettings {
       bUseBrushTransparency=false
       roundingType=HalfHeightRadius
       width=0
       color {colorUseRule=UseColor_Specified, specifiedColor=<rgba 0 0 0 0>}
       cornerRadii=<0 0 0 0>
      }
      tintColor {colorUseRule=UseColor_Specified, specifiedColor=<rgba 1 1 1 1>}
      uVRegion {bIsValid=false, max=<0 0>, min=<0 0>}
     }
     underlineBrush {
      bIsDynamicallyLoaded=false
      drawAs=Image
      imageType=NoImage
      mirroring=NoMirror
      resourceName=None
      resourceObject=null
      tiling=NoTile
      imageSize=<32 32>
      margin=<ltrb 0 0 0 0>
      outlineSettings {
       bUseBrushTransparency=false
       roundingType=HalfHeightRadius
       width=0
       color {colorUseRule=UseColor_Specified, specifiedColor=<rgba 0 0 0 0>}
       cornerRadii=<0 0 0 0>
      }
      tintColor {colorUseRule=UseColor_Specified, specifiedColor=<rgba 1 1 1 1>}
      uVRegion {bIsValid=false, max=<0 0>, min=<0 0>}
     }
    }
    renderTransform {angle=0, scale=<1 1>, shear=<-8 0>, translation=<0 0>}
   }
   slot {class=VerticalBoxSlot}
  }
  Text_MiddleLine {
   widgetClassPath=&CommonRichTextBlock
   bInherited=true
   bIsVariable=true
   parent=ContentVB
   widget=Text_MiddleLine
   properties {
    autoWrapText=true
    bOverrideDefaultStyle=true
    mobileTextBlockScale=1.1
    scrollStyle=/FortUILibrary/Styles/Typography/Foundation/TextScrollStyle-NoFade.TextScrollStyle-NoFade_C
    text=Channel : Game
    textTransformPolicy=ToUpper
    wrapTextAt=600
    defaultTextStyleOverride {
     overflowPolicy=Clip
     transformPolicy=None
     colorAndOpacity {colorUseRule=UseColor_Specified, specifiedColor=<rgba 0.4452 1 1 1>}
     font {
      bForceMonospaced=false
      bMaterialIsStencil=false
      fontMaterial=null
      fontObject=&HeadingNow
      letterSpacing=0
      monospacedWidth=1
      size=20.25
      skewAmount=0
      typefaceFontName=75 Medium
      outlineSettings {
       bApplyOutlineToDropShadows=false
       bMiteredCorners=false
       bSeparateFillAlpha=false
       outlineMaterial=null
       outlineSize=0
       outlineColor=<rgba 0 0 0 1>
      }
     }
     highlightColor {colorUseRule=UseColor_Specified, specifiedColor=<rgba 0 0 0 1>}
     highlightShape {
      bIsDynamicallyLoaded=false
      drawAs=Image
      imageType=NoImage
      mirroring=NoMirror
      resourceName=None
      resourceObject=null
      tiling=NoTile
      imageSize=<32 32>
      margin=<ltrb 0 0 0 0>
      outlineSettings {
       bUseBrushTransparency=false
       roundingType=HalfHeightRadius
       width=0
       color {colorUseRule=UseColor_Specified, specifiedColor=<rgba 0 0 0 0>}
       cornerRadii=<0 0 0 0>
      }
      tintColor {colorUseRule=UseColor_Specified, specifiedColor=<rgba 1 1 1 1>}
      uVRegion {bIsValid=false, max=<0 0>, min=<0 0>}
     }
     selectedBackgroundColor {colorUseRule=UseColor_Foreground, specifiedColor=<rgba 1 0 1 1>}
     shadowColorAndOpacity=<rgba 0 0 0 1>
     shadowOffset=<0 0>
     strikeBrush {
      bIsDynamicallyLoaded=false
      drawAs=Image
      imageType=NoImage
      mirroring=NoMirror
      resourceName=None
      resourceObject=null
      tiling=NoTile
      imageSize=<32 32>
      margin=<ltrb 0 0 0 0>
      outlineSettings {
       bUseBrushTransparency=false
       roundingType=HalfHeightRadius
       width=0
       color {colorUseRule=UseColor_Specified, specifiedColor=<rgba 0 0 0 0>}
       cornerRadii=<0 0 0 0>
      }
      tintColor {colorUseRule=UseColor_Specified, specifiedColor=<rgba 1 1 1 1>}
      uVRegion {bIsValid=false, max=<0 0>, min=<0 0>}
     }
     underlineBrush {
      bIsDynamicallyLoaded=false
      drawAs=Image
      imageType=NoImage
      mirroring=NoMirror
      resourceName=None
      resourceObject=null
      tiling=NoTile
      imageSize=<32 32>
      margin=<ltrb 0 0 0 0>
      outlineSettings {
       bUseBrushTransparency=false
       roundingType=HalfHeightRadius
       width=0
       color {colorUseRule=UseColor_Specified, specifiedColor=<rgba 0 0 0 0>}
       cornerRadii=<0 0 0 0>
      }
      tintColor {colorUseRule=UseColor_Specified, specifiedColor=<rgba 1 1 1 1>}
      uVRegion {bIsValid=false, max=<0 0>, min=<0 0>}
     }
    }
   }
   slot {class=VerticalBoxSlot}
  }
  Text_BottomLine {
   widgetClassPath=&CommonRichTextBlock
   bInherited=true
   bIsVariable=true
   parent=ContentVB
   widget=Text_BottomLine
   properties {
    bOverrideDefaultStyle=true
    mobileTextBlockScale=1.2
    scrollStyle=&TextScrollStyle-NoFade_C
    text=Switch within the Friends Menu
    wrapTextAt=384
    defaultTextStyleOverride {
     overflowPolicy=Clip
     transformPolicy=None
     colorAndOpacity {colorUseRule=UseColor_Specified, specifiedColor=<rgba 0.3185 0.7231 1 1>}
     font {
      bForceMonospaced=false
      bMaterialIsStencil=false
      fontMaterial=null
      fontObject=&HeadingNow
      letterSpacing=10
      monospacedWidth=1
      size=15.75
      skewAmount=0
      typefaceFontName=75 Medium
      outlineSettings {
       bApplyOutlineToDropShadows=false
       bMiteredCorners=false
       bSeparateFillAlpha=false
       outlineMaterial=null
       outlineSize=0
       outlineColor=<rgba 0 0 0 1>
      }
     }
     highlightColor {colorUseRule=UseColor_Specified, specifiedColor=<rgba 0 0 0 1>}
     highlightShape {
      bIsDynamicallyLoaded=false
      drawAs=Image
      imageType=NoImage
      mirroring=NoMirror
      resourceName=None
      resourceObject=null
      tiling=NoTile
      imageSize=<32 32>
      margin=<ltrb 0 0 0 0>
      outlineSettings {
       bUseBrushTransparency=false
       roundingType=HalfHeightRadius
       width=0
       color {colorUseRule=UseColor_Specified, specifiedColor=<rgba 0 0 0 0>}
       cornerRadii=<0 0 0 0>
      }
      tintColor {colorUseRule=UseColor_Specified, specifiedColor=<rgba 1 1 1 1>}
      uVRegion {bIsValid=false, max=<0 0>, min=<0 0>}
     }
     selectedBackgroundColor {colorUseRule=UseColor_Foreground, specifiedColor=<rgba 1 0 1 1>}
     shadowColorAndOpacity=<rgba 0 0 0 1>
     shadowOffset=<0 0>
     strikeBrush {
      bIsDynamicallyLoaded=false
      drawAs=Image
      imageType=NoImage
      mirroring=NoMirror
      resourceName=None
      resourceObject=null
      tiling=NoTile
      imageSize=<32 32>
      margin=<ltrb 0 0 0 0>
      outlineSettings {
       bUseBrushTransparency=false
       roundingType=HalfHeightRadius
       width=0
       color {colorUseRule=UseColor_Specified, specifiedColor=<rgba 0 0 0 0>}
       cornerRadii=<0 0 0 0>
      }
      tintColor {colorUseRule=UseColor_Specified, specifiedColor=<rgba 1 1 1 1>}
      uVRegion {bIsValid=false, max=<0 0>, min=<0 0>}
     }
     underlineBrush {
      bIsDynamicallyLoaded=false
      drawAs=Image
      imageType=NoImage
      mirroring=NoMirror
      resourceName=None
      resourceObject=null
      tiling=NoTile
      imageSize=<32 32>
      margin=<ltrb 0 0 0 0>
      outlineSettings {
       bUseBrushTransparency=false
       roundingType=HalfHeightRadius
       width=0
       color {colorUseRule=UseColor_Specified, specifiedColor=<rgba 0 0 0 0>}
       cornerRadii=<0 0 0 0>
      }
      tintColor {colorUseRule=UseColor_Specified, specifiedColor=<rgba 1 1 1 1>}
      uVRegion {bIsValid=false, max=<0 0>, min=<0 0>}
     }
    }
   }
   slot {class=VerticalBoxSlot, properties {size {sizeRule=Fill, value=1}}}
  }
  VerticalBox_128 {widgetClassPath=&VerticalBox, parent=HorizontalBox_3, widget=VerticalBox_128, slot {class=HorizontalBoxSlot}}
  Switcher_InputButtons {
   widgetClassPath=&CommonWidgetSwitcherLegacy
   bInherited=true
   bIsVariable=true
   parent=VerticalBox_128
   widget=Switcher_InputButtons
   slot {class=VerticalBoxSlot, properties {horizontalAlignment=HAlign_Center, padding=<ltrb 8 8 8 8>, size {sizeRule=Fill, value=1}}}
  }
  ActionButton {
   widgetClassPath=&IconTextButton
   bInherited=true
   bIsVariable=true
   parent=Switcher_InputButtons
   widget=ActionButton
   properties {
    bIsPersistentBinding=true
    buttonText=InputDefinedText
    onButtonBaseClicked=(&self:WidgetTree.ActionButton.HandleOnButtonClicked)
    onButtonBaseDoubleClicked=(&self:WidgetTree.ActionButton.HandleOnButtonDoubleClicked)
    onButtonBaseHovered=(&self:WidgetTree.ActionButton.HandleOnButtonHovered)
    onButtonBaseUnhovered=(&self:WidgetTree.ActionButton.HandleOnButtonUnhovered)
    onSelectedChangedBase=(&self:WidgetTree.ActionButton.HandleOnSelectedChanged)
    style=/FortUILibrary/Foundation/Buttons/ButtonStyles/Tab/ButtonStyle-Tab-Main-Recolor.ButtonStyle-Tab-Main-Recolor_C
    triggeringInputAction {dataTable=&InputActionDataTable, rowName=ToastAccept}
   }
   slot {class=WidgetSwitcherSlot}
  }
  Button_JoinGameChat {
   widgetClassPath=&IconTextButton
   bInherited=true
   bIsVariable=true
   parent=Switcher_InputButtons
   widget=Button_JoinGameChat
   properties {
    bIsPersistentBinding=true
    buttonText=""
    onButtonBaseClicked=(&self:WidgetTree.Button_JoinGameChat.HandleOnButtonClicked)
    onButtonBaseDoubleClicked=(&self:WidgetTree.Button_JoinGameChat.HandleOnButtonDoubleClicked)
    onButtonBaseHovered=(&self:WidgetTree.Button_JoinGameChat.HandleOnButtonHovered)
    onButtonBaseUnhovered=(&self:WidgetTree.Button_JoinGameChat.HandleOnButtonUnhovered)
    onSelectedChangedBase=(&self:WidgetTree.Button_JoinGameChat.HandleOnSelectedChanged)
    style=&ButtonStyle-Tab-Main-Recolor_C
   }
   slot {class=WidgetSwitcherSlot}
  }
  HorizontalBox_0 {widgetClassPath=&HorizontalBox, namedSlotHost=Button_JoinGameChat, widget=HorizontalBox_0}
  Switcher_KeybindWidgets {
   widgetClassPath=&CommonWidgetSwitcherLegacy
   bIsVariable=true
   parent=HorizontalBox_0
   widget=Switcher_KeybindWidgets
   slot {class=HorizontalBoxSlot, properties {padding=<ltrb 8 8 8 8>}}
  }
  KBKeybindWrapper {
   widgetClassPath=&HorizontalBox
   bIsVariable=true
   parent=Switcher_KeybindWidgets
   widget=KBKeybindWrapper
   slot {class=WidgetSwitcherSlot}
  }
  HoldKeybindWidget_Keyboard {
   widgetClassPath=/FortUILibrary/Input/KeybindWidget
   bInherited=true
   bIsVariable=true
   parent=KBKeybindWrapper
   widget=HoldKeybindWidget_Keyboard
   properties {boundAction=GameChannelRecommendationHold, forcedHoldKeybindStatus=ForcedHold}
   slot {class=HorizontalBoxSlot}
  }
  GamepadKeybindWrapper {
   widgetClassPath=&HorizontalBox
   bIsVariable=true
   parent=Switcher_KeybindWidgets
   widget=GamepadKeybindWrapper
   slot {class=WidgetSwitcherSlot}
  }
  HoldKeybindWidget_GamePad {
   widgetClassPath=&KeybindWidget
   bInherited=true
   bIsVariable=true
   parent=GamepadKeybindWrapper
   widget=HoldKeybindWidget_GamePad
   properties {boundAction=GamepadGameChannelRecommendationHold, forcedHoldKeybindStatus=ForcedHold}
   slot {class=HorizontalBoxSlot}
  }
  Text_Hold {
   widgetClassPath=&CommonRichTextBlock
   bIsVariable=true
   parent=HorizontalBox_0
   widget=Text_Hold
   properties {
    bOverrideDefaultStyle=true
    text=Hold
    textTransformPolicy=ToUpper
    defaultTextStyleOverride {
     overflowPolicy=Clip
     transformPolicy=None
     colorAndOpacity {colorUseRule=UseColor_Specified, specifiedColor=<rgba 0.3185 0.7231 1 1>}
     font {
      bForceMonospaced=false
      bMaterialIsStencil=false
      fontMaterial=null
      fontObject=&BurbankSmall
      letterSpacing=0
      monospacedWidth=1
      size=15
      skewAmount=0
      typefaceFontName=Black
      outlineSettings {
       bApplyOutlineToDropShadows=false
       bMiteredCorners=false
       bSeparateFillAlpha=false
       outlineMaterial=null
       outlineSize=0
       outlineColor=<rgba 0 0 0 1>
      }
     }
     highlightColor {colorUseRule=UseColor_Specified, specifiedColor=<rgba 0 0 0 1>}
     highlightShape {
      bIsDynamicallyLoaded=false
      drawAs=Image
      imageType=NoImage
      mirroring=NoMirror
      resourceName=None
      resourceObject=null
      tiling=NoTile
      imageSize=<32 32>
      margin=<ltrb 0 0 0 0>
      outlineSettings {
       bUseBrushTransparency=false
       roundingType=HalfHeightRadius
       width=0
       color {colorUseRule=UseColor_Specified, specifiedColor=<rgba 0 0 0 0>}
       cornerRadii=<0 0 0 0>
      }
      tintColor {colorUseRule=UseColor_Specified, specifiedColor=<rgba 1 1 1 1>}
      uVRegion {bIsValid=false, max=<0 0>, min=<0 0>}
     }
     selectedBackgroundColor {colorUseRule=UseColor_Foreground, specifiedColor=<rgba 1 0 1 1>}
     shadowColorAndOpacity=<rgba 0 0 0 1>
     shadowOffset=<0 0>
     strikeBrush {
      bIsDynamicallyLoaded=false
      drawAs=Image
      imageType=NoImage
      mirroring=NoMirror
      resourceName=None
      resourceObject=null
      tiling=NoTile
      imageSize=<32 32>
      margin=<ltrb 0 0 0 0>
      outlineSettings {
       bUseBrushTransparency=false
       roundingType=HalfHeightRadius
       width=0
       color {colorUseRule=UseColor_Specified, specifiedColor=<rgba 0 0 0 0>}
       cornerRadii=<0 0 0 0>
      }
      tintColor {colorUseRule=UseColor_Specified, specifiedColor=<rgba 1 1 1 1>}
      uVRegion {bIsValid=false, max=<0 0>, min=<0 0>}
     }
     underlineBrush {
      bIsDynamicallyLoaded=false
      drawAs=Image
      imageType=NoImage
      mirroring=NoMirror
      resourceName=None
      resourceObject=null
      tiling=NoTile
      imageSize=<32 32>
      margin=<ltrb 0 0 0 0>
      outlineSettings {
       bUseBrushTransparency=false
       roundingType=HalfHeightRadius
       width=0
       color {colorUseRule=UseColor_Specified, specifiedColor=<rgba 0 0 0 0>}
       cornerRadii=<0 0 0 0>
      }
      tintColor {colorUseRule=UseColor_Specified, specifiedColor=<rgba 1 1 1 1>}
      uVRegion {bIsValid=false, max=<0 0>, min=<0 0>}
     }
    }
   }
   slot {class=HorizontalBoxSlot, properties {verticalAlignment=VAlign_Center, padding=<ltrb 0 0 8 0>}}
  }
 ]
}
